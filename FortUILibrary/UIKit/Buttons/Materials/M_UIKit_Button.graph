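# Translucent UI material for UIKit buttons, drawing a rounded backing, an irregular "wonk" border and animated noise.
# The backing uses an SDF-style box function; the border combines four independently rotated, offset edges with rounded corners.
# Two panning screen-aligned noise samples multiply into optional alpha variation.
# Shared state functions blend fill and border colors across idle, hovered, focused, pressed and disabled, with curve-atlas transition easing and static switches selecting features.
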Material /FortUILibrary/UIKit/Buttons/Materials/M_UIKit_Button
graph {
 properties {
  bUsedWithStaticMesh=false
  blendMode=BLEND_Translucent
  materialDomain=MD_UI
  pixelDepthOffsetMode=PDOM_Legacy
  usedShadingModels=Default Lit
 }
 nodes [
  n1 = LinearInterpolate(
   A=Max(A=ScalarParameter() {defaultValue=1, group=Backing, parameterName=Backing_Scale_0, sortPriority=2}) {constB=0.001}
   B=Max(A=ScalarParameter() {defaultValue=1, group=Backing, parameterName=Backing_Scale_1, sortPriority=3}) {constB=0.001}
   Alpha=ScalarParameter() {group=Backing, parameterName=BackingScale_Anim, sortPriority=4}
  ) {constA=1}
  n2 = LinearInterpolate(
   A=ScalarParameter() {defaultValue=0.35, group=Backing, parameterName=Backing_CornerRadius_0, sortPriority=12}
   B=ScalarParameter() {defaultValue=0.125, group=Backing, parameterName=Backing_CornerRadius_1, sortPriority=13}
   Alpha=ScalarParameter() {group=Backing, parameterName=Backing_CornerRadius_Anim, sortPriority=14}
  )
  n3 = MaterialFunctionCall() {materialFunction=&GetUserInterfaceUV}
  NamedRerouteDeclaration(None=Saturate(None=MaterialFunctionCall(UVs=@n3.9-Slice UV, Size=@n1, CornerRoundness=@n2) {materialFunction=&MF_UI_Box}.Fill)) {name=Backing Shape, nodeColor=<rgba 0 0 0 1>}
  n4 = ScalarParameter() {group=Noise, parameterName=LiftNoise_Bottom, sortPriority=10}
  n5 = LinearInterpolate(
   A=@n4
   B=ScalarParameter() {defaultValue=1, group=Noise, parameterName=LiftNoise_Top, sortPriority=11}
   Alpha=ScalarParameter() {group=Noise, parameterName=NoiseOn_Anim, sliderMax=1, sortPriority=12}
  ) {constAlpha=0}
  n6 = MaterialFunctionCall() {materialFunction=&ScreenAlignedUVs}
  n7 = MaterialFunctionCall(UVs=@n6."X 100%, Y 100%", Texture Scale=ScalarParameter() {defaultValue=0.25, group=Noise, parameterName=Noise_1_Scale}) {materialFunction=&ScaleUVsByCenter}
  n8 = Time()
  n9 = AppendVector(
   A=Multiply(A=@n8, B=ScalarParameter() {defaultValue=0.0625, group=Noise, parameterName=Noise_1_TimeSpeed+Direction_X})
   B=Multiply(A=@n8, B=ScalarParameter() {defaultValue=-0.088, group=Noise, parameterName=Noise_1_TimeSpeed+Direction_Y})
  )
  n10 = TextureObject() {samplerType=SAMPLERTYPE_Grayscale, texture=/FortUILibrary/UIKit/Buttons/Textures/LowResBlurredNoise}
  n11 = MaterialFunctionCall(UVs=@n6."X 100%, Y 100%", Texture Scale=ScalarParameter() {defaultValue=0.25, group=Noise, parameterName=Noise_2_Scale}) {materialFunction=&ScaleUVsByCenter}
  n12 = Time()
  n13 = AppendVector(
   A=Multiply(A=@n12, B=ScalarParameter() {defaultValue=-0.125, group=Noise, parameterName=Noise_2_TimeSpeed+Direction_X})
   B=Multiply(A=@n12, B=ScalarParameter() {defaultValue=0.125, group=Noise, parameterName=Noise_2_TimeSpeed+Direction_Y})
  )
  n14 = Multiply(A=TextureSample(UVs=Add(A=@n7.UVs, B=@n9), Tex=@n10).R, B=TextureSample(UVs=Add(A=@n11.UVs, B=@n13), Tex=@n10).R)
  NamedRerouteDeclaration(None=Saturate(None=LinearInterpolate(A=@n4, B=@n5, Alpha=@n14))) {name=Noise, nodeColor=<rgba 0 0 0 1>}
  n15 = MaterialFunctionCall() {materialFunction=&GetUserInterfaceUV}
  n16 = Add(B=Multiply(
    A=ScalarParameter() {group=Border, parameterName=Rotation_Left, sortPriority=25}
    B=ScalarParameter() {defaultValue=1, group=BorderOffset, parameterName=BorderOffset_HorizontalScale}
   )) {constA=0.5}
  n17 = LinearInterpolate(
   A=ScalarParameter() {defaultValue=0.75, desc="1: on edge, 0: center", group=BorderOffset, parameterName=Offset_Left_0, sortPriority=10}
   B=ScalarParameter() {defaultValue=0.75, group=BorderOffset, parameterName=Offset_Left_1, sortPriority=20}
   Alpha=ScalarParameter() {group=BorderOffset, parameterName=BorderOffset_Anim, sortPriority=50}
  )
  n18 = OneMinus(None=Multiply(A=OneMinus(None=@n17), B=ScalarParameter() {defaultValue=1, group=BorderOffset, parameterName=BorderOffset_HorizontalScale}))
  n19 = Subtract(
   A=ComponentMask(None=MaterialFunctionCall(UVs=@n15.9-Slice UV, "Rotation Angle (0-1)"=@n16) {materialFunction=&CustomRotator}.Rotated Values) {r=true}
   B=@n18
  )
  n20 = OneMinus(None=@n18)
  n21 = Divide(
   A=Saturate(None=Divide(A=@n19, B=@n20))
   B=OneMinus(None=ScalarParameter() {defaultValue=0.5, group=BorderOffset, parameterName=Size_Left, sortPriority=30})
  )
  n22 = ScalarParameter() {defaultValue=0.5, group=Border, parameterName=Border_CornerRadius, sortPriority=20}
  n23 = Divide(A=Subtract(A=Abs(None=@n21), B=OneMinus(None=@n22)), B=@n22)
  n24 = MaterialFunctionCall() {materialFunction=&GetUserInterfaceUV}
  n25 = Multiply(
   A=ScalarParameter() {group=Border, parameterName=Rotation_Right, sortPriority=26}
   B=ScalarParameter() {defaultValue=1, group=BorderOffset, parameterName=BorderOffset_HorizontalScale}
  )
  n26 = LinearInterpolate(
   A=ScalarParameter() {defaultValue=0.75, group=BorderOffset, parameterName=Offset_Right_0, sortPriority=11}
   B=ScalarParameter() {defaultValue=0.75, group=BorderOffset, parameterName=Offset_Right_1, sortPriority=21}
   Alpha=ScalarParameter() {group=BorderOffset, parameterName=BorderOffset_Anim, sortPriority=50}
  )
  n27 = OneMinus(None=Multiply(A=OneMinus(None=@n26), B=ScalarParameter() {defaultValue=1, group=BorderOffset, parameterName=BorderOffset_HorizontalScale}))
  n28 = Subtract(
   A=ComponentMask(None=MaterialFunctionCall(UVs=@n24.9-Slice UV, "Rotation Angle (0-1)"=@n25) {materialFunction=&CustomRotator}.Rotated Values) {r=true}
   B=@n27
  )
  n29 = OneMinus(None=@n27)
  n30 = Divide(
   A=Saturate(None=Divide(A=@n28, B=@n29))
   B=OneMinus(None=ScalarParameter() {defaultValue=0.5, group=BorderOffset, parameterName=Size_Right, sortPriority=31})
  )
  n31 = ScalarParameter() {defaultValue=0.5, group=Border, parameterName=Border_CornerRadius, sortPriority=20}
  n32 = Divide(A=Subtract(A=Abs(None=@n30), B=OneMinus(None=@n31)), B=@n31)
  n33 = MaterialFunctionCall() {materialFunction=&GetUserInterfaceUV}
  n34 = StaticSwitchParameter(True=@n33.9-Slice UV, False=@n33.9-Slice UV) {group=_Switch, parameterName=OnlyBorder}
  n35 = Add(B=Multiply(
    A=ScalarParameter() {group=Border, parameterName=Rotation_Bottom, sortPriority=28}
    B=ScalarParameter() {defaultValue=1, group=BorderOffset, parameterName=BorderOffset_VerticalScale}
   )) {constA=0.75}
  n36 = LinearInterpolate(
   A=ScalarParameter() {defaultValue=0.75, group=BorderOffset, parameterName=Offset_Bottom_0, sortPriority=13}
   B=ScalarParameter() {defaultValue=0.75, group=BorderOffset, parameterName=Offset_Bottom_1, sortPriority=23}
   Alpha=ScalarParameter() {group=BorderOffset, parameterName=BorderOffset_Anim, sortPriority=50}
  )
  n37 = OneMinus(None=Multiply(A=OneMinus(None=@n36), B=ScalarParameter() {defaultValue=1, group=BorderOffset, parameterName=BorderOffset_VerticalScale}))
  n38 = Subtract(
   A=ComponentMask(None=MaterialFunctionCall(UVs=@n34, "Rotation Angle (0-1)"=@n35) {materialFunction=&CustomRotator}.Rotated Values) {r=true}
   B=@n37
  )
  n39 = OneMinus(None=@n37)
  n40 = Divide(
   A=Saturate(None=Divide(A=@n38, B=@n39))
   B=OneMinus(None=ScalarParameter() {defaultValue=0.5, group=BorderOffset, parameterName=Size_Bottom, sortPriority=33})
  )
  n41 = ScalarParameter() {defaultValue=0.5, group=Border, parameterName=Border_CornerRadius, sortPriority=20}
  n42 = Divide(A=Subtract(A=Abs(None=@n40), B=OneMinus(None=@n41)), B=@n41)
  n43 = MaterialFunctionCall() {materialFunction=&GetUserInterfaceUV}
  n44 = StaticSwitchParameter(True=@n43.9-Slice UV, False=@n43.9-Slice UV) {group=_Switch, parameterName=OnlyBorder}
  n45 = Add(B=Multiply(
    A=ScalarParameter() {group=Border, parameterName=Rotation_Top, sortPriority=27}
    B=ScalarParameter() {defaultValue=1, group=BorderOffset, parameterName=BorderOffset_VerticalScale}
   )) {constA=0.25}
  n46 = LinearInterpolate(
   A=ScalarParameter() {defaultValue=0.75, group=BorderOffset, parameterName=Offset_Top_0, sortPriority=12}
   B=ScalarParameter() {defaultValue=0.75, group=BorderOffset, parameterName=Offset_Top_1, sortPriority=22}
   Alpha=ScalarParameter() {group=BorderOffset, parameterName=BorderOffset_Anim, sortPriority=50}
  )
  n47 = OneMinus(None=Multiply(A=OneMinus(None=@n46), B=ScalarParameter() {defaultValue=1, group=BorderOffset, parameterName=BorderOffset_VerticalScale}))
  n48 = Subtract(
   A=ComponentMask(None=MaterialFunctionCall(UVs=@n44, "Rotation Angle (0-1)"=@n45) {materialFunction=&CustomRotator}.Rotated Values) {r=true}
   B=@n47
  )
  n49 = OneMinus(None=@n47)
  n50 = Divide(A=Saturate(None=Divide(A=@n48, B=@n49)), B=OneMinus(None=ScalarParameter() {defaultValue=0.5, group=BorderOffset, parameterName=Size_Top}))
  n51 = ScalarParameter() {defaultValue=0.5, group=Border, parameterName=Border_CornerRadius, sortPriority=20}
  n52 = Divide(A=Subtract(A=Abs(None=@n50), B=OneMinus(None=@n51)), B=@n51)
  n53 = AppendVector(A=Add(A=Saturate(None=@n23), B=Saturate(None=@n32)), B=Add(A=Saturate(None=@n42), B=Saturate(None=@n52)))
  n54 = Saturate(None=DotProduct(A=@n53, B=@n53))
  n55 = Power(
   Base=SmoothStep(
    Min=ScalarParameter() {group=Border, parameterName=BorderSmooth_Min, sortPriority=21}
    Max=ScalarParameter() {defaultValue=0.05, group=Border, parameterName=BorderSmooth_Max, sortPriority=22}
    Value=OneMinus(None=@n54)
   )
   Exp=ScalarParameter() {defaultValue=1, group=Border, parameterName=Additonal_Falloff_Taper}
  )
  NamedRerouteDeclaration(None=OneMinus(None=Saturate(None=@n55))) {name=WonkBorder Shape, nodeColor=<rgba 0 0 0 1>}
  n56 = NamedRerouteUsage()
  n57 = Saturate(None=DotProduct(A=@n56, B=NamedRerouteUsage()))
  n58 = Saturate(None=DotProduct(A=@n57, B=NamedRerouteUsage()))
  NamedRerouteDeclaration(None=StaticSwitchParameter(True=StaticSwitchParameter(True=@n58, False=@n57) {group=_Switch, parameterName=HasNoise}, False=@n56) {group=_Switch, parameterName=OnlyBorder}) {name=Final Button Alpha, nodeColor=<rgba 0 0 0 1>}
  n59 = MaterialFunctionCall() {materialFunction=/FortUILibrary/Systems/ModularButton/MaterialFunctions/MF_UIKit_ButtonStatesBase_Params}
  n60 = CurveAtlasRowParameter(CurveTime=@n59.Trans_Pressed) {
   atlas=/FortUILibrary/UIKit/Buttons/Curves/CA_UI_Button_Transition
   curve=/FortUILibrary/UIKit/Buttons/Curves/CC_UI_Button_Transition
   parameterName=Transition_Curves
  }
  NamedRerouteDeclaration(None=@n60.R) {name=PressedAnimCurve, nodeColor=<rgba 0 0 0 1>}
  n61 = CurveAtlasRowParameter(CurveTime=@n59.Trans_Hovered) {
   atlas=&CA_UI_Button_Transition
   curve=&CC_UI_Button_Transition
   parameterName=Transition_Curves
  }
  NamedRerouteDeclaration(None=@n61.R) {name=HoverAnimCurve, nodeColor=<rgba 0 0 0 1>}
  n62 = CurveAtlasRowParameter(CurveTime=@n59.Trans_Focused) {
   atlas=&CA_UI_Button_Transition
   curve=&CC_UI_Button_Transition
   parameterName=Transition_Curves
  }
  NamedRerouteDeclaration(None=@n62.R) {name=FocusAnimCurve, nodeColor=<rgba 0 0 0 1>}
  n63 = CurveAtlasRowParameter(CurveTime=@n59.Trans_Disabled) {
   atlas=&CA_UI_Button_Transition
   curve=&CC_UI_Button_Transition
   parameterName=Transition_Curves
  }
  NamedRerouteDeclaration(None=@n63.R) {name=DisableAnimCurve, nodeColor=<rgba 0 0 0 1>}
  n64 = VectorParameter() {group=State colors, parameterName=Color_Idle_Border, sortPriority=0, defaultValue=<rgba 1 0 1 1>}
  n65 = VectorParameter() {group=State colors, parameterName=Color_Idle, sortPriority=0, defaultValue=<rgba 0.6297 0 1 1>}
  n66 = VectorParameter() {group=State colors, parameterName=Color_Idle_Border, sortPriority=0, defaultValue=<rgba 1 0 1 1>}
  n67 = NamedRerouteUsage()
  n68 = VectorParameter() {group=State colors, parameterName=Color_Idle, sortPriority=0, defaultValue=<rgba 0.6297 0 1 1>}
  n69 = StaticSwitchParameter(
   True=@n64.RGB
   False=StaticSwitchParameter(True=LinearInterpolate(A=@n65.RGB, B=@n66.RGB, Alpha=@n67), False=@n68.RGB) {defaultValue=true, group=_Switch, parameterName=HasWonkBorder}
  ) {group=_Switch, parameterName=OnlyBorder}
  n70 = StaticSwitchParameter(
   True=@n64.A
   False=StaticSwitchParameter(True=LinearInterpolate(A=@n65.A, B=@n66.A, Alpha=@n67), False=@n68.A) {defaultValue=true, group=_Switch, parameterName=HasWonkBorder}
  ) {group=_Switch, parameterName=OnlyBorder}
  n71 = VectorParameter() {group=State colors, parameterName=Color_Hovered_Border, sortPriority=1, defaultValue=<rgba 1 1 1 1>}
  n72 = VectorParameter() {group=State colors, parameterName=Color_Hovered, sortPriority=1, defaultValue=<rgba 1 1 1 1>}
  n73 = VectorParameter() {group=State colors, parameterName=Color_Hovered_Border, sortPriority=1, defaultValue=<rgba 1 1 1 1>}
  n74 = NamedRerouteUsage()
  n75 = VectorParameter() {group=State colors, parameterName=Color_Hovered, sortPriority=1, defaultValue=<rgba 1 1 1 1>}
  n76 = StaticSwitchParameter(
   True=@n71.RGB
   False=StaticSwitchParameter(True=LinearInterpolate(A=@n72.RGB, B=@n73.RGB, Alpha=@n74), False=@n75.RGB) {defaultValue=true, group=_Switch, parameterName=HasWonkBorder}
  ) {group=_Switch, parameterName=OnlyBorder}
  n77 = StaticSwitchParameter(
   True=@n71.A
   False=StaticSwitchParameter(True=LinearInterpolate(A=@n72.A, B=@n73.A, Alpha=@n74), False=@n75.A) {defaultValue=true, group=_Switch, parameterName=HasWonkBorder}
  ) {group=_Switch, parameterName=OnlyBorder}
  n78 = VectorParameter() {group=State colors, parameterName=Color_Focused_Border, sortPriority=2, defaultValue=<rgba 1 1 1 1>}
  n79 = VectorParameter() {group=State colors, parameterName=Color_Focused, sortPriority=2, defaultValue=<rgba 1 1 1 1>}
  n80 = VectorParameter() {group=State colors, parameterName=Color_Focused_Border, sortPriority=2, defaultValue=<rgba 1 1 1 1>}
  n81 = NamedRerouteUsage()
  n82 = VectorParameter() {group=State colors, parameterName=Color_Focused, sortPriority=2, defaultValue=<rgba 1 1 1 1>}
  n83 = StaticSwitchParameter(
   True=@n78.RGB
   False=StaticSwitchParameter(True=LinearInterpolate(A=@n79.RGB, B=@n80.RGB, Alpha=@n81), False=@n82.RGB) {defaultValue=true, group=_Switch, parameterName=HasWonkBorder}
  ) {group=_Switch, parameterName=OnlyBorder}
  n84 = StaticSwitchParameter(
   True=@n78.A
   False=StaticSwitchParameter(True=LinearInterpolate(A=@n79.A, B=@n80.A, Alpha=@n81), False=@n82.A) {defaultValue=true, group=_Switch, parameterName=HasWonkBorder}
  ) {group=_Switch, parameterName=OnlyBorder}
  n85 = VectorParameter() {group=State colors, parameterName=Color_Pressed_Border, sortPriority=3, defaultValue=<rgba 1 1 1 1>}
  n86 = VectorParameter() {group=State colors, parameterName=Color_Pressed, sortPriority=3, defaultValue=<rgba 1 1 1 1>}
  n87 = VectorParameter() {group=State colors, parameterName=Color_Pressed_Border, sortPriority=3, defaultValue=<rgba 1 1 1 1>}
  n88 = NamedRerouteUsage()
  n89 = VectorParameter() {group=State colors, parameterName=Color_Pressed, sortPriority=3, defaultValue=<rgba 1 1 1 1>}
  n90 = StaticSwitchParameter(
   True=@n85.RGB
   False=StaticSwitchParameter(True=LinearInterpolate(A=@n86.RGB, B=@n87.RGB, Alpha=@n88), False=@n89.RGB) {defaultValue=true, group=_Switch, parameterName=HasWonkBorder}
  ) {group=_Switch, parameterName=OnlyBorder}
  n91 = StaticSwitchParameter(
   True=@n85.A
   False=StaticSwitchParameter(True=LinearInterpolate(A=@n86.A, B=@n87.A, Alpha=@n88), False=@n89.A) {defaultValue=true, group=_Switch, parameterName=HasWonkBorder}
  ) {group=_Switch, parameterName=OnlyBorder}
  n92 = VectorParameter() {group=State colors, parameterName=Color_Disabled_Border, sortPriority=4, defaultValue=<rgba 1 1 1 1>}
  n93 = VectorParameter() {group=State colors, parameterName=Color_Disabled, sortPriority=4, defaultValue=<rgba 1 1 1 1>}
  n94 = VectorParameter() {group=State colors, parameterName=Color_Disabled_Border, sortPriority=4, defaultValue=<rgba 1 1 1 1>}
  n95 = NamedRerouteUsage()
  n96 = VectorParameter() {group=State colors, parameterName=Color_Disabled, sortPriority=4, defaultValue=<rgba 1 1 1 1>}
  n97 = StaticSwitchParameter(
   True=@n92.RGB
   False=StaticSwitchParameter(True=LinearInterpolate(A=@n93.RGB, B=@n94.RGB, Alpha=@n95), False=@n96.RGB) {defaultValue=true, group=_Switch, parameterName=HasWonkBorder}
  ) {group=_Switch, parameterName=OnlyBorder}
  n98 = StaticSwitchParameter(
   True=@n92.A
   False=StaticSwitchParameter(True=LinearInterpolate(A=@n93.A, B=@n94.A, Alpha=@n95), False=@n96.A) {defaultValue=true, group=_Switch, parameterName=HasWonkBorder}
  ) {group=_Switch, parameterName=OnlyBorder}
  Saturate(None=DotProduct(
    A=MaterialFunctionCall!(
     Idle=@n69
     Idle Opacity=@n70
     Hovered=@n76
     Hovered Opacity=@n77
     Focused=@n83
     Focused Opacity=@n84
     Pressed=@n90
     Pressed Opacity=@n91
     Disabled=@n97
     Disabled Opacity=@n98
    ) {materialFunction=/FortUILibrary/Systems/ModularButton/MaterialFunctions/MF_UIKit_ButtonStatesBase_Lerp}.Opacity
    B=NamedRerouteUsage()
   ))
 ]
}
