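# UI material for UIKit buttons that draws a rounded-corner box, filled or outline-only, with per-state colors.
# Button-state lerp functions blend Idle, Hovered, Focused, Pressed and Disabled colors, with static switches for selectable states or a rotatable two-color gradient.
# State opacity is clipped by the rounded mask, then scales color alongside brightness, vertex alpha and Opacity.
# AlphaComposite blending and a separate Additive Percentage alpha multiplier allow additive-style compositing.
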Material /FortUILibrary/UIKit/Buttons/Materials/M_UIKit_Button_RoundedCorners
graph {
 properties {
  bUsedWithStaticMesh=false
  blendMode=BLEND_AlphaComposite
  materialDomain=MD_UI
  pixelDepthOffsetMode=PDOM_Legacy
  usedShadingModels=Default Lit
 }
 nodes [
  n1 = VectorParameter() {group=State colors, parameterName=Color_Idle, sortPriority=0, defaultValue=<rgba 0.6297 0 1 1>}
  n2 = VectorParameter() {group=State colors, parameterName=Color_Hovered, sortPriority=1, defaultValue=<rgba 1 1 1 1>}
  n3 = VectorParameter() {group=State colors, parameterName=Color_Focused, sortPriority=2, defaultValue=<rgba 1 1 1 1>}
  n4 = VectorParameter() {group=State colors, parameterName=Color_Pressed, sortPriority=3, defaultValue=<rgba 1 1 1 1>}
  n5 = VectorParameter() {group=State colors, parameterName=Color_Disabled, sortPriority=4, defaultValue=<rgba 1 1 1 1>}
  n6 = MaterialFunctionCall!(
   Idle=@n1.RGB
   Idle Opacity=@n1.RGB
   Hovered=@n2.RGB
   Hovered Opacity=@n2.RGB
   Focused=@n3.RGB
   Focused Opacity=@n3.RGB
   Pressed=@n4.RGB
   Pressed Opacity=@n4.RGB
   Disabled=@n5.RGB
   Disabled Opacity=@n5.RGB
  ) {materialFunction=/FortUILibrary/Systems/ModularButton/MaterialFunctions/MF_UIKit_ButtonStatesBase_Lerp}
  n7 = MaterialFunctionCall!(
   Base States=@n6.Color
   Base States Opacity=@n6.Color
   Selected_Idle
   Selected_Idle Opacity
   Selected_Hovered
   Selected_Hovered Opacity
   Selected_Focused
   Selected_Focused Opacity
   Selected_Pressed
   Selected_Pressed Opacity
   Selected_Disabled
   Selected_Disabled Opacity
  ) {materialFunction=/FortUILibrary/Systems/ModularButton/MaterialFunctions/MF_UIKit_ButtonStatesSelectable_Lerp}
  n8 = StaticSwitchParameter(True=NamedRerouteUsage(), False=StaticSwitchParameter(True=@n7.Color, False=@n6.Color) {parameterName=Is Selectable}) {group=Switches, parameterName=Gradient Color, sortPriority=0}
  n9 = Min(
   A=StaticSwitchParameter(True=@n7.Opacity, False=@n6.Opacity) {parameterName=Is Selectable}
   B=StaticSwitchParameter(True=NamedRerouteUsage(), False=NamedRerouteUsage()) {group=_Switch, parameterName=Outline Only}
  )
  n10 = Multiply(A=ScalarParameter() {defaultValue=1, group=Color, parameterName=Brightness Percentage, sliderMax=1, sortPriority=2}, B=@n9)
  Multiply(A=ScalarParameter() {defaultValue=1, parameterName=Opacity, sliderMax=1}, B=Multiply(A=VertexColor().A, B=Multiply(A=@n8, B=@n10)))
  Multiply(A=@n9, B=ScalarParameter() {defaultValue=1, group=Color, parameterName=Additive Percentage, sliderMax=1, sortPriority=3})
  n11 = MaterialFunctionCall(
   UV=NamedRerouteUsage()
   Size=AppendVector(A=ScalarParameter() {defaultValue=1, parameterName=Length}, B=ScalarParameter() {defaultValue=1, parameterName=Height})
   Corner Radius=ScalarParameter() {defaultValue=0.5, parameterName=Corner Radius}
   Radius in Pixels=ScalarParameter() {parameterName=Radius in Pixels, sliderMax=1}
   Outline Thickness=ScalarParameter() {defaultValue=1, parameterName=Outline Thickness}
  ) {materialFunction=&MF_UI_CoreUI_RoundedMask}
  NamedRerouteDeclaration(None=@n11.Shape Mask) {name=Fill_Box, nodeColor=<rgba 0 0 0 1>}
  NamedRerouteDeclaration(None=MaterialFunctionCall() {materialFunction=&GetUserInterfaceUV}.Normalized UV) {name=Normalized UV, nodeColor=<rgba 0 0 0 1>}
  n12 = ScalarParameter() {defaultValue=0.5, parameterName=Gradient Center X, sliderMax=1}
  n13 = ScalarParameter() {defaultValue=0.5, parameterName=Gradient Center Y, sliderMax=1}
  n14 = ComponentMask(None=MaterialFunctionCall(
    UVs=TextureCoordinate()
    Rotation Center=AppendVector(A=@n12, B=@n13)
    "Rotation Angle (0-1)"=ScalarParameter() {parameterName=Gradient Rotation, sliderMax=1}
   ) {materialFunction=&CustomRotator}.Rotated Values) {r=true}
  n15 = ScalarParameter() {defaultValue=1, parameterName=Gradient Softness}
  NamedRerouteDeclaration(None=LinearInterpolate(
    A=VectorParameter() {parameterName=Color A, defaultValue=<rgba 0.0646 0.0091 0.125 1>}.RGB
    B=VectorParameter() {parameterName=Color B, defaultValue=<rgba 0.0139 0.1458 0.0685 1>}.RGB
    Alpha=Power(Base=@n14, Exp=@n15)
   )) {name=Gradient Color, nodeColor=<rgba 0 0 0 1>}
  NamedRerouteDeclaration(None=@n11.Outline Mask) {name=Outline_Box, nodeColor=<rgba 0 0.2031 1 1>}
 ]
}
