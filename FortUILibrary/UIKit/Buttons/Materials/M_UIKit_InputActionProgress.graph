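# Translucent UI material that draws a radial progress indicator for UIKit input actions, shaped as a circle or rounded box.
# UVs rotated by StartPoint become an angle, optionally inverted, which floored sums threshold into a Range arc and a Percentage-scaled fill within it.
# ShapeSwitch blends circle and box masks, adding a separately scaled fill shape.
# Progress blends background to progress color and alpha, with opacity masked by shape, arc and HideProgress.
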Material /FortUILibrary/UIKit/Buttons/Materials/M_UIKit_InputActionProgress
graph {
 properties {
  bUsedWithStaticMesh=false
  blendMode=BLEND_Translucent
  materialDomain=MD_UI
  pixelDepthOffsetMode=PDOM_Legacy
  usedShadingModels=Default Lit
 }
 nodes [
  n1 = VectorParameter() {group=Colors, parameterName=Color_Bg, sortPriority=2, defaultValue=<rgba 0.08 0 0.08 0.5>}
  n2 = VectorParameter() {group=Colors, parameterName=Color_Progress, sortPriority=1, defaultValue=<rgba 1 0 1 1>}
  n3 = NamedRerouteUsage()
  LinearInterpolate(A=@n1.RGB, B=@n2.RGB, Alpha=@n3)
  n4 = Constant() {r=1}
  n5 = ScalarParameter() {defaultValue=0.01, group=Shape, parameterName=Circle_Hardness, sortPriority=22}
  n6 = MaterialFunctionCall(
   UVs=MaterialFunctionCall(Texture Scale=ScalarParameter() {defaultValue=0.95, group=FillShape, parameterName=FillCircle_Scale, sortPriority=21}) {materialFunction=&ScaleUVsByCenter}.UVs
   Size=@n4
   GlowMax=@n5
  ) {materialFunction=&MF_UI_Circle}
  n7 = ScalarParameter() {defaultValue=0.85, group=FillShape, parameterName=FillBox_XScale, sortPriority=31}
  n8 = ScalarParameter() {defaultValue=0.85, group=FillShape, parameterName=FillBox_YScale, sortPriority=31}
  n9 = ScalarParameter() {defaultValue=0.5, group=Shape, parameterName=Box_CornerRadius}
  n10 = ScalarParameter() {defaultValue=0.95, group=Shape, parameterName=Box_Hardness, sortPriority=31}
  n11 = MaterialFunctionCall!(
   UVs=MaterialFunctionCall(Texture Scale=AppendVector(A=@n7, B=@n8)) {materialFunction=&ScaleUVsByCenter}.UVs
   Radius=@n9
   Aspect Ratio
   UV Scale from center
   Hardness=@n10
  ) {materialFunction=/Game/Packages/Fortress_SharedMaterials/Base_Material_Functions/MF_RoundedRectangle}
  n12 = ScalarParameter() {defaultValue=1, group=Shape, parameterName=ShapeSwitch, sortPriority=20}
  n13 = MaterialFunctionCall(Texture Scale=ScalarParameter() {defaultValue=0.85, group=Shape, parameterName=Circle_Scale, sortPriority=21}) {desc=Scale To Fit, materialFunction=&ScaleUVsByCenter}
  n14 = MaterialFunctionCall(Texture Scale=AppendVector(
    A=ScalarParameter() {defaultValue=0.75, group=Shape, parameterName=Box_XScale, sortPriority=31}
    B=ScalarParameter() {defaultValue=0.75, group=Shape, parameterName=Box_YScale, sortPriority=31}
   )) {desc=Scale To Fit, materialFunction=&ScaleUVsByCenter}
  n15 = Add(
   A=Multiply(A=LinearInterpolate(A=@n6.Fill, B=@n11.Result, Alpha=@n12), B=NamedRerouteUsage())
   B=LinearInterpolate(
    A=MaterialFunctionCall(UVs=@n13.UVs, Size=@n4, GlowMax=@n5) {materialFunction=&MF_UI_Circle}.Fill
    B=MaterialFunctionCall(UVs=@n14.UVs, Radius=@n9, Hardness=@n10) {materialFunction=&MF_RoundedRectangle}.Result
    Alpha=@n12
   )
  )
  n16 = ScalarParameter() {defaultValue=0.5, group=Progress, parameterName=Range, sliderMax=1, sortPriority=12}
  n17 = MaterialFunctionCall!(Vector or UVs=MaterialFunctionCall(
    UVs=TextureCoordinate()
    "Rotation Angle (0-1)"=ScalarParameter() {defaultValue=0.5, group=Progress, parameterName=StartPoint, sliderMax=1, sortPriority=11}
   ) {materialFunction=&CustomRotator}.Rotated Values) {materialFunction=/FortUILibrary/MaterialFunctions/Basics/MF_UI_CheaperVectorToRadial}
  n18 = StaticSwitchParameter(True=@n17.Vector Converted to Angle, False=OneMinus(None=@n17.Vector Converted to Angle)) {defaultValue=true, group=Progress, parameterName=Clockwise}
  n19 = Floor(None=Add(A=@n16, B=@n18))
  Multiply(
   A=Multiply(A=LinearInterpolate(A=@n1.A, B=@n2.A, Alpha=@n3), B=Multiply(A=@n15, B=@n19))
   B=ScalarParameter() {defaultValue=1, group=Progress, parameterName=HideProgress, sliderMax=1}
  )
  n20 = Multiply(A=@n16, B=ScalarParameter() {defaultValue=0.35, group=Progress, parameterName=Percentage, sliderMax=1, sortPriority=10})
  NamedRerouteDeclaration(None=Floor(None=Add(A=@n20, B=@n18))) {name=ProgressFill, nodeColor=<rgba 0 0 0 1>}
 ]
}
