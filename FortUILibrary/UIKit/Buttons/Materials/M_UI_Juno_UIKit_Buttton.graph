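# Translucent UI material for the Juno UIKit button, drawing its interaction states inside a rounded-box shape.
# A state-lerp function blends idle, hovered, focused, pressed and disabled; an Is Selectable static switch chooses a second function adding selected variants.
# An anti-aliased SDF box with size and corner-roundness parameters masks the opacity.
# The selectable function's Base States Opacity input is wired to the base Color output, not Opacity.
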Material /FortUILibrary/UIKit/Buttons/Materials/M_UI_Juno_UIKit_Buttton
graph {
 properties {
  bUsedWithStaticMesh=false
  blendMode=BLEND_Translucent
  materialDomain=MD_UI
  pixelDepthOffsetMode=PDOM_Legacy
  usedShadingModels=Default Lit
 }
 nodes [
  n1 = MaterialFunctionCall!(Idle, Idle Opacity, Hovered, Hovered Opacity, Focused, Focused Opacity, Pressed, Pressed Opacity, Disabled, Disabled Opacity) {materialFunction=/FortUILibrary/Systems/ModularButton/MaterialFunctions/MF_UIKit_ButtonStatesBase_Lerp}
  n2 = MaterialFunctionCall!(
   Base States=@n1.Color
   Base States Opacity=@n1.Color
   Selected_Idle
   Selected_Idle Opacity
   Selected_Hovered
   Selected_Hovered Opacity
   Selected_Focused
   Selected_Focused Opacity
   Selected_Pressed
   Selected_Pressed Opacity
   Selected_Disabled
   Selected_Disabled Opacity
  ) {materialFunction=/FortUILibrary/Systems/ModularButton/MaterialFunctions/MF_UIKit_ButtonStatesSelectable_Lerp}
  Multiply(A=StaticSwitchParameter(True=@n2.Opacity, False=@n1.Opacity) {parameterName=Is Selectable}, B=NamedRerouteUsage())
  n3 = MaterialFunctionCall() {materialFunction=&GetUserInterfaceUV}
  n4 = StaticSwitchParameter(True=@n3.Normalized UV, False=@n3.Normalized UV) {defaultValue=true, parameterName=Use Normalized UV}
  n5 = ScalarParameter() {defaultValue=1, parameterName=Size}
  n6 = ScalarParameter() {defaultValue=1, parameterName=Corner Roundness}
  n7 = StaticSwitchParameter(True=StaticBool() {value=true}, False=StaticBool()) {defaultValue=true, parameterName=Use Normalized UV}
  NamedRerouteDeclaration(None=MaterialFunctionCall(DistanceField=MaterialFunctionCall(UVs=@n4, Size=@n5, CornerRoundness=@n6, KeepAspectRatio=@n7) {materialFunction=&MF_UI_Box}.SDF) {materialFunction=&MF_UI_AntiAliasing}.Inner) {name=Button Shape, nodeColor=<rgba 0 0 0 1>}
  StaticSwitchParameter(True=@n2.Color, False=@n1.Color) {parameterName=Is Selectable}
 ]
}
