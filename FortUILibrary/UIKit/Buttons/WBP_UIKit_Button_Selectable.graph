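# UIKit selectable button widget, derived from WBP_UIKit_ButtonGeneric_Base with no widgets or variables of its own.
# SetData fills the inherited text and icon from a UIKitSelectableButtonViewModel's Label and Icon.
# SetPreviewData casts a generic MVVM viewmodel to that type and does the same, also setting the TopNav icon material instance before the icon.
# PreConstruct and Construct only call the parent.
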WidgetBlueprint /FortUILibrary/UIKit/Buttons/WBP_UIKit_Button_Selectable
graphs [
 EventGraph {nodes [
   n1 = Event PreConstruct()
   > Parent: PreConstruct(IsDesignTime=@n1.IsDesignTime)
   Event Construct()
   > Parent: Construct()
  ]}
 SetData {nodes [
   n1 = SetData!() -> (then, VM Button:UI Kit Selectable Button View Model Object Reference) {is=title K2Node_FunctionEntry}
   > SetText!(
    execute
    self:WBP UIKit Button Generic Base Object Reference
    Text:Text=Get Label!(self:UI Kit Selectable Button View Model Object Reference=@n1) -> (Label:Text)
   ) -> (then)
   > SetIcon!(
    execute
    self:WBP UIKit Button Generic Base Object Reference
    Icon:Object Reference=Get Icon!(self:UI Kit Selectable Button View Model Object Reference=@n1) -> (Icon:Texture 2D Object Reference)
   ) -> (then)
  ]}
 SetPreviewData {nodes [
   n1 = SetPreviewData!() -> (then, VM:MVVM Base Viewmodel Object Reference) {is=title K2Node_FunctionEntry}
   > n2 = Cast To UIKitSelectableButtonViewModel!(execute, Object=@n1) -> (then, CastFailed, AsUI Kit Selectable Button View Model:UI Kit Selectable Button View Model Object Reference) {is=title K2Node_DynamicCast}
   > SetText(Text=Get Label(self=@n2))
   > SetIconMaterial!(
    execute
    self:WBP UIKit Button Generic Base Object Reference
    Material Instance:Material Instance Object Reference=/FortUILibrary/UIKit/Buttons/TabButtons/_MaterialInstances/MI_UI_TopNav_Icon
   ) -> (then)
   > SetIcon(Icon=Get Icon(self=@n2))
  ]}
]
variables []
widgets {
 info {
  inheritedWidgetCount=0
  parentClass=/FortUILibrary/Systems/ModularButton/WBP_UIKit_ButtonGeneric_Base.WBP_UIKit_ButtonGeneric_Base_C
  rootWidgetClass=null
  widgetCount=0
 }
 widgets []
}
