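# Collapsible UI Kit tab button, extending the selectable button with a vertical pip marker.
# PreConstruct applies left-aligned layout, mobile-aware padding and text wrapping, then resets both animations.
# OnSelected scales content and grows and brightens the pip; OnCollapsed fades content, blocks hit testing and shifts the pip, both reversible.
# Animated properties drive setters that unwrap text and lerp pip and spacer heights from measured text height.
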WidgetBlueprint /FortUILibrary/UIKit/Buttons/WBP_UIKit_Button_Tab_Collapsible
animations [
 OnCollapsed {
  binding_count=5
  length_seconds=0.25
  object=&self:OnCollapsed
  sequence {
   bindings [
    GridSlot_2 {
     id=#1
     parent_id=#2
     possessed_class=GridSlot
     tracks [{
       type=MovieSceneDoubleVectorTrack
       display_name=Nudge
       sections [{
         type=MovieSceneDoubleVectorSection
         end_frame=25
         start_frame=0
         channels [
          Vector.X {type=MovieSceneScriptingDoubleChannel, const [0,2]}
          Vector.Y {type=MovieSceneScriptingDoubleChannel, const [0,1]}
         ]
        }]
      }]
    }
    WBP_UIKit_Button_Tab_Collapsible_C_21 {
     id=#3
     possessed_class=WBP_UIKit_Button_Tab_Collapsible_C
     tracks [
      {
       type=MovieSceneDoubleTrack
       display_name=Text on One Line
       sections [{
         type=MovieSceneDoubleSection
         end_frame=25
         start_frame=0
         channels [None {type=MovieSceneScriptingDoubleChannel, keys [[0,0],[5,0],[10,1],[25,1]]}]
        }]
      }
      {
       type=MovieSceneDoubleTrack
       display_name=Collapsed Anim Value
       sections [{
         type=MovieSceneDoubleSection
         end_frame=25
         start_frame=0
         channels [None {type=MovieSceneScriptingDoubleChannel, keys [[0,0],[10,0],[25,1]]}]
        }]
      }
      {
       type=MovieSceneByteTrack
       display_name=Visibility
       sections [{type=MovieSceneByteSection, end_frame=25, start_frame=0, channels [None {type=MovieSceneScriptingByteChannel, keys [[0,4],[10,3],[25,3]]}]}]
      }
     ]
    }
    SizeBox_Pip {id=#4, possessed_class=SizeBox, tracks []}
    OverlaySlot_0 {
     id=#5
     parent_id=#4
     possessed_class=OverlaySlot
     tracks [{
       type=MovieSceneMarginTrack
       display_name=Padding
       sections [{
         type=MovieSceneMarginSection
         end_frame=25
         start_frame=0
         channels [
          Left {type=MovieSceneScriptingFloatChannel, keys [[0,0],[25,-8]]}
          Top {type=MovieSceneScriptingFloatChannel, const [0,1]}
          Right {type=MovieSceneScriptingFloatChannel, const [0,1]}
          Bottom {type=MovieSceneScriptingFloatChannel, const [0,1]}
         ]
        }]
      }]
    }
    Main {
     id=#2
     possessed_class=GridPanel
     tracks [{
       type=MovieSceneFloatTrack
       display_name=Render Opacity
       sections [{
         type=MovieSceneFloatSection
         end_frame=25
         start_frame=0
         channels [None {type=MovieSceneScriptingFloatChannel, keys [[0,1],[9.975,0],[25,0]]}]
        }]
      }]
    }
   ]
   display_rate {numerator=100}
  }
 }
 OnSelected {
  binding_count=3
  length_seconds=0.15
  object=&self:OnSelected
  sequence {bindings [
    WBP_UIKit_Button_Tab_Collapsible_C_58 {
     id=#6
     possessed_class=WBP_UIKit_Button_Tab_Collapsible_C
     tracks [{
       type=MovieSceneDoubleTrack
       display_name=Selected Anim Value
       sections [{type=MovieSceneDoubleSection, end_frame=3, start_frame=0, channels [None {type=MovieSceneScriptingDoubleChannel, keys [[0,0],[3,1]]}]}]
      }]
    }
    Main {
     id=#7
     possessed_class=GridPanel
     tracks [
      {
       type=MovieSceneDoubleVectorTrack
       display_name=Pivot
       sections [{
         type=MovieSceneDoubleVectorSection
         end_frame=3
         start_frame=0
         channels [
          Vector.X {type=MovieSceneScriptingDoubleChannel, const [0,2]}
          Vector.Y {type=MovieSceneScriptingDoubleChannel, const [0.5,1]}
         ]
        }]
      }
      {
       type=MovieScene2DTransformTrack
       sections [{
         type=MovieScene2DTransformSection
         end_frame=3
         start_frame=0
         channels [
          Scale.X {type=MovieSceneScriptingFloatChannel, keys [[0,0.9],[3,1]]}
          Scale.Y {type=MovieSceneScriptingFloatChannel, keys [[0,0.9],[3,1]]}
         ]
        }]
      }
     ]
    }
    Image_Pip {
     id=#8
     possessed_class=Image
     tracks [{
       type=MovieSceneColorTrack
       display_name=Color and Opacity
       sections [{
         type=MovieSceneColorSection
         end_frame=3
         start_frame=0
         channels [
          Color.R {type=MovieSceneScriptingFloatChannel, const [1,2]}
          Color.G {type=MovieSceneScriptingFloatChannel, const [1,2]}
          Color.B {type=MovieSceneScriptingFloatChannel, const [1,2]}
          Color.A {type=MovieSceneScriptingFloatChannel, keys [[0,0.5],[3,1]]}
         ]
        }]
      }]
    }
   ]}
 }
]
graphs [
 EventGraph {nodes [
   Event On Expanded() {is=title K2Node_CustomEvent}
   n1 = Get OnCollapsed() -> (OnCollapsed:Widget Animation Object Reference)
   > PlayAnimationReverse(InAnimation=@n1)
   Event On Collapsed() {is=title K2Node_CustomEvent}
   > PlayAnimationForward(InAnimation=@n1)
   n2 = Event PreConstruct()
   > Parent: PreConstruct(IsDesignTime=@n2.IsDesignTime)
   > InitializeLayout()
   > ResetAnimations()
   > InitializeText()
   > SetBlocksTiming!(
    execute
    self:UIKit Modular Button Object Reference
    Timing_Hovering:float=0
    Timing_Unhovering:float=0
    Timing_Focusing:float=0
    Timing_Unfocusing:float=0
    Timing_Pressing:float=0.1
    Timing_Releasing:float=0.1
    Timing_Disabling:float=0.1
    Timing_Enabling:float=0.1
    Timing_Selecting:float=0
    Timing_DeselectingIdle:float=0
    Timing_DeselectingFocused:float=0
   ) -> (then)
   Event On Deselected()
   > Parent: On Deselected() {is=title K2Node_CallParentFunction}
   n3 = Get OnSelected() -> (OnSelected:Widget Animation Object Reference)
   > PlayAnimationReverse(InAnimation=@n3)
   Event On Selected()
   > Parent: On Selected() {is=title K2Node_CallParentFunction}
   > PlayAnimationForward(InAnimation=@n3)
  ]}
 InitializeLayout {nodes [
   InitializeLayout() {is=title K2Node_FunctionEntry}
   > UpdatePadding!(execute, self, SelectionAlignment:E_UI_Alignment Enum=NewEnumerator2) -> (then)
   > UpdateAlignment!(execute, self, Selection Alignment:E_UI_Alignment Enum=NewEnumerator2) -> (then)
   > SetGridSlotAlignment!(
    execute
    Widget:Widget Object Reference=Get Slot_Main() -> (Slot_Main:Named Slot Object Reference)
    Horizontal:EHorizontalAlignment Enum=HAlign_Left
    Vertical:EVerticalAlignment Enum=VAlign_Center
   ) -> (then)
  ]}
 InitializeText {nodes [
   InitializeText() {is=title K2Node_FunctionEntry}
   n1 = Get TextBlock!(self:WBP UIKit Block Text Object Reference=Get Block_Text() -> (Block_Text:WBP UIKit Block Text Object Reference)) -> (TextBlock:Common Text Object Reference)
   > SetWrapTextWidth!(execute, self:Common Text Object Reference=@n1, InWrapTextAt:int=Get WrapTextAt() -> (WrapTextAt:int=0)) -> (then)
   > Set Justification!(execute, Justification:ETextJustify Enum=Left, self:Text Layout Widget Object Reference=@n1) -> (then, Output_Get:ETextJustify Enum=Left)
  ]}
 ResetAnimations {nodes [
   ResetAnimations() {is=title K2Node_FunctionEntry}
   n1 = Get OnSelected()
   > PlayAnimation(InAnimation=@n1, StartAtTime=GetEndTime(self=@n1), PlayMode=Reverse)
   n2 = Get OnCollapsed()
   > PlayAnimation(InAnimation=@n2, StartAtTime=GetEndTime(self=@n2), PlayMode=Reverse)
  ]}
 SetCollapsedAnimValue {nodes [
   SetCollapsedAnimValue!() -> (then, CollapsedAnimationProgress:double=~) {is=title K2Node_FunctionEntry}
   > n1 = Sequence()
   n2 = Get TextMaxHeight() -> (TextMaxHeight:double)
   n3 = Get PipHeightSelected() -> (PipHeightSelected:float)
   n4 = Get CollapsedAnimationProgress() -> (CollapsedAnimationProgress:double)
   SetHeightOverride(
    execute=@n1.then_1
    self=Get SizeBox_Pip() -> (SizeBox_Pip:Size Box Object Reference)
    InHeightOverride=Lerp(
     A=Get PipHeightUnselected() -> (PipHeightUnselected:float)
     B=Lerp(A=@n2, B=@n3, Alpha=@n4)
     Alpha=Get SelectedAnimationProgress() -> (SelectedAnimationProgress:double)
    )
   )
   > SetSize!(
    execute
    self:Spacer Object Reference=Get Spacer_MinHeight() -> (Spacer_MinHeight:Spacer Object Reference)
    InSize_X:double=1
    InSize_Y:double=Lerp(A=Get TextMaxHeight(), B=1, Alpha=Get CollapsedAnimationProgress())
   ) -> (then)
   Branch(execute=@n1.then_0, Condition="Nearly Equal (Float)"(A=Get CollapsedAnimationProgress(), B=0, ErrorTolerance=0))
   > Set TextMaxHeight(TextMaxHeight:double=GetDesiredSize!(self:Widget Object Reference=Get Block_Text()) -> (ReturnValue_X:double, ReturnValue_Y:double).ReturnValue_Y) -> (Output_Get:double)
  ]}
 SetSelectedAnimValue {nodes [
   n1 = SetSelectedAnimValue!() -> (then, AnimValue:double=~) {is=title K2Node_FunctionEntry}
   > Set SelectedAnimationProgress(SelectedAnimationProgress:double=@n1) -> (Output_Get:double)
   n2 = Get Block_Text()
   > SetHeightOverride(
    self=Get SizeBox_Pip()
    InHeightOverride=Lerp(A=Get PipHeightUnselected(), B=GetDesiredSize(self=@n2).ReturnValue_Y, Alpha=Get SelectedAnimationProgress())
   )
  ]}
 SetSelectionAlignment {nodes [
   n1 = SetSelectionAlignment!() -> (then, Selection Alignment:E_UI_Alignment Enum=NewEnumerator0) {is=title K2Node_FunctionEntry}
   > Set SelectedBarPosition(SelectedBarPosition:E_UI_Alignment Enum=@n1) -> (Output_Get:E_UI_Alignment Enum=NewEnumerator0)
  ]}
 SetTextOnOneLine {nodes [
   n1 = SetTextOnOneLine!() -> (then, value:double=~) {is=title K2Node_FunctionEntry}
   > SetWrapTextWidth(
    self=Get TextBlock(self=Get Block_Text())
    InWrapTextAt=Select(Option 0:int=Get WrapTextAt() -> (WrapTextAt=0), Option 1:int=0, Index="Nearly Equal (Float)"(A=@n1, B=1, ErrorTolerance=0)) -> (ReturnValue:int)
   )
  ]}
 UpdateAlignment {nodes [
   n1 = UpdateAlignment!() -> (then, Selection Alignment:E_UI_Alignment Enum=NewEnumerator0) {is=title K2Node_FunctionEntry}
   > SetButtonContentAlignment!(
    execute
    self:WBP UIKit Button Generic Base Object Reference
    HorizontalAlignment:EHorizontalAlignment Enum=Select!(
     NewEnumerator0:EHorizontalAlignment Enum=HAlign_Center
     NewEnumerator1:EHorizontalAlignment Enum=HAlign_Center
     NewEnumerator2:EHorizontalAlignment Enum=HAlign_Left
     NewEnumerator3:EHorizontalAlignment Enum=HAlign_Right
     Index:E_UI_Alignment Enum=@n1
    ) -> (ReturnValue:EHorizontalAlignment Enum)
    VerticalAlignment:EVerticalAlignment Enum=Select(
     NewEnumerator0:EVerticalAlignment Enum=VAlign_Top
     NewEnumerator1:EVerticalAlignment Enum=VAlign_Bottom
     NewEnumerator2:EVerticalAlignment Enum=VAlign_Center
     NewEnumerator3:EVerticalAlignment Enum=VAlign_Center
     Index=@n1
    ) -> (ReturnValue:EVerticalAlignment Enum)
   ) -> (then)
  ]}
 UpdatePadding {nodes [
   n1 = UpdatePadding!() -> (then, SelectionAlignment:E_UI_Alignment Enum=NewEnumerator0) {is=title K2Node_FunctionEntry}
   n2 = Get RightAlignmentPaddingMobile() -> (RightAlignmentPaddingMobile:Margin Structure)
   n3 = Get RightAlignmentPadding() -> (RightAlignmentPadding:Margin Structure)
   n4 = IsMobileGame()
   n5 = Get LeftAlignmentPadding() -> (LeftAlignmentPadding:Margin Structure)
   n6 = Get LeftAlignmentPaddingMobile() -> (LeftAlignmentPaddingMobile:Margin Structure)
   n7 = Get BottomAlignmentPadding() -> (BottomAlignmentPadding:Margin Structure)
   n8 = Get BottomAlignmentPaddingMobile() -> (BottomAlignmentPaddingMobile:Margin Structure)
   n9 = Get TopAlignmentPadding() -> (TopAlignmentPadding:Margin Structure)
   n10 = Get TopAlignmentPaddingMobile() -> (TopAlignmentPaddingMobile:Margin Structure)
   > SetButtonPadding!(
    execute
    self:WBP UIKit Button Generic Base Object Reference
    Padding:Margin Structure=Select(
     NewEnumerator0:Margin Structure=Select!(Option 0:Margin Structure=@n9, Option 1:Margin Structure=@n10, Index:bool=@n4) -> (ReturnValue)
     NewEnumerator1:Margin Structure=Select(Option 0=@n7, Option 1=@n8, Index=@n4)
     NewEnumerator2:Margin Structure=Select(Option 0=@n5, Option 1=@n6, Index=@n4)
     NewEnumerator3:Margin Structure=Select(Option 0=@n3, Option 1=@n2, Index=@n4)
     Index=@n1
    ) -> (ReturnValue:Margin Structure)
   ) -> (then)
  ]}
]
variables [
 BottomAlignmentPadding {type=Margin, default=<ltrb 16 0 16 64>}
 BottomAlignmentPaddingMobile {type=Margin, default=<ltrb 16 0 16 64>}
 CollapsedAnimValue {type=number}
 LeftAlignmentPadding {type=Margin, default=<ltrb 21 0 0 0>}
 LeftAlignmentPaddingMobile {type=Margin, default=<ltrb 21 16 0 16>}
 PipHeightSelected {type=number, default=32}
 PipHeightUnselected {type=number, default=5}
 RightAlignmentPadding {type=Margin, default=<ltrb 0 16 80 16>}
 RightAlignmentPaddingMobile {type=Margin, default=<ltrb 0 16 80 16>}
 SelectedAnimationProgress {type=number}
 SelectedAnimValue {type=number}
 SelectedBarPosition {type=E_UI_Alignment, default=Left}
 TextMaxHeight {type=number}
 TextOnOneLine {type=number}
 TopAlignmentPadding {type=Margin, default=<ltrb 16 64 16 0>}
 TopAlignmentPaddingMobile {type=Margin, default=<ltrb 16 64 16 0>}
 WrapTextAt {type=integer, default=291}
]
widgets {
 info {
  inheritedWidgetCount=0
  namedSlotCount=1
  parentClass=&/WBP_UIKit_Button_Selectable.WBP_UIKit_Button_Selectable_C
  rootWidgetClass=null
  widgetCount=4
 }
 widgets [
  Overlay_252 {widgetClassPath=&Overlay, widget=Overlay_252}
  SizeBox_Pip {
   widgetClassPath=&SizeBox
   bIsVariable=true
   parent=Overlay_252
   widget=SizeBox_Pip
   slot {class=OverlaySlot, properties {horizontalAlignment=HAlign_Center, verticalAlignment=VAlign_Center}}
  }
  Image_Pip {
   widgetClassPath=&Image
   bIsVariable=true
   parent=SizeBox_Pip
   widget=Image_Pip
   properties {
    brush {
     bIsDynamicallyLoaded=false
     drawAs=Image
     imageType=NoImage
     mirroring=NoMirror
     resourceName=None
     resourceObject=null
     tiling=NoTile
     imageSize=<5 32>
     margin=<ltrb 0 0 0 0>
     outlineSettings {
      bUseBrushTransparency=false
      roundingType=FixedRadius
      width=0
      color {colorUseRule=UseColor_Specified, specifiedColor=<rgba 0 0 0 0>}
      cornerRadii=<4 4 4 4>
     }
     tintColor {colorUseRule=UseColor_Specified, specifiedColor=<rgba 1 1 1 1>}
     uVRegion {bIsValid=false, max=<0 0>, min=<0 0>}
    }
    colorAndOpacity=<rgba 1 1 1 0.5>
   }
   slot {class=SizeBoxSlot}
  }
  Spacer_MinHeight {
   widgetClassPath=&Spacer
   bIsVariable=true
   parent=Overlay_252
   widget=Spacer_MinHeight
   slot {class=OverlaySlot, properties {verticalAlignment=VAlign_Center}}
  }
 ]
}
