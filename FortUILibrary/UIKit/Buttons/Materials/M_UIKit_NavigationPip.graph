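# Translucent UI material for a navigation pip button graphic, drawn as a procedural circle with a selected state.
# A Selected scalar interpolates the size fed to MF_UI_Circle, so the dot grows from about two-thirds to full size.
# The same parameter raises a multiplier on the circle's fill from half to full, making unselected pips fainter.
# A constant mid-grey vector is also present.
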Material /FortUILibrary/UIKit/Buttons/Materials/M_UIKit_NavigationPip
graph {
 properties {
  bUsedWithStaticMesh=false
  blendMode=BLEND_Translucent
  materialDomain=MD_UI
  pixelDepthOffsetMode=PDOM_Legacy
  usedShadingModels=Default Lit
 }
 nodes [
  Constant3Vector() {constant=<rgba 0.521 0.521 0.521 1>}
  n1 = ScalarParameter() {parameterName=Selected, sliderMax=1}
  Multiply(
   A=MaterialFunctionCall(Size=LinearInterpolate(Alpha=@n1) {constA=0.666}) {materialFunction=&MF_UI_Circle}.Fill
   B=LinearInterpolate(Alpha=ScalarParameter() {parameterName=Selected}) {constA=0.5}
  )
 ]
}
