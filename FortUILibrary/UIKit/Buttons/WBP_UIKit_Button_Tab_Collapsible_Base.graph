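# Base widget class for collapsible tab buttons in the UIKit, derived from WBP_UIKit_Button_Selectable.
# It declares empty events for expanding and collapsing (each with an immediate flag), setting category status, and reacting to category expanded-state changes.
# PreConstruct, Construct and the selection events only call the parent.
# It defines no widgets or variables of its own.
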WidgetBlueprint /FortUILibrary/UIKit/Buttons/WBP_UIKit_Button_Tab_Collapsible_Base
graphs [EventGraph {nodes [
   n1 = Event PreConstruct()
   > Parent: PreConstruct(IsDesignTime=@n1.IsDesignTime)
   Event Construct()
   > Parent: Construct()
   Event On Selected()
   > Parent: On Selected() {is=title K2Node_CallParentFunction}
   Event On Deselected()
   > Parent: On Deselected() {is=title K2Node_CallParentFunction}
   On Expanded!() -> (OutputDelegate, then, IsImmediate:bool=~) {is=title K2Node_CustomEvent}
   On Collapsed!() -> (OutputDelegate, then, IsImmediate:bool=~) {is=title K2Node_CustomEvent}
   Set Is Category!() -> (OutputDelegate, then, IsCategory:bool=~) {is=title K2Node_CustomEvent}
   OnCategoryExpandedStateChanged!() -> (OutputDelegate, then, IsCategoryExpanded:bool=~) {is=title K2Node_CustomEvent}
  ]}]
variables []
widgets {
 info {
  inheritedWidgetCount=0
  parentClass=&/WBP_UIKit_Button_Selectable.WBP_UIKit_Button_Selectable_C
  rootWidgetClass=null
  widgetCount=0
 }
 widgets []
}
