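# UIKit sub-tab button, derived from the collapsible tab base, that shrinks from a left-aligned wrapping label into a navigation pip.
# One reversible OnCollapsed animation fades and scales content, shows the pip and drives CollapsedAnimValue to lerp vertical text padding; immediate changes jump to the animation's end.
# Text wrap toggles through a sequencer event or directly on immediate changes.
# Selection animates pip and backplate, hover drives the outline, and category expansion swaps typeface.
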WidgetBlueprint /FortUILibrary/UIKit/Buttons/WBP_UIKit_Button_Tab_CollapsibleSub_Enhanced
animations [
 OnCollapsed {
  binding_count=5
  length_seconds=0.25
  object=&self:OnCollapsed
  sequence {
   bindings [
    Slot_Background {
     id=#1
     possessed_class=NamedSlot
     tracks [{
       type=MovieSceneByteTrack
       display_name=Visibility
       sections [{type=MovieSceneByteSection, end_frame=25, start_frame=0, channels [None {type=MovieSceneScriptingByteChannel, keys [[0,4],[10,1],[25,1]]}]}]
      }]
    }
    UIKit_Backplate {
     id=#2
     possessed_class=WBP_UIKit_Backplate_C
     tracks [{
       type=MovieSceneFloatTrack
       display_name=Render Opacity
       sections [{type=MovieSceneFloatSection, end_frame=25, start_frame=0, channels [None {type=MovieSceneScriptingFloatChannel, keys [[0,1],[10,0],[25,0]]}]}]
      }]
    }
    Main {
     id=#3
     possessed_class=GridPanel
     tracks [
      {
       type=MovieSceneFloatTrack
       display_name=Render Opacity
       sections [{
         type=MovieSceneFloatSection
         end_frame=25
         start_frame=0
         channels [None {type=MovieSceneScriptingFloatChannel, thinned_from=4, keys [[0,1],[9.975,0],[25,0]]}]
        }]
      }
      {
       type=MovieScene2DTransformTrack
       sections [{
         type=MovieScene2DTransformSection
         end_frame=25
         start_frame=0
         channels [
          Translation.X {type=MovieSceneScriptingFloatChannel, const [0,1]}
          Translation.Y {type=MovieSceneScriptingFloatChannel, const [0,1]}
          Angle {type=MovieSceneScriptingFloatChannel, const [0,1]}
          Scale.X {type=MovieSceneScriptingFloatChannel, keys [[0,1],[25,0]]}
          Scale.Y {type=MovieSceneScriptingFloatChannel, keys [[0,1],[25,0]]}
          Shear.X {type=MovieSceneScriptingFloatChannel, const [0,1]}
          Shear.Y {type=MovieSceneScriptingFloatChannel, const [0,1]}
         ]
        }]
      }
      {
       type=MovieSceneDoubleVectorTrack
       display_name=Pivot
       sections [{
         type=MovieSceneDoubleVectorSection
         end_frame=0
         start_frame=0
         channels [
          Vector.X {type=MovieSceneScriptingDoubleChannel, const [0,1]}
          Vector.Y {type=MovieSceneScriptingDoubleChannel, const [0,1]}
         ]
        }]
      }
     ]
    }
    Self {
     id=#4
     possessed_class=WBP_UIKit_Button_Tab_CollapsibleSub_Enhanced_C
     tracks [
      {
       type=MovieSceneDoubleTrack
       display_name=Collapsed Anim Value
       sections [{type=MovieSceneDoubleSection, end_frame=25, start_frame=0, channels [None {type=MovieSceneScriptingDoubleChannel, keys [[0,0],[25,1]]}]}]
      }
      {
       type=MovieSceneByteTrack
       display_name=Visibility
       sections [{type=MovieSceneByteSection, end_frame=25, start_frame=0, channels [None {type=MovieSceneScriptingByteChannel, keys [[0,4],[10,3],[25,3]]}]}]
      }
     ]
    }
    Image_Pip {
     id=#5
     possessed_class=Image
     tracks [{
       type=MovieSceneColorTrack
       display_name=Color and Opacity
       sections [{
         type=MovieSceneColorSection
         end_frame=25
         start_frame=0
         channels [
          Color.R {type=MovieSceneScriptingFloatChannel, const [1,1]}
          Color.G {type=MovieSceneScriptingFloatChannel, const [1,1]}
          Color.B {type=MovieSceneScriptingFloatChannel, const [1,1]}
          Color.A {type=MovieSceneScriptingFloatChannel, keys [[0,0],[25,1]]}
         ]
        }]
      }]
    }
   ]
   display_rate {numerator=100}
   tracks [{
     type=MovieSceneEventTrack
     sections [{
       type=MovieSceneEventTriggerSection
       channels [None {type=MovieSceneScriptingEventChannel, const ["<Struct 'MovieSceneEvent' (0x466b2f2dca0) {}>", 1]}]
      }]
    }]
  }
 }
 OnSelected {
  binding_count=2
  length_seconds=0.15
  object=&self:OnSelected
  sequence {bindings [
    Image_Pip {
     id=#6
     possessed_class=Image
     tracks [{
       type=MovieSceneWidgetMaterialTrack
       sections [{type=MovieSceneParameterSection, end_frame=3, start_frame=0, channels [Selected {type=MovieSceneScriptingFloatChannel, keys [[0,0],[3,1]]}]}]
      }]
    }
    UIKit_Backplate {
     id=#7
     possessed_class=WBP_UIKit_Backplate_C
     tracks [{
       type=MovieSceneColorTrack
       display_name=Color and Opacity
       sections [{
         type=MovieSceneColorSection
         end_frame=3
         start_frame=0
         channels [
          Color.R {type=MovieSceneScriptingFloatChannel, const [1,1]}
          Color.G {type=MovieSceneScriptingFloatChannel, const [1,1]}
          Color.B {type=MovieSceneScriptingFloatChannel, const [1,1]}
          Color.A {type=MovieSceneScriptingFloatChannel, keys [[0,0],[3,1]]}
         ]
        }]
      }]
    }
   ]}
 }
]
graphs [
 EventGraph {nodes [
   n1 = Event PreConstruct()
   > Parent: PreConstruct(IsDesignTime=@n1.IsDesignTime)
   > InitializeLayout()
   > ResetAnimations()
   > InitializeText()
   > SetBlocksTiming!(
    execute
    self:UIKit Modular Button Object Reference
    Timing_Hovering:float=0
    Timing_Unhovering:float=0
    Timing_Focusing:float=0
    Timing_Unfocusing:float=0
    Timing_Pressing:float=0.1
    Timing_Releasing:float=0.1
    Timing_Disabling:float=0.1
    Timing_Enabling:float=0.1
    Timing_Selecting:float=0
    Timing_DeselectingIdle:float=0
    Timing_DeselectingFocused:float=0
   ) -> (then)
   > Set CachedTextMargin(CachedTextMargin:Margin Structure=Get TextPadding() -> (TextPadding:Margin Structure)) -> (Output_Get:Margin Structure)
   Event On Widget Styled() {is=title K2Node_Event}
   > Parent: On Widget Styled() {is=title K2Node_CallParentFunction}
   > Set TypefaceCollapsed(TypefaceCollapsed:Name=Break Slate Font Info!(SlateFontInfo:"Slate Font Info Structure (by ref)"=Get Font() -> (Font:Slate Font Info Structure)) -> (TypefaceFontName:Name, bMaterialIsStencil:bool=~) {is=title K2Node_BreakStruct}) -> (Output_Get:Name=None)
   > OnCategoryExpandedStateChanged!(
    execute
    self:WBP UIKit Button Tab Collapsible Base Object Reference
    IsCategoryExpanded:bool=Get Is Category Expanded() -> (Is Category Expanded:bool)
   ) -> (then)
   Event On Unhovered()
   > Parent: On Unhovered() {is=title K2Node_CallParentFunction}
   n2 = Get UIKit_Block_Outline() -> (UIKit_Block_Outline:WBP UIKit Block Outline Object Reference)
   > RequestTransition!(execute, self:UIKit Block Object Reference=@n2, State:EUIKitBlockVisualState Enum=Unfocused) -> (then)
   n3 = Event OnCategoryExpandedStateChanged!() -> (OutputDelegate, then, IsCategoryExpanded:bool) {is=title K2Node_Event}
   > n4 = Set Is Category Expanded(Is Category Expanded:bool=@n3.IsCategoryExpanded) -> (Output_Get:bool)
   > n5 = Get!(execute:Exec) -> (then:Exec, else:Exec, Block_Text:WBP UIKit Block Text Object Reference)
   n6 = Get!() -> (
    Font_FontObject:Object Reference
    Font_FontMaterial:Object Reference
    Font_OutlineSettings:Font Outline Settings Structure
    Font_TypefaceFontName:Name
    Font_Size:float=~
    Font_LetterSpacing:int
    Font_SkewAmount:float=~
    Font_bForceMonospaced:bool=~
    Font_bMaterialIsStencil:bool=~
    Font_MonospacedWidth:float=~
   )
   > SetFont!(
    execute
    self:WBP UIKit Block Text Object Reference=@n5
    Font_FontObject:Object Reference=@n6
    Font_FontMaterial:Object Reference=@n6.Font_FontMaterial
    Font_OutlineSettings:Font Outline Settings Structure=@n6.Font_OutlineSettings
    Font_TypefaceFontName:Name=SelectName!(
     A:Name=Get TypefaceExpanded() -> (TypefaceExpanded:Name=None)
     B:Name=Get TypefaceCollapsed() -> (TypefaceCollapsed:Name=None)
     bPickA:bool=@n4
    ) -> (ReturnValue:Name=None)
    Font_Size:float=@n6.Font_Size
    Font_LetterSpacing:int=@n6.Font_LetterSpacing
    Font_SkewAmount:float=@n6.Font_SkewAmount
    Font_bForceMonospaced:bool=@n6.Font_bForceMonospaced
    Font_bMaterialIsStencil:bool=false
    Font_MonospacedWidth:float=@n6.Font_MonospacedWidth
   ) -> (then)
   n7 = Event On Collapsed!() -> (OutputDelegate, then, IsImmediate:bool) {is=title K2Node_Event}
   > Parent: On Collapsed!(execute, IsImmediate:bool=@n7.IsImmediate) -> (then) {is=title K2Node_CallParentFunction}
   > n8 = Branch(Condition=@n7.IsImmediate)
   n9 = Get OnCollapsed() -> (OnCollapsed:Widget Animation Object Reference)
   > PlayAnimation(InAnimation=@n9, StartAtTime=GetEndTime(self=@n9))
   > Update Text Wrap!(execute, self, ShouldCollapse:bool=true) -> (then)
   PlayAnimationForward(execute=@n8.else, InAnimation=Get OnCollapsed())
   n10 = Event On Expanded!() -> (OutputDelegate, then, IsImmediate:bool) {is=title K2Node_Event}
   > Parent: On Expanded!(execute, IsImmediate:bool=@n10.IsImmediate) -> (then) {is=title K2Node_CallParentFunction}
   > n11 = Branch(Condition=@n10.IsImmediate)
   n12 = Get OnCollapsed()
   > PlayAnimation(InAnimation=@n12, StartAtTime=GetEndTime(self=@n12), PlayMode=Reverse)
   > Update Text Wrap(ShouldCollapse=false)
   PlayAnimationReverse(execute=@n11.else, InAnimation=Get OnCollapsed())
   Event On Focused() {is=title K2Node_Event}
   > Parent: On Focused() {is=title K2Node_CallParentFunction}
   Event On Unfocused() {is=title K2Node_Event}
   > Parent: On Unfocused() {is=title K2Node_CallParentFunction}
   Event On Selected()
   > Parent: On Selected() {is=title K2Node_CallParentFunction}
   n13 = Get OnSelected() -> (OnSelected:Widget Animation Object Reference)
   > PlayAnimationForward(InAnimation=@n13)
   Event On Deselected()
   > Parent: On Deselected() {is=title K2Node_CallParentFunction}
   > PlayAnimationReverse(InAnimation=@n13)
   Event On Hovered()
   > Parent: On Hovered() {is=title K2Node_CallParentFunction}
   > RequestTransition(self=@n2, State=Focused)
  ]}
 InitializeLayout {nodes [
   InitializeLayout() {is=title K2Node_FunctionEntry}
   > UpdatePadding!(execute, self, SelectionAlignment:E_UI_Alignment Enum=NewEnumerator2) -> (then)
   > UpdateAlignment!(execute, self, Selection Alignment:E_UI_Alignment Enum=NewEnumerator2) -> (then)
   > SetGridSlotAlignment!(
    execute
    Widget:Widget Object Reference=Get Slot_Main() -> (Slot_Main:Named Slot Object Reference)
    Horizontal:EHorizontalAlignment Enum=HAlign_Fill
    Vertical:EVerticalAlignment Enum=VAlign_Fill
   ) -> (then)
  ]}
 InitializeText {nodes [
   InitializeText() {is=title K2Node_FunctionEntry}
   n1 = Get TextBlock!(self:WBP UIKit Block Text Object Reference=Get Block_Text() -> (Block_Text:WBP UIKit Block Text Object Reference)) -> (TextBlock:Common Text Object Reference)
   > SetWrapTextWidth!(execute, self:Common Text Object Reference=@n1, InWrapTextAt:int=Get WrapTextAt() -> (WrapTextAt:int=0)) -> (then)
   > Set Justification!(execute, Justification:ETextJustify Enum=Left, self:Text Layout Widget Object Reference=@n1) -> (then, Output_Get:ETextJustify Enum=Left)
   > Set TextTransformPolicy!(execute, TextTransformPolicy:ETextTransformPolicy Enum=None, self:Text Object Reference=@n1) -> (then, Output_Get:ETextTransformPolicy Enum=None)
  ]}
 ResetAnimations {nodes [
   ResetAnimations() {is=title K2Node_FunctionEntry}
   n1 = Get OnSelected()
   > PlayAnimation(InAnimation=@n1, StartAtTime=GetEndTime(self=@n1), PlayMode=Reverse)
   n2 = Get OnCollapsed()
   > PlayAnimation(InAnimation=@n2, StartAtTime=GetEndTime(self=@n2), PlayMode=Reverse)
  ]}
 Sequencer Events {nodes [
   SequencerEventToggleTextWrap() {is=title K2Node_CustomEvent}
   n1 = Get TextBlock(self=Get Block_Text())
   n2 = Get WrapTextAt!(self:Text Layout Widget Object Reference=@n1) -> (WrapTextAt:float)
   > SetWrapTextWidth(self=@n1, InWrapTextAt=SelectInt(A=0, B=Get WrapTextAt!() -> (WrapTextAt:int=0), bPickA="float > float"(A=@n2)))
  ]}
 SetCollapsedAnimValue {nodes [
   n1 = SetCollapsedAnimValue!() -> (then, Value:double=~) {is=title K2Node_FunctionEntry}
   n2 = Get!() -> (CachedTextMargin_Left:float=~, CachedTextMargin_Top:float=~, CachedTextMargin_Right:float=~, CachedTextMargin_Bottom:float=~)
   > SetTextPadding!(
    execute
    self:WBP UIKit Button Generic Base Object Reference
    TextPadding_Left:float=@n2
    TextPadding_Top:float=Lerp(A=@n2.CachedTextMargin_Top, B=0, Alpha=@n1)
    TextPadding_Right:float=@n2.CachedTextMargin_Right
    TextPadding_Bottom:float=Lerp(A=@n2.CachedTextMargin_Bottom, B=0, Alpha=@n1)
   ) -> (then)
  ]}
 SetSelectionAlignment {nodes [
   n1 = SetSelectionAlignment!() -> (then, Selection Alignment:E_UI_Alignment Enum=NewEnumerator0) {is=title K2Node_FunctionEntry}
   > Set SelectedBarPosition(SelectedBarPosition:E_UI_Alignment Enum=@n1) -> (Output_Get:E_UI_Alignment Enum=NewEnumerator0)
  ]}
 Update Text Wrap {nodes [
   n1 = Update Text Wrap!() -> (then, ShouldCollapse:bool=~) {is=title K2Node_FunctionEntry}
   > SetWrapTextWidth(self=Get TextBlock(self=Get Block_Text()), InWrapTextAt=SelectInt(A=0, B=Get WrapTextAt() -> (WrapTextAt=0), bPickA=@n1))
  ]}
 UpdateAlignment {nodes [
   n1 = UpdateAlignment!() -> (then, Selection Alignment:E_UI_Alignment Enum=NewEnumerator0) {is=title K2Node_FunctionEntry}
   > SetButtonContentAlignment!(
    execute
    self:WBP UIKit Button Generic Base Object Reference
    HorizontalAlignment:EHorizontalAlignment Enum=Select!(
     NewEnumerator0:EHorizontalAlignment Enum=HAlign_Center
     NewEnumerator1:EHorizontalAlignment Enum=HAlign_Center
     NewEnumerator2:EHorizontalAlignment Enum=HAlign_Left
     NewEnumerator3:EHorizontalAlignment Enum=HAlign_Right
     Index:E_UI_Alignment Enum=@n1
    ) -> (ReturnValue:EHorizontalAlignment Enum)
    VerticalAlignment:EVerticalAlignment Enum=Select(
     NewEnumerator0:EVerticalAlignment Enum=VAlign_Top
     NewEnumerator1:EVerticalAlignment Enum=VAlign_Bottom
     NewEnumerator2:EVerticalAlignment Enum=VAlign_Center
     NewEnumerator3:EVerticalAlignment Enum=VAlign_Center
     Index=@n1
    ) -> (ReturnValue:EVerticalAlignment Enum)
   ) -> (then)
  ]}
 UpdatePadding {nodes [
   n1 = UpdatePadding!() -> (then, SelectionAlignment:E_UI_Alignment Enum=NewEnumerator0) {is=title K2Node_FunctionEntry}
   > SetButtonPadding!(
    execute
    self:WBP UIKit Button Generic Base Object Reference
    Padding:Margin Structure=Select(
     NewEnumerator0:Margin Structure=Get TopAlignmentPadding() -> (TopAlignmentPadding:Margin Structure)
     NewEnumerator1:Margin Structure=Get BottomAlignmentPadding() -> (BottomAlignmentPadding:Margin Structure)
     NewEnumerator2:Margin Structure=Get LeftAlignmentPadding() -> (LeftAlignmentPadding:Margin Structure)
     NewEnumerator3:Margin Structure=Get RightAlignmentPadding() -> (RightAlignmentPadding:Margin Structure)
     Index=@n1
    ) -> (ReturnValue:Margin Structure)
   ) -> (then)
  ]}
]
variables [
 BottomAlignmentPadding {type=Margin, default=<ltrb 0 0 0 0>}
 CachedTextMargin {type=Margin, default=<ltrb 0 0 0 0>}
 CollapsedAnimValue {type=number}
 Is Category Expanded {type=boolean}
 LeftAlignmentPadding {type=Margin, default=<ltrb 0 0 0 0>}
 RightAlignmentPadding {type=Margin, default=<ltrb 0 0 0 0>}
 SelectedBarPosition {type=E_UI_Alignment, default=Left}
 TopAlignmentPadding {type=Margin, default=<ltrb 0 0 0 0>}
 TypefaceCollapsed {type=string, default=74 Regular}
 TypefaceExpanded {type=string, default=75 Medium}
 WrapTextAt {type=integer, default=228}
]
widgets {
 info {
  inheritedWidgetCount=0
  namedSlotCount=2
  parentClass=&/WBP_UIKit_Button_Tab_Collapsible_Base.WBP_UIKit_Button_Tab_Collapsible_Base_C
  rootWidgetClass=null
  widgetCount=4
 }
 widgets [
  UIKit_Backplate {widgetClassPath=&WBP_UIKit_Backplate, widget=UIKit_Backplate, properties {cornerRadius=Tab}}
  Overlay_40 {widgetClassPath=&Overlay, widget=Overlay_40}
  UIKit_Block_Outline {
   widgetClassPath=/FortUILibrary/UIKit/Buttons/Blocks/WBP_UIKit_Block_Outline
   bIsVariable=true
   parent=Overlay_40
   widget=UIKit_Block_Outline
   properties {
    bottomLeftCornerRadius=16
    bottomRightCornerRadius=16
    contain Outline Padding=true
    radiusInPixels=true
    topLeftCornerRadius=16
    topRightCornerRadius=16
    visibility=SelfHitTestInvisible
   }
   slot {class=OverlaySlot, properties {horizontalAlignment=HAlign_Fill, verticalAlignment=VAlign_Fill, padding=<ltrb -8 -8 -8 -8>}}
  }
  Image_Pip {
   widgetClassPath=&Image
   bIsVariable=true
   parent=Overlay_40
   widget=Image_Pip
   properties {brush {
     bIsDynamicallyLoaded=false
     drawAs=Image
     imageType=NoImage
     mirroring=NoMirror
     resourceName=None
     resourceObject=/FortUILibrary/UIKit/Buttons/Materials/M_UIKit_NavigationPip
     tiling=NoTile
     imageSize=<18 18>
     margin=<ltrb 0 0 0 0>
     outlineSettings {
      bUseBrushTransparency=false
      roundingType=FixedRadius
      width=0
      color {colorUseRule=UseColor_Specified, specifiedColor=<rgba 0 0 0 0>}
      cornerRadii=<4 4 4 4>
     }
     tintColor {colorUseRule=UseColor_Specified, specifiedColor=<rgba 1 1 1 1>}
     uVRegion {bIsValid=false, max=<0 0>, min=<0 0>}
    }}
   slot {class=OverlaySlot, properties {horizontalAlignment=HAlign_Center, verticalAlignment=VAlign_Center}}
  }
 ]
}
