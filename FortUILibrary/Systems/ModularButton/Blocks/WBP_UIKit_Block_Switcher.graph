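# UIKit modular button block that draws a toggle switch, with a material-driven background Image and a Toggle named slot that slides horizontally on selection.
# Transition events from the base block play animations or call instant-transition helpers that drive per-state `Trans_*` scalar parameters on ImageMID.
# PreConstruct applies the Brush and ToggleType, which becomes a material scalar.
# Brush or material swaps rebuild the MID and copy its parameter overrides.
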WidgetBlueprint /FortUILibrary/Systems/ModularButton/Blocks/WBP_UIKit_Block_Switcher
animations [
 Anim_Hovered_Transition {
  binding_count=1
  length_seconds=1
  object=&self:Anim_Hovered_Transition
  sequence {bindings [Image {
     id=#1
     possessed_class=Image
     tracks [{
       type=MovieSceneWidgetMaterialTrack
       sections [{
         type=MovieSceneParameterSection
         end_frame=20
         start_frame=0
         channels [Trans_Hovered {type=MovieSceneScriptingFloatChannel, keys [[0,0],[20,1]]}]
        }]
      }]
    }]}
 }
 Anim_Focused_Transition {
  binding_count=1
  length_seconds=1
  object=&self:Anim_Focused_Transition
  sequence {bindings [Image {
     id=#1
     possessed_class=Image
     tracks [{
       type=MovieSceneWidgetMaterialTrack
       sections [{
         type=MovieSceneParameterSection
         end_frame=20
         start_frame=0
         channels [Trans_Focused {type=MovieSceneScriptingFloatChannel, keys [[0,0],[20,1]]}]
        }]
      }]
    }]}
 }
 Anim_Pressed_Transition {
  binding_count=1
  length_seconds=1
  object=&self:Anim_Pressed_Transition
  sequence {bindings [Image {
     id=#2
     possessed_class=Image
     tracks [{
       type=MovieSceneWidgetMaterialTrack
       sections [{
         type=MovieSceneParameterSection
         end_frame=20
         start_frame=0
         channels [Trans_Pressed {type=MovieSceneScriptingFloatChannel, keys [[0,0],[20,1]]}]
        }]
      }]
    }]}
 }
 Anim_Disabled_Transition {
  binding_count=1
  length_seconds=1
  object=&self:Anim_Disabled_Transition
  sequence {bindings [Image {
     id=#2
     possessed_class=Image
     tracks [{
       type=MovieSceneWidgetMaterialTrack
       sections [{
         type=MovieSceneParameterSection
         end_frame=20
         start_frame=0
         channels [Trans_Disabled {type=MovieSceneScriptingFloatChannel, keys [[0,0],[20,1]]}]
        }]
      }]
    }]}
 }
 Anim_Selected_Transition {
  binding_count=2
  length_seconds=1
  object=&self:Anim_Selected_Transition
  sequence {bindings [
    Toggle {
     id=#3
     possessed_class=NamedSlot
     tracks [{
       type=MovieScene2DTransformTrack
       sections [{
         type=MovieScene2DTransformSection
         end_frame=20
         start_frame=0
         channels [
          Translation.X {type=MovieSceneScriptingFloatChannel, keys [[0,0],[20,35]]}
          Translation.Y {type=MovieSceneScriptingFloatChannel, const [0,1]}
          Angle {type=MovieSceneScriptingFloatChannel, const [0,1]}
          Scale.X {type=MovieSceneScriptingFloatChannel, const [1,1]}
          Scale.Y {type=MovieSceneScriptingFloatChannel, const [1,1]}
          Shear.X {type=MovieSceneScriptingFloatChannel, const [0,1]}
          Shear.Y {type=MovieSceneScriptingFloatChannel, const [0,1]}
         ]
        }]
      }]
    }
    Image {
     id=#4
     possessed_class=FortMobileImage
     tracks [{
       type=MovieSceneWidgetMaterialTrack
       sections [{
         type=MovieSceneParameterSection
         end_frame=20
         start_frame=0
         channels [Trans_SelectedIdle {type=MovieSceneScriptingFloatChannel, keys [[0,0],[19.9997,1]]}]
        }]
      }]
    }
   ]}
 }
 Anim_SelectedHovered_Transition {
  binding_count=3
  length_seconds=1
  object=&self:Anim_SelectedHovered_Transition
  sequence {bindings [
    Toggle {
     id=#5
     possessed_class=NamedSlot
     tracks [{
       type=MovieScene2DTransformTrack
       sections [{
         type=MovieScene2DTransformSection
         end_frame=20
         start_frame=0
         channels [
          Translation.X {type=MovieSceneScriptingFloatChannel, keys [[0,0],[20,35]]}
          Translation.Y {type=MovieSceneScriptingFloatChannel, const [0,1]}
          Angle {type=MovieSceneScriptingFloatChannel, const [0,1]}
          Scale.X {type=MovieSceneScriptingFloatChannel, const [1,1]}
          Scale.Y {type=MovieSceneScriptingFloatChannel, const [1,1]}
          Shear.X {type=MovieSceneScriptingFloatChannel, const [0,1]}
          Shear.Y {type=MovieSceneScriptingFloatChannel, const [0,1]}
         ]
        }]
      }]
    }
    Image {
     id=#6
     possessed_class=FortMobileImage
     tracks [{
       type=MovieSceneWidgetMaterialTrack
       sections [{
         type=MovieSceneParameterSection
         end_frame=20
         start_frame=0
         channels [Trans_SelectedHovered {type=MovieSceneScriptingFloatChannel, keys [[0,0],[19.9997,1]]}]
        }]
      }]
    }
   ]}
 }
 Anim_SelectedFocused_Transition {
  binding_count=1
  length_seconds=1
  object=&self:Anim_SelectedFocused_Transition
  sequence {bindings [Image {
     id=#6
     possessed_class=FortMobileImage
     tracks [{
       type=MovieSceneWidgetMaterialTrack
       sections [{
         type=MovieSceneParameterSection
         end_frame=20
         start_frame=0
         channels [Trans_SelectedFocused {type=MovieSceneScriptingFloatChannel, keys [[0,0],[19.9997,1]]}]
        }]
      }]
    }]}
 }
 Anim_SelectedPressed_Transition {
  binding_count=3
  length_seconds=1
  object=&self:Anim_SelectedPressed_Transition
  sequence {bindings [
    Image {
     id=#6
     possessed_class=FortMobileImage
     tracks [{
       type=MovieSceneWidgetMaterialTrack
       sections [{
         type=MovieSceneParameterSection
         end_frame=20
         start_frame=0
         channels [Trans_SelectedPressed {type=MovieSceneScriptingFloatChannel, keys [[0,0],[19.9997,1]]}]
        }]
      }]
    }
    Toggle {
     id=#7
     possessed_class=NamedSlot
     tracks [{
       type=MovieScene2DTransformTrack
       sections [{
         type=MovieScene2DTransformSection
         end_frame=20
         start_frame=0
         channels [
          Translation.X {type=MovieSceneScriptingFloatChannel, keys [[0,0],[20,35]]}
          Translation.Y {type=MovieSceneScriptingFloatChannel, const [0,1]}
          Angle {type=MovieSceneScriptingFloatChannel, const [0,1]}
          Scale.X {type=MovieSceneScriptingFloatChannel, const [1,1]}
          Scale.Y {type=MovieSceneScriptingFloatChannel, const [1,1]}
          Shear.X {type=MovieSceneScriptingFloatChannel, const [0,1]}
          Shear.Y {type=MovieSceneScriptingFloatChannel, const [0,1]}
         ]
        }]
      }]
    }
   ]}
 }
 Anim_SelectedDisabled_Transition {
  binding_count=3
  length_seconds=1
  object=&self:Anim_SelectedDisabled_Transition
  sequence {bindings [
    Toggle {
     id=#8
     possessed_class=NamedSlot
     tracks [{
       type=MovieScene2DTransformTrack
       sections [{
         type=MovieScene2DTransformSection
         end_frame=20
         start_frame=0
         channels [
          Translation.X {type=MovieSceneScriptingFloatChannel, keys [[0,0],[20,35]]}
          Translation.Y {type=MovieSceneScriptingFloatChannel, const [0,1]}
          Angle {type=MovieSceneScriptingFloatChannel, const [0,1]}
          Scale.X {type=MovieSceneScriptingFloatChannel, const [1,1]}
          Scale.Y {type=MovieSceneScriptingFloatChannel, const [1,1]}
          Shear.X {type=MovieSceneScriptingFloatChannel, const [0,1]}
          Shear.Y {type=MovieSceneScriptingFloatChannel, const [0,1]}
         ]
        }]
      }]
    }
    Image {
     id=#6
     possessed_class=FortMobileImage
     tracks [{
       type=MovieSceneWidgetMaterialTrack
       sections [{
         type=MovieSceneParameterSection
         end_frame=20
         start_frame=0
         channels [Trans_SelectedDisabled {type=MovieSceneScriptingFloatChannel, keys [[0,0],[19.9997,1]]}]
        }]
      }]
    }
   ]}
 }
]
graphs [
 EventGraph {nodes [
   Event OnInstantTransitionToSelected() {is=title K2Node_Event}
   > Stop All Transitions!(In:Exec) -> (Out:Exec) {is=title K2Node_MacroInstance}
   > InstantTransitionToSelected!(execute, MID:Material Instance Dynamic Object Reference=Get ImageMID() -> (ImageMID:Material Instance Dynamic Object Reference)) -> (then)
   Event OnTransitionSelectedRequested() {is=title K2Node_Event}
   > PlayTransition!(
    execute
    self:UIKit Block Object Reference
    Animation:Widget Animation Object Reference=Get Anim_Selected_Transition() -> (Anim_Selected_Transition:Widget Animation Object Reference)
    bPlayForward:bool=true
   ) -> (then)
   Event OnTransitionFocusedRequested() {is=title K2Node_Event}
   > n1 = Branch(Condition=GetIsSelected!() -> (IsSelected:bool))
   > PlayTransition(Animation=Get Anim_SelectedFocused_Transition() -> (Anim_SelectedFocused_Transition:Widget Animation Object Reference), bPlayForward=true)
   PlayTransition(execute=@n1.else, Animation=Get Anim_Focused_Transition() -> (Anim_Focused_Transition:Widget Animation Object Reference), bPlayForward=true)
   Event OnTransitionUnfocusedRequested() {is=title K2Node_Event}
   > n2 = Branch(Condition=GetIsSelected())
   > SetScalarParameterValue(self=Get ImageMID(), ParameterName=Trans_SelectedIdle, Value=1)
   > PlayTransition(Animation=Get Anim_SelectedFocused_Transition(), bPlayForward=false)
   StopAnimation(execute=@n2.else, InAnimation=Get Anim_Selected_Transition())
   > PlayTransition(Animation=Get Anim_Focused_Transition(), bPlayForward=false)
   Event PreConstruct()
   > SetBrush(self:Self Object Reference, InBrush=Get Brush() -> (Brush:Slate Brush Structure))
   > SetToggleType!(execute, self, Type:E_UI_ToggleType Enum=Get ToggleType() -> (ToggleType:E_UI_ToggleType Enum=NewEnumerator0)) -> (then)
   Event OnTransitionDeselectedFocusedRequested() {is=title K2Node_Event}
   > PlayTransition(Animation=Get Anim_SelectedHovered_Transition() -> (Anim_SelectedHovered_Transition:Widget Animation Object Reference), bPlayForward=false)
   > StopAnimation(InAnimation=Get Anim_Selected_Transition())
   > n3 = TransitionToDeselectedFocused!(execute, MID:Material Instance Dynamic Object Reference=Get ImageMID()) -> (then)
   TransitionToDeselectedIdle!(execute=[@n3.then, @n4.then], MID:Material Instance Dynamic Object Reference=Get ImageMID()) -> (then)
   > SetScalarParameterValue(self=Get ImageMID(), ParameterName=Trans_SelectedIdle, Value=0)
   Event OnTransitionDeselectedIdleRequested() {is=title K2Node_Event}
   > n4 = PlayTransition(Animation=Get Anim_Selected_Transition(), bPlayForward=false)
   n5 = Branch(Condition=GetIsSelected())
   > PlayTransition(Animation=Get Anim_SelectedDisabled_Transition() -> (Anim_SelectedDisabled_Transition:Widget Animation Object Reference), bPlayForward=false)
   PlayTransition(
    execute=@n5.else
    Animation=Get Anim_Disabled_Transition() -> (Anim_Disabled_Transition:Widget Animation Object Reference)
    bPlayForward=false
   )
   n6 = Branch(Condition=GetIsSelected())
   > PlayTransition(Animation=Get Anim_SelectedDisabled_Transition(), bPlayForward=true)
   PlayTransition(execute=@n6.else, Animation=Get Anim_Disabled_Transition(), bPlayForward=true)
   Stop All Transitions()
   > InstantTransitionToEnabled!(execute, MID:Material Instance Dynamic Object Reference=Get ImageMID(), IsSelected:bool=GetIsSelected()) -> (then)
   Stop All Transitions()
   > InstantTransitionToDisabled!(execute, MID:Material Instance Dynamic Object Reference=Get ImageMID(), IsSelected:bool=GetIsSelected()) -> (then)
   Event OnTransitionDisabledRequested() {is=title K2Node_Event}
   Event OnTransitionEnabledRequested() {is=title K2Node_Event}
   Event OnInstantTransitionToEnabled() {is=title K2Node_Event}
   Event OnInstantTransitionToDisabled() {is=title K2Node_Event}
  ]}
 RefreshImageMID {nodes [
   RefreshImageMID() {is=title K2Node_FunctionEntry}
   > n1 = GetDynamicMaterial(self=Get Image() -> (Image:Fort Mobile Image Object Reference))
   > n2 = Set NewMID(NewMID:Material Instance Dynamic Object Reference=@n1) -> (Output_Get:Material Instance Dynamic Object Reference)
   > Copy MID Parameter Overrides!(execute, Target MID:Material Instance Dynamic Object Reference=@n2, Source MID:Material Instance Dynamic Object Reference=Get ImageMID()) -> (then)
   > Set ImageMID(ImageMID:Material Instance Dynamic Object Reference=@n2) -> (Output_Get:Material Instance Dynamic Object Reference)
  ]}
 SetBrush {nodes [
   n1 = SetBrush!() -> (then, InBrush:"Slate Brush Structure (by ref)") {is=title K2Node_FunctionEntry}
   > SetBrush(self:Image Object Reference=Get Image(), InBrush=@n1) {is=title K2Node_CallFunction}
   > RefreshImageMID() {is=title K2Node_CallFunction}
  ]}
 SetMaterial {nodes [
   n1 = SetMaterial!() -> (then, InMaterial:"Material Instance Object Reference (by ref)") {is=title K2Node_FunctionEntry}
   > SetBrushFromMaterial(self=Get Image(), Material=@n1)
   > RefreshImageMID()
  ]}
 SetSelected {nodes [
   n1 = SetSelected!() -> (then, Selected:bool=~) {is=title K2Node_FunctionEntry}
   > n2 = Branch(Condition=@n1)
   n3 = Get ImageMID()
   > n4 = SetScalarParameterValue(self=@n3, ParameterName=Trans_SelectedIdle, Value=1)
   n5 = Get ImageMID()
   SetScalarParameterValue(execute=[@n4.then, @n6.then], self=@n5, ParameterName=Trans_Hovered, Value=0)
   > SetScalarParameterValue(self=@n5, ParameterName=Trans_Focused, Value=0)
   > SetScalarParameterValue(self=@n5, ParameterName=Trans_Pressed, Value=0)
   > SetScalarParameterValue(self=@n5, ParameterName=Trans_Disabled, Value=0)
   > SetScalarParameterValue(self=@n5, ParameterName=Trans_SelectedHovered, Value=0)
   > SetScalarParameterValue(self=@n5, ParameterName=Trans_SelectedFocused, Value=0)
   > SetScalarParameterValue(self=@n5, ParameterName=Trans_SelectedPressed, Value=0)
   > SetScalarParameterValue(self=@n5, ParameterName=Trans_SelectedDisabled, Value=0)
   n6 = SetScalarParameterValue(execute=@n2.else, self=@n3, ParameterName=Trans_SelectedIdle, Value=0)
  ]}
 SetSize {nodes [
   n1 = SetSize!() -> (then, Desired Size:Vector 2D Structure) {is=title K2Node_FunctionEntry}
   > SetDesiredSizeOverride!(execute, self:Image Object Reference=Get Image(), DesiredSize:Vector 2D Structure=@n1) -> (then)
  ]}
 SetToggleType {nodes [
   n1 = SetToggleType!() -> (then, Type:E_UI_ToggleType Enum=NewEnumerator0) {is=title K2Node_FunctionEntry}
   > SetScalarParameterValue(
    self=Get ImageMID()
    ParameterName=ToggleType
    Value=Select!(NewEnumerator0:float=0, NewEnumerator1:float=0.5, NewEnumerator2:float=1, Index:E_UI_ToggleType Enum=@n1) -> (ReturnValue:float)
   )
  ]}
]
variables [
 Brush {
  type=SlateBrush
  default {
   bIsDynamicallyLoaded=false
   drawAs=Image
   imageType=FullColor
   mirroring=NoMirror
   resourceName=None
   resourceObject=/FortUILibrary/UIKit/Buttons/MaterialInstances/SelectableButtons/MI_UIKit_ToggleSwitchedBackground
   tiling=NoTile
   imageSize=<68 32>
   margin=<ltrb 0 0 0 0>
   outlineSettings {
    bUseBrushTransparency=false
    roundingType=HalfHeightRadius
    width=0
    color {colorUseRule=UseColor_Specified, specifiedColor=<rgba 0 0 0 0>}
    cornerRadii=<0 0 0 1>
   }
   tintColor {colorUseRule=UseColor_Specified, specifiedColor=<rgba 1 1 1 1>}
   uVRegion {bIsValid=false, max=<0 0>, min=<0 0>}
  }
 }
 ImageMID {type=MaterialInstanceDynamic}
 ToggleType {type=E_UI_ToggleType, default=CheckBox}
]
widgets {
 info {
  inheritedWidgetCount=0
  parentClass=&/WBP_UIKit_Block_Base.WBP_UIKit_Block_Base_C
  rootWidgetClass=&GridPanel
  widgetCount=3
 }
 widgets [
  Root {widgetClassPath=&GridPanel, widget=Root, properties {columnFill [1], rowFill [1]}}
  Image {
   widgetClassPath=/Script/FortniteUI.FortMobileImage
   bIsVariable=true
   parent=Root
   widget=Image
   properties {
    visibility=SelfHitTestInvisible
    brush {
     bIsDynamicallyLoaded=false
     drawAs=Image
     imageType=FullColor
     mirroring=NoMirror
     resourceName=None
     resourceObject=&MI_UIKit_ToggleSwitchedBackground
     tiling=NoTile
     imageSize=<68 32>
     margin=<ltrb 0 0 0 0>
     outlineSettings {
      bUseBrushTransparency=false
      roundingType=HalfHeightRadius
      width=0
      color {colorUseRule=UseColor_Specified, specifiedColor=<rgba 0 0 0 0>}
      cornerRadii=<0 0 0 1>
     }
     tintColor {colorUseRule=UseColor_Specified, specifiedColor=<rgba 1 1 1 1>}
     uVRegion {bIsValid=false, max=<0 0>, min=<0 0>}
    }
   }
   slot {class=GridSlot}
  }
  Toggle {
   widgetClassPath=&NamedSlot
   bIsVariable=true
   parent=Root
   widget=Toggle
   slot {class=GridSlot, properties {horizontalAlignment=HAlign_Left, layer=5, verticalAlignment=VAlign_Center, padding=<ltrb 2.5 0 2 0>}}
  }
 ]
}
