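# Base CTA button widget built on the UIKit modular button, with background, primary text, and optional icon and secondary text blocks.
# Registered blocks receive timed state transitions for focus, press and enable changes, alongside a press scale animation.
# A button size enum drives fonts, icon size, padding and minimum dimensions, with compact-UI overrides.
# An input prompt follows input type and shows hold progress through a dynamic material, with hover, press and hold sounds.
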WidgetBlueprint /FortUILibrary/Systems/ModularButton/WBP_UIKit_ButtonCTA_Base
animations [Anim_Pressed_Base {
  binding_count=2
  length_seconds=0.1
  object=&self:Anim_Pressed_Base
  sequence {bindings [
    Texts {
     id=#1
     possessed_class=StackBox
     tracks [{
       type=MovieScene2DTransformTrack
       sections [{
         type=MovieScene2DTransformSection
         channels [
          Scale.X {type=MovieSceneScriptingFloatChannel, keys [[0,1],[2,0.95]]}
          Scale.Y {type=MovieSceneScriptingFloatChannel, keys [[0,1],[2,0.975]]}
         ]
        }]
      }]
    }
    Root {
     id=#2
     possessed_class=GridPanel
     tracks [{
       type=MovieScene2DTransformTrack
       sections [{
         type=MovieScene2DTransformSection
         channels [
          Scale.X {type=MovieSceneScriptingFloatChannel, keys [[0,1],[2,0.975]]}
          Scale.Y {type=MovieSceneScriptingFloatChannel, keys [[0,1],[2,0.975]]}
         ]
        }]
      }]
    }
   ]}
 }]
graphs [
 AddIconBlock {nodes [
   AddIconBlock() {is=title K2Node_FunctionEntry}
   > n1 = Create Widget!(
    execute:Exec
    Class:User Widget Class Reference=/FortUILibrary/Systems/ModularButton/Blocks/WBP_UIKit_Block_Image.WBP_UIKit_Block_Image_C
    OwningPlayer:Player Controller Object Reference=GetOwningPlayer()
   ) -> (then:Exec, ReturnValue:WBP UIKit Block Image Object Reference)
   > n2 = Set Block_Icon(Block_Icon:WBP UIKit Block Image Object Reference=@n1) -> (Output_Get:WBP UIKit Block Image Object Reference)
   > Set HasIcon(HasIcon:bool=true) -> (Output_Get:bool)
   > RegisterModularBlock!(execute, self:UIKit Modular Button Object Reference, Block:UIKit Block Object Reference=@n2) -> (then)
   n3 = Get IconBeforeText() -> (IconBeforeText:bool)
   > n4 = AddChildToGrid!(
    execute
    self:Grid Panel Object Reference=Get Root() -> (Root:Grid Panel Object Reference)
    Content:Widget Object Reference=@n2
    InRow:int=1
    InColumn:int=SelectInt(A=SelectInt(A=2, B=4, bPickA=@n3), B=3, bPickA=Get HasText() -> (HasText:bool))
   ) -> (then, ReturnValue:Grid Slot Object Reference)
   > SetHorizontalAlignment(self:Grid Slot Object Reference=@n4, InHorizontalAlignment=HAlign_Center)
   > SetVerticalAlignment(self:Grid Slot Object Reference=@n4, InVerticalAlignment=VAlign_Center)
   > Set Padding!(execute, Padding:Margin Structure=Get IconPadding() -> (IconPadding:Margin Structure), self:Grid Slot Object Reference=@n4) -> (then, Output_Get:Margin Structure)
   > SetSize!(
    execute
    self:WBP UIKit Block Image Object Reference=@n2
    Desired Size:Vector 2D Structure=Select!(
     NewEnumerator0:Vector 2D Structure=Get LargeIconSize() -> (LargeIconSize:Vector 2D Structure)
     NewEnumerator1:Vector 2D Structure=Get MediumIconSize() -> (MediumIconSize:Vector 2D Structure)
     NewEnumerator2:Vector 2D Structure=Get SmallIconSize() -> (SmallIconSize:Vector 2D Structure)
     Index:E_UI_ButtonSize Enum=Get ButtonSize() -> (ButtonSize:E_UI_ButtonSize Enum=NewEnumerator0)
    ) -> (ReturnValue:Vector 2D Structure)
   ) -> (then)
   > SetBlockTiming!(
    execute
    self:UIKit Block Object Reference=@n2
    Timing_Hovering:float=0
    Timing_Unhovering:float=0
    Timing_Focusing:float=0.15
    Timing_Unfocusing:float=0.15
    Timing_Pressing:float=0.1
    Timing_Releasing:float=0.1
    Timing_Disabling:float=0.1
    Timing_Enabling:float=0.1
    Timing_Selecting:float=0
    Timing_DeselectingIdle:float=0
    Timing_DeselectingFocused:float=0
   ) -> (then)
  ]}
 AddSecondTextBlock {nodes [
   AddSecondTextBlock() {is=title K2Node_FunctionEntry}
   > n1 = Create Widget(Class=/FortUILibrary/Systems/ModularButton/Blocks/WBP_UIKit_Block_Text.WBP_UIKit_Block_Text_C, OwningPlayer=GetOwningPlayer()) -> (ReturnValue:WBP UIKit Block Text Object Reference)
   > n2 = Set Block_SecondaryText(Block_SecondaryText:WBP UIKit Block Text Object Reference=@n1) -> (Output_Get:WBP UIKit Block Text Object Reference)
   > RegisterModularBlock(Block=@n2)
   > n3 = AddChildToVerticalBox(self=Get Texts() -> (Texts:Vertical Box Object Reference), Content=@n2)
   > n4 = Set CurrentVerticalSlot(CurrentVerticalSlot:Vertical Box Slot Object Reference=@n3) -> (Output_Get:Vertical Box Slot Object Reference)
   > Set HorizontalAlignment!(execute, HorizontalAlignment:EHorizontalAlignment Enum=HAlign_Center, self:Vertical Box Slot Object Reference=@n4) -> (then, Output_Get:EHorizontalAlignment Enum=HAlign_Fill)
   > Set VerticalAlignment!(execute, VerticalAlignment:EVerticalAlignment Enum=VAlign_Top, self:Vertical Box Slot Object Reference=@n4) -> (then, Output_Get:EVerticalAlignment Enum=VAlign_Fill)
   > Set Padding(Padding=Get TextSecondPadding() -> (TextSecondPadding:Margin Structure), self:Vertical Box Slot Object Reference=@n4)
   > SetBlockTiming(
    self=@n2
    Timing_Hovering=0.15
    Timing_Unhovering=0.1
    Timing_Focusing=0.2
    Timing_Unfocusing=0.15
    Timing_Pressing=0.1
    Timing_Releasing=0.1
    Timing_Disabling=0.1
    Timing_Enabling=0.1
    Timing_Selecting=0
    Timing_DeselectingIdle=0
    Timing_DeselectingFocused=0
   )
   > n5 = Branch(Condition=Get OverrideSecondaryTextSize() -> (OverrideSecondaryTextSize:bool))
   n6 = Select(
    NewEnumerator0:"Slate Font Info Structure (by ref)"=Get LargeSecondaryFont() -> (LargeSecondaryFont:Slate Font Info Structure)
    NewEnumerator1:"Slate Font Info Structure (by ref)"=Get MeduimSecondaryFont() -> (MeduimSecondaryFont:Slate Font Info Structure)
    NewEnumerator2:"Slate Font Info Structure (by ref)"=Get SmallSecondaryFont() -> (SmallSecondaryFont:Slate Font Info Structure)
    Index=Get ButtonSize() -> (ButtonSize=NewEnumerator0)
   ) -> (ReturnValue:"Slate Font Info Structure (by ref)")
   n7 = Break Slate Font Info!(SlateFontInfo:"Slate Font Info Structure (by ref)"=@n6) -> (
    FontObject:Object Reference
    FontMaterial:Object Reference
    OutlineSettings:Font Outline Settings Structure
    TypefaceFontName:Name
    Size:float=~
    LetterSpacing:int
    SkewAmount:float=~
    bForceMonospaced:bool=~
    bMaterialIsStencil:bool=~
    MonospacedWidth:float=~
   ) {is=title K2Node_BreakStruct}
   n8 = Break Slate Font Info!(SlateFontInfo=Select(
     NewEnumerator0=Get LargePrimaryFont() -> (LargePrimaryFont:Slate Font Info Structure)
     NewEnumerator1=Get MediumPrimaryFont() -> (MediumPrimaryFont:Slate Font Info Structure)
     NewEnumerator2=Get SmallPrimaryFont() -> (SmallPrimaryFont:Slate Font Info Structure)
     Index=Get ButtonSize() -> (ButtonSize=NewEnumerator0)
    )) -> (Size=~, bMaterialIsStencil=~)
   n9 = Get OverrideSecondaryTextSizeMultiplier() -> (OverrideSecondaryTextSizeMultiplier:double)
   n10 = Get Block_SecondaryText() -> (Block_SecondaryText:WBP UIKit Block Text Object Reference)
   > SetFont!(
    execute
    self:WBP UIKit Block Text Object Reference=@n10
    Font:"Slate Font Info Structure (by ref)"=MakeSlateFontInfo!(
     FontObject:Object Reference=@n7
     FontMaterial:Object Reference=@n7.FontMaterial
     OutlineSettings:Font Outline Settings Structure=@n7.OutlineSettings
     TypefaceFontName:Name=@n7.TypefaceFontName
     Size:float=float * float(A=@n8, B=@n9) {is=title K2Node_PromotableOperator}
     LetterSpacing:int=@n7.LetterSpacing
     SkewAmount:float=@n7.SkewAmount
     bForceMonospaced:bool=@n7.bForceMonospaced
     bMaterialIsStencil:bool=false
     MonospacedWidth:float=@n7.MonospacedWidth
    ) -> (ReturnValue:Slate Font Info Structure)
   ) -> (then)
   n11 = Get MobileFontSizeMultiplierLarge() -> (MobileFontSizeMultiplierLarge:double)
   n12 = Get MobileFontSizeMultiplierMedium() -> (MobileFontSizeMultiplierMedium:double)
   n13 = Get MobileFontSizeMultiplierSmall() -> (MobileFontSizeMultiplierSmall:double)
   n14 = Get ButtonSize() -> (ButtonSize=NewEnumerator0)
   > SetMobileFontSizeMultiplier!(
    execute
    self:WBP UIKit Block Text Object Reference=@n10
    MobileFontSizeMultiplier:double=Select!(
     Option 0:double=Select!(NewEnumerator0:double=@n11, NewEnumerator1:double=@n12, NewEnumerator2:double=@n13, Index:E_UI_ButtonSize Enum=@n14) -> (ReturnValue)
     Option 1:double=Get MobileFontSizeMultiplierSecondaryOverride() -> (MobileFontSizeMultiplierSecondaryOverride:double)
     Index:bool=Get OverrideDefaultMobileFontSizeMultiplier() -> (OverrideDefaultMobileFontSizeMultiplier:bool)
    ) -> (ReturnValue:double)
   ) -> (then)
   n15 = Get Block_SecondaryText()
   SetFont(execute=@n5.else, self=@n15, Font=@n6)
   n16 = Get ButtonSize() -> (ButtonSize=NewEnumerator0)
   n17 = Get MobileFontSizeMultiplierSecondaryLarge() -> (MobileFontSizeMultiplierSecondaryLarge:double)
   n18 = Get MobileFontSizeMultiplierSecondarySmall() -> (MobileFontSizeMultiplierSecondarySmall:double)
   n19 = Get MobileFontSizeMultiplierSecondarySmall()
   > SetMobileFontSizeMultiplier(
    self=@n15
    MobileFontSizeMultiplier=Select!(
     Option 0:float=Select!(NewEnumerator0:double=@n17, NewEnumerator1:double=@n18, NewEnumerator2:double=@n19, Index:E_UI_ButtonSize Enum=@n16) -> (ReturnValue:double)
     Option 1:float=Get MobileFontSizeMultiplierSecondaryOverride()
     Index:bool=Get OverrideDefaultMobileFontSizeMultiplier()
    ) -> (ReturnValue:float)
   )
  ]}
 EventGraph {nodes [
   n1 = EventOnFocusedManually() {is=title K2Node_CustomEvent}
   ShowPromptFallbackAction!(execute=[@n1.then, @n13.then], self, Show:bool=true) -> (then)
   > RequestTransitionOnBlocks!(execute, self:UIKit Modular Button Object Reference, State:EUIKitBlockVisualState Enum=Focused) -> (then)
   > PlaySoundHovered()
   Switch Input() {is=title K2Node_CustomEvent}
   > UpdatePrompt!(execute, self, InputType:ECommonInputType Enum=Get CurrentInputType() -> (CurrentInputType:ECommonInputType Enum)) -> (then)
   > ResetButtonState()
   > FadeOutSound()
   n2 = Event PreConstruct()
   > n3 = Sequence!(execute) -> (then_0, then_1, then_2:Exec)
   > Branch(Condition=@n2.IsDesignTime)
   > PresetForDesignTime()
   UnregisterAllModularBlocks(execute=@n3.then_1, self:UIKit Modular Button Object Reference)
   n4 = Get Block_Text() -> (Block_Text:WBP UIKit Block Text Object Reference)
   > RegisterModularBlock(Block=@n4)
   > SetBlockTiming(
    self=@n4
    Timing_Hovering=0
    Timing_Unhovering=0
    Timing_Focusing=0.15
    Timing_Unfocusing=0.15
    Timing_Pressing=0.1
    Timing_Releasing=0.1
    Timing_Disabling=0.1
    Timing_Enabling=0.1
    Timing_Selecting=0
    Timing_DeselectingIdle=0
    Timing_DeselectingFocused=0
   )
   > SetText!(execute, self:Fort CTAButton Object Reference, ButtonText:"Text (by ref)"=Get Text!() -> (Text)) -> (then)
   > SetSecondaryText!(execute, self:Fort CTAButton Object Reference, ButtonSecondaryText:"Text (by ref)"=Get SecondaryText() -> (SecondaryText:Text)) -> (then)
   > InitIcon()
   n5 = Get Block_Background() -> (Block_Background:WBP UIKit Block Background With Outline Object Reference)
   > RegisterModularBlock(Block=@n5)
   > SetBlockTiming(
    self=@n5
    Timing_Hovering=0
    Timing_Unhovering=0
    Timing_Focusing=0.3
    Timing_Unfocusing=0.15
    Timing_Pressing=0.1
    Timing_Releasing=0.1
    Timing_Disabling=0.1
    Timing_Enabling=0.1
    Timing_Selecting=0
    Timing_DeselectingIdle=0
    Timing_DeselectingFocused=0
   )
   InitializeButton(execute=@n3.then_2)
   > ResetButtonState()
   > Set EnableOnWidgetStyled(EnableOnWidgetStyled:bool=true) -> (Output_Get:bool)
   Event On Pressed() {is=title K2Node_Event}
   > RequestTransitionOnBlocks(State=Pressed)
   > PlayAnimationForward(InAnimation=Get Anim_Pressed_Base() -> (Anim_Pressed_Base:Widget Animation Object Reference))
   > n6 = Branch(Condition=ShouldPlayHoldSound!() -> (Holdable:bool))
   > PlaySoundHoldPressed()
   PlaySoundPressed(execute=@n6.else)
   n7 = Event OnActionProgress!() -> (OutputDelegate, then, HeldPercent:float) {is=title K2Node_Event}
   > SetHoldPercentage!(execute, self, HeldPercentageValue:double=@n7.HeldPercent) -> (then)
   > Call On Hold Action Progress!(execute:Exec, self:WBP UIKit Button CTA Base Object Reference, HeldPercent:double=@n7.HeldPercent) -> (then:Exec) {is=title K2Node_CallDelegate}
   Event Construct()
   > SetPrompt()
   > Set Icon Dark Mode!(execute, self, IsDark:bool=Get IsDarkMode() -> (IsDarkMode:bool)) -> (then) {is=title K2Node_CallFunction}
   Event Destruct()
   > FadeOutSound()
   Event On Input Action Triggered() {is=title K2Node_Event}
   > PlaySoundPressed()
   Event On Widget Styled() {is=title K2Node_Event}
   > Branch(Condition=Get EnableOnWidgetStyled() -> (EnableOnWidgetStyled:bool))
   > InitializeButton()
   Event On Focused() {is=title K2Node_Event}
   > Call On Focus Changed!(execute:Exec, self:WBP UIKit Button CTA Base Object Reference, HasFocus:bool=true) -> (then:Exec) {is=title K2Node_CallDelegate}
   n8 = Event On Locked Changed!() -> (OutputDelegate, then, bIsLocked:bool) {is=title K2Node_Event}
   > Branch(Condition=Get UseDisabledVisualsOnLocked() -> (UseDisabledVisualsOnLocked:bool))
   > n9 = Set LastEnabledEvent(LastEnabledEvent:bool=NOT Boolean(A=@n8.bIsLocked)) -> (Output_Get:bool)
   DelayUntilNextTick!(execute=[@n12.then, @n9.then, @n14.then]) -> (then)
   > SetButtonIsEnabled()
   Event On Unfocused() {is=title K2Node_Event}
   > Call On Focus Changed()
   Event BP_OnInputActionUpdated() {is=title K2Node_Event}
   > UpdatePrompt(InputType=Get CurrentInputType())
   Event OnInitialized()
   > FORT-833205 RetrieveMinDimensions()
   Event On Released() {is=title K2Node_Event}
   > Call On Released!(execute:Exec, self:WBP UIKit Button CTA Base Object Reference) -> (then:Exec) {is=title K2Node_CallDelegate}
   > RequestTransitionOnBlocks(State=Released)
   > PlayAnimationReverse(InAnimation=Get Anim_Pressed_Base(), PlaybackSpeed=2)
   n10 = Get CurrentHoldProgress() -> (CurrentHoldProgress:double)
   > n11 = Branch(Condition=AND Boolean!(A=ShouldPlayHoldSound(), B="float > float"(A=@n10), C:bool="float < float"(A=@n10, B=1)) -> (ReturnValue))
   > StopSoundHoldReleased()
   FadeOutSound(execute=@n11.else)
   Event On Clicked()
   > Branch(Condition=IsHoldable!() -> (Holdable:bool))
   > SetHoldPercentage(HeldPercentageValue=0)
   > ResetButtonState()
   > PlaySoundHoldCompleted()
   > Call On Hold Completed!(execute:Exec, self:WBP UIKit Button CTA Base Object Reference) -> (then:Exec) {is=title K2Node_CallDelegate}
   Event On Requires Hold Changed() {is=title K2Node_Event}
   > UpdatePrompt(InputType=Get CurrentInputType())
   Event On Disabled() {is=title K2Node_Event}
   > n12 = Set LastEnabledEvent(LastEnabledEvent=false)
   n13 = Event On Hovered()
   Event On Unhovered()
   > ShowPromptFallbackAction(Show=false)
   > RequestTransitionOnBlocks(State=Unfocused)
   Event On Enabled() {is=title K2Node_Event}
   > n14 = Set LastEnabledEvent(LastEnabledEvent=true)
   n15 = Event BP_SetText!() -> (OutputDelegate, then, ButtonText:"Text (by ref)") {is=title K2Node_Event}
   > SetText_BP!(execute, self, Text:Text=@n15.ButtonText) -> (then)
   n16 = Event BP_SetSecondaryText!() -> (OutputDelegate, then, ButtonSecondaryText:"Text (by ref)") {is=title K2Node_Event}
   > SetSecondText_BP!(execute, self, SecondaryText:Text=@n16.ButtonSecondaryText) -> (then)
   n17 = Event On Input Method Changed!() -> (OutputDelegate, then, CurrentInputType:ECommonInputType Enum) {is=title K2Node_Event}
   > Set CurrentInputType(CurrentInputType:ECommonInputType Enum=@n17.CurrentInputType) -> (Output_Get:ECommonInputType Enum)
   > Set Timer by Event(Delegate=Create Event(), Time=Get InputSwitchDelay() -> (InputSwitchDelay:double), bLooping=false, bMaxOncePerFrame=true)
  ]}
 FORT-833205 RetrieveMinDimensions {nodes [
   FORT-833205 RetrieveMinDimensions() {is=title K2Node_FunctionEntry}
   n1 = Get MinHeightOverride() -> (MinHeightOverride:int=0)
   n2 = Get MinWidthOverride() -> (MinWidthOverride:int=0)
   > n3 = Branch(Condition=OR Boolean(A="integer > integer"(A=@n2), B="integer > integer"(A=@n1)))
   Set MinWidthOverride(execute=@n3.else, MinWidthOverride:int=Get MinWidth() -> (MinWidth:int=0)) -> (Output_Get:int=0)
   > Set MinHeightOverride(MinHeightOverride:int=Get MinHeight() -> (MinHeight:int=0)) -> (Output_Get:int=0)
   n4 = Get MobileMinHeight() -> (MobileMinHeight:int=0)
   n5 = "integer > integer"(A=@n4)
   n6 = "Equal (Integer)"(A=Get MinHeight() -> (MinHeight=0), B=@n4)
   n7 = Get MobileMinWidth() -> (MobileMinWidth:int=0)
   n8 = "integer > integer"(A=@n7)
   n9 = "Equal (Integer)"(A=Get MinWidth() -> (MinWidth=0), B=@n7)
   Branch(Condition=OR Boolean(A=AND Boolean!(A=@n9, B=@n8) -> (ReturnValue), B=AND Boolean(A=@n6, B=@n5)))
   > PrintString!(
    execute
    InString=There might be a corrupted values as MinWidth is the same as MobileMinWidth
    bPrintToScreen
    bPrintToLog
    TextColor=(R=1,G=0,B=0,A=1)
    Duration=4
    Key
   ) -> (then)
  ]}
 FadeOutSound {nodes [
   FadeOutSound() {is=title K2Node_FunctionEntry}
   > n1 = Sequence!(execute) -> (then_0, then_1)
   > n2 = Get!(execute:Exec) -> (then:Exec, else:Exec, PressedSoundAudioComponent:Audio Component Object Reference)
   > FadeOut!(
    execute
    self:Audio Component Object Reference=@n2
    FadeOutDuration:float=Get FadeOutTime() -> (FadeOutTime:double)
    FadeVolumeLevel:float=0
    FadeCurve:EAudioFaderCurve Enum=Linear
   ) -> (then)
   n3 = Get!(execute=@n1.then_1) -> (then, else, HoldSoundAudioComponent:Audio Component Object Reference)
   > FadeOut(self=@n3, FadeOutDuration=Get FadeOutTime(), FadeVolumeLevel=0, FadeCurve=Linear)
  ]}
 InitIcon {nodes [
   InitIcon() {is=title K2Node_FunctionEntry}
   > n1 = Get!(execute) -> (then, else, Block_Icon:WBP UIKit Block Image Object Reference)
   > n2 = RemoveBlock!(execute, self, Block:WBP UIKit Block Base Object Reference=@n1) -> (then)
   n3 = Get!(execute=[@n2.then, @n1.else]) -> (then, else, IconMaterial:Material Instance Object Reference)
   > AddIconBlock() {is=title K2Node_CallFunction}
   n4 = Get Block_Icon() -> (Block_Icon:WBP UIKit Block Image Object Reference)
   > SetMaterial!(execute, self:WBP UIKit Block Image Object Reference=@n4, InMaterial:"Material Instance Object Reference (by ref)"=@n3) -> (then)
   > n5 = Get!(execute) -> (then, else, Icon:Texture 2D Object Reference)
   > SetTextureWithParameter!(execute, self:WBP UIKit Block Image Object Reference=@n4, ParameterName:Name=Icon, InTexture:"Texture 2D Object Reference (by ref)"=@n5) -> (then)
   n6 = Get(execute=@n3.else)
   > AddIconBlock()
   > SetBrush(
    self:WBP UIKit Block Image Object Reference=Get Block_Icon()
    InBrush=Make Slate Brush!(ResourceObject:Object Reference=@n6) -> (SlateBrush:Slate Brush Structure) {is=title K2Node_MakeStruct}
   )
   Set HasIcon(execute=@n6.else, HasIcon=false)
   > Set Block_Icon()
  ]}
 InitializeButton {nodes [
   InitializeButton() {is=title K2Node_FunctionEntry}
   > UpdateFonts()
   > UpdateButtonSIze()
   > UpdateButtonPadding()
   > UpdateBackground()
   > UpdateAltPromptVisibility()
  ]}
 IsHoldable {nodes [
   IsHoldable() {is=title K2Node_FunctionEntry}
   n1 = Get InputActionWidget() -> (InputActionWidget:Common Action Widget Object Reference)
   > Return Node!(
    execute
    Holdable:bool=OR Boolean(A=Get bRequiresHold() -> (bRequiresHold:bool), B=IsHeldAction!(self:Common Action Widget Object Reference=@n1) -> (ReturnValue:bool))
   )
  ]}
 IsPromptIcon {nodes [
   IsPromptIcon() {is=title K2Node_FunctionEntry}
   n1 = GetIcon!(self:Common Action Widget Object Reference=Get InputActionWidget()) -> (ReturnValue:Slate Brush Structure)
   > Return Node!(
    execute
    Valid:bool=IsValid(Object=Break Slate Brush!(SlateBrush:"Slate Brush Structure (by ref)"=@n1) -> (
      TintColor:Slate Color Structure
      DrawAs:ESlateBrushDrawType Enum
      Tiling:ESlateBrushTileType Enum
      Mirroring:ESlateBrushMirrorType Enum
      ImageSize:Vector 2D Structure
      Margin:Margin Structure
      ResourceObject:Object Reference
      OutlineSettings:Slate Brush Outline Settings Structure
     ) {is=title K2Node_BreakStruct}.ResourceObject)
   )
  ]}
 OnFocusChanged {nodes [OnFocusChanged!() -> (then, HasFocus:bool=~) {is=title K2Node_FunctionEntry}]}
 OnHoldActionProgress {nodes [OnHoldActionProgress!() -> (then, HeldPercent:double=~) {is=title K2Node_FunctionEntry}]}
 OnHoldCompleted {nodes [OnHoldCompleted() {is=title K2Node_FunctionEntry}]}
 OnReleased {nodes [OnReleased() {is=title K2Node_FunctionEntry}]}
 PlaySoundHoldCompleted {nodes [
   PlaySoundHoldCompleted() {is=title K2Node_FunctionEntry}
   > n1 = Sequence()
   > n2 = Get!(execute) -> (then, else, HoldCompletedSound:Sound Base Object Reference)
   > SpawnSound2D!(
    execute
    Sound:Sound Base Object Reference=@n2
    VolumeMultiplier:float=1
    PitchMultiplier:float=1
    StartTime:float=0
    ConcurrencySettings:Sound Concurrency Object Reference
    bPersistAcrossLevelTransition:bool=true
    bAutoDestroy:bool=true
   ) -> (then, ReturnValue:Audio Component Object Reference)
   FadeOutSound(execute=@n1.then_1) {is=title K2Node_CallFunction}
  ]}
 PlaySoundHoldPressed {nodes [
   PlaySoundHoldPressed() {is=title K2Node_FunctionEntry}
   > n1 = Sequence()
   > n2 = Get!(execute) -> (then, else, HoldPressedSound:Sound Base Object Reference)
   > n3 = SpawnSound2D(Sound=@n2, VolumeMultiplier=1, PitchMultiplier=1, StartTime=0, bPersistAcrossLevelTransition=false, bAutoDestroy=true)
   > Set PressedSoundAudioComponent(PressedSoundAudioComponent:Audio Component Object Reference=@n3) -> (Output_Get:Audio Component Object Reference)
   n4 = Get!(execute=@n1.then_1) -> (then, else, HoldSound:Sound Base Object Reference)
   n5 = Get HoldSoundAudioComponent() -> (HoldSoundAudioComponent:Audio Component Object Reference)
   > n6 = Branch(Condition=IsValid(Object=@n5))
   > n7 = Branch(Condition=IsPlaying!(self:Audio Component Object Reference=@n5) -> (ReturnValue:bool))
   > n8 = Stop(self:Audio Component Object Reference=@n5)
   n9 = SpawnSound2D(
    execute=[@n8.then, @n7.else, @n6.else]
    Sound=@n4
    VolumeMultiplier=1
    PitchMultiplier=1
    StartTime=0
    bPersistAcrossLevelTransition=false
    bAutoDestroy=true
   )
   > Set HoldSoundAudioComponent(HoldSoundAudioComponent:Audio Component Object Reference=@n9) -> (Output_Get:Audio Component Object Reference)
  ]}
 PlaySoundHovered {nodes [
   PlaySoundHovered() {is=title K2Node_FunctionEntry}
   > n1 = Get!(execute) -> (then, else, HoveredSound:Sound Base Object Reference)
   > SpawnSound2D(Sound=@n1, VolumeMultiplier=1, PitchMultiplier=1, StartTime=0, bPersistAcrossLevelTransition=false, bAutoDestroy=true)
  ]}
 PlaySoundPressed {nodes [
   PlaySoundPressed() {is=title K2Node_FunctionEntry}
   > n1 = Get!(execute) -> (then, else, PressedSound:Sound Base Object Reference)
   > SpawnSound2D(Sound=@n1, VolumeMultiplier=1, PitchMultiplier=1, StartTime=0, bPersistAcrossLevelTransition=false, bAutoDestroy=true)
  ]}
 PresetForDesignTime {nodes [
   PresetForDesignTime() {is=title K2Node_FunctionEntry}
   > UpdateAltPromptVisibility()
   n1 = Get BelowTextSlot() -> (BelowTextSlot:Named Slot Object Reference)
   > SetVisibility(
    self=@n1
    InVisibility=Select!(
     Option 0:ESlateVisibility Enum=Collapsed
     Option 1:ESlateVisibility Enum=SelfHitTestInvisible
     Index:bool=HasAnyChildren!(self:Panel Widget Object Reference=@n1) -> (ReturnValue:bool)
    ) -> (ReturnValue:ESlateVisibility Enum)
   )
   n2 = GetInputAction!(self:Common Button Base Object Reference) -> (InputActionRow:Data Table Row Handle Structure, ReturnValue:bool)
   > n3 = Branch(Condition=OR Boolean(A=IsHoldable() {is=title K2Node_CallFunction}, B="Not Equal (Boolean)"(A=@n2.ReturnValue) -> (ReturnValue=~)))
   SetVisibility(execute=@n3.else, self=Get Prompt() -> (Prompt:Grid Panel Object Reference), InVisibility=Collapsed)
  ]}
 RemoveBlock {nodes [
   n1 = RemoveBlock!() -> (then, Block:WBP UIKit Block Base Object Reference) {is=title K2Node_FunctionEntry}
   > UnregisterModularBlock!(execute, self:UIKit Modular Button Object Reference, Block:UIKit Block Object Reference=@n1) -> (then)
   > RemoveFromParent(self=@n1)
  ]}
 ResetButtonState {nodes [
   ResetButtonState() {is=title K2Node_FunctionEntry}
   > n1 = Branch(Condition=IsAnyAnimationPlaying!(self:User Widget Object Reference) -> (ReturnValue:bool))
   > n2 = StopAllAnimations(self:User Widget Object Reference)
   n3 = OR Boolean(
    A=NOT Boolean(A=GetIsEnabled!(self:Widget Object Reference) -> (ReturnValue:bool))
    B=AND Boolean(A=Get UseDisabledVisualsOnLocked(), B=GetLocked!(self:Common Button Base Object Reference) -> (ReturnValue:bool))
   )
   RequestInstantTransitionOnBlocks!(
    execute=[@n2.then, @n1.else]
    self:UIKit Modular Button Object Reference
    State:EUIKitBlockInstantTransitionState Enum=Select(Option 0:EUIKitBlockInstantTransitionState Enum=InstantEnabled, Option 1:EUIKitBlockInstantTransitionState Enum=InstantDisabled, Index=@n3) -> (ReturnValue:EUIKitBlockInstantTransitionState Enum)
   ) -> (then)
   > Set RenderOpacity!(
    execute
    RenderOpacity:float=Select(Option 0:float=1, Option 1:float=0.3, Index=@n3) -> (ReturnValue:float)
    self:Widget Object Reference=Get Prompt()
   ) -> (then, Output_Get:float)
  ]}
 Set Icon Dark Mode {nodes [
   Set Icon Dark Mode!() -> (then, IsDark:bool=~) {is=title K2Node_FunctionEntry}
   > n1 = GetIconDynamicMaterial!(execute, self:Common Action Widget Object Reference=Get InputActionWidget()) -> (then, ReturnValue:Material Instance Dynamic Object Reference)
   > SetScalarParameterValue(self=@n1, ParameterName=Use Dark, Value="To Float (Boolean)"(InBool=Get IsDark() -> (IsDark:bool)))
   > SetScalarParameterValue(self=@n1, ParameterName=Outline Opacity, Value=Select(Option 0=0.5, Option 1=0, Index=Get IsDark()))
  ]}
 SetBackgroundMaterial {nodes [
   n1 = SetBackgroundMaterial!() -> (then, Background:Material Instance Object Reference) {is=title K2Node_FunctionEntry}
   > n2 = Set BackgroundMaterial(BackgroundMaterial:Material Instance Object Reference=@n1) -> (Output_Get:Material Instance Object Reference)
   > SetBackgroundMaterial!(
    execute
    self:WBP UIKit Block Background Base Object Reference=Get Block_Background()
    Material Instance:"Material Instance Object Reference (by ref)"=@n2
   ) -> (then) {is=title K2Node_CallFunction}
  ]}
 SetBorderMaterial {nodes [
   n1 = SetBorderMaterial!() -> (then, BorderMaterial:Material Instance Object Reference, IsAdditive:bool=~) {is=title K2Node_FunctionEntry}
   > n2 = Set BorderMaterial(BorderMaterial:Material Instance Object Reference=@n1) -> (Output_Get:Material Instance Object Reference)
   > SetOutlineMaterial!(
    execute
    self:WBP UIKit Block Background Base Object Reference=Get Block_Background()
    Material Instance:"Material Instance Object Reference (by ref)"=@n2
   ) -> (then)
  ]}
 SetButtonIsEnabled {nodes [
   SetButtonIsEnabled() {is=title K2Node_FunctionEntry}
   > n1 = Branch(Condition=Get LastEnabledEvent() -> (LastEnabledEvent:bool))
   > n2 = Branch(Condition=IsHovered())
   > n3 = RequestTransitionOnBlocks(State=Enabled)
   SetRenderOpacity(execute=[@n3.then, @n4.then], self=Get Prompt(), InOpacity=1)
   n4 = RequestInstantTransitionOnBlocks(execute=@n2.else, State=InstantEnabled)
   RequestTransitionOnBlocks(execute=@n1.else, State=Disabled)
   > SetRenderOpacity(self=Get Prompt(), InOpacity=0.3)
  ]}
 SetButtonSize {nodes [
   n1 = SetButtonSize!() -> (then, ButtonSize:E_UI_ButtonSize Enum=NewEnumerator0) {is=title K2Node_FunctionEntry}
   > Set ButtonSize(ButtonSize:E_UI_ButtonSize Enum=@n1) -> (Output_Get:E_UI_ButtonSize Enum=NewEnumerator0)
   > UpdateFonts()
   > UpdateButtonSIze()
   > UpdateButtonPadding()
  ]}
 SetHoldPercentage {nodes [
   n1 = SetHoldPercentage!() -> (then, HeldPercentageValue:double=~) {is=title K2Node_FunctionEntry}
   > n2 = Set CurrentHoldProgress(CurrentHoldProgress:double=@n1) -> (Output_Get:double)
   > SetScalarParameterValue(self=Get HoldProgressMID() -> (HoldProgressMID:Material Instance Dynamic Object Reference), ParameterName=Percentage, Value=@n2)
  ]}
 SetIcon {nodes [
   n1 = SetIcon!() -> (then, Icon:Object Reference) {is=title K2Node_FunctionEntry}
   > n2 = Branch(Condition=IsValid(Object=@n1))
   > n3 = Cast To MaterialInstance!(execute, Object=@n1) -> (then, CastFailed, AsMaterial Instance:Material Instance Object Reference) {is=title K2Node_DynamicCast}
   > n4 = Set IconMaterial(IconMaterial:Material Instance Object Reference=@n3) -> (Output_Get:Material Instance Object Reference)
   > n5 = Get!(execute) -> (then, else, Block_Icon:WBP UIKit Block Image Object Reference)
   SetMaterial(execute=[@n6.then, @n5.then], self=@n5, InMaterial=@n4)
   n6 = AddIconBlock(execute=@n5.else)
   n7 = Cast To Texture2D(execute=@n3.CastFailed, Object=@n1)
   > n8 = Set Icon(Icon:Texture 2D Object Reference=@n7) -> (Output_Get:Texture 2D Object Reference)
   > n9 = Get()
   n10 = Get!(execute=[@n11.then, @n9.then]) -> (then, else, IconMaterial:Material Instance Object Reference)
   > SetTextureWithParameter(self=@n9, ParameterName=Icon, InTexture=@n8)
   SetBrush(execute=@n10.else, self=@n9, InBrush=Make Slate Brush(ResourceObject=@n8))
   AddIconBlock(execute=@n9.else)
   > n11 = SetMaterial(self=@n9, InMaterial=Get IconMaterial() -> (IconMaterial:Material Instance Object Reference))
   n12 = Self-Reference()
   PrintString(
    execute=@n7.CastFailed
    InString=Append!(A="WARNING: ", B=GetDisplayName(Object=@n12), C:String=" - Incorrect Icon type in SetIcon fucntion.") -> (ReturnValue)
    TextColor=(R=1,G=0,B=0.5683,A=1)
    Duration=8
   )
   Set HasIcon(execute=@n2.else, HasIcon=false)
   > Set Icon()
   > n13 = Get!(execute) -> (then, else, Block_Icon:WBP UIKit Block Image Object Reference)
   > RemoveBlock(Block=@n13) {is=title K2Node_CallFunction}
   > Set Block_Icon()
  ]}
 SetIconMaterial {nodes [
   n1 = SetIconMaterial!() -> (then, IconMaterial:Material Instance Object Reference) {is=title K2Node_FunctionEntry}
   > n2 = Set IconMaterial(IconMaterial=@n1)
   > n3 = Branch(Condition=IsValid(Object=@n2))
   > n4 = Get()
   SetMaterial(execute=[@n5.then, @n4.then], self=Get Block_Icon(), InMaterial=@n2)
   n5 = AddIconBlock(execute=@n4.else)
   Set HasIcon(execute=@n3.else, HasIcon=false)
   > n6 = Get()
   > RemoveBlock(Block=@n6)
   > Set Block_Icon()
  ]}
 SetPrompt {nodes [
   SetPrompt() {is=title K2Node_FunctionEntry}
   n1 = Get CommonInputSubsystem!() -> (ReturnValue:Common Input Object Reference) {is=title K2Node_GetBlueprintContextEx}
   > Branch(Condition=IsValid(Object=@n1))
   > n2 = GetDynamicMaterial(self=Get HoldProgressBar() -> (HoldProgressBar:Image Object Reference))
   > Set HoldProgressMID(HoldProgressMID:Material Instance Dynamic Object Reference=@n2) -> (Output_Get:Material Instance Dynamic Object Reference)
   > n3 = Set CurrentInputType(CurrentInputType=GetCurrentInputType(self=@n1))
   > UpdatePrompt(InputType=@n3)
  ]}
 SetSecondText_BP {nodes [
   n1 = SetSecondText_BP!() -> (then, SecondaryText:Text) {is=title K2Node_FunctionEntry}
   n2 = TextIsEmpty(InText=@n1)
   > n3 = Branch(Condition=AND Boolean(A=NOT Boolean(A=@n2), B=Get HasText()))
   > n4 = Get!(execute) -> (then, else, Block_SecondaryText:WBP UIKit Block Text Object Reference)
   SetText!(execute=[@n5.then, @n4.then], self:WBP UIKit Block Text Object Reference=@n4, Text:Text=@n1) -> (then)
   n5 = AddSecondTextBlock(execute=@n4.else) {is=title K2Node_CallFunction}
   n6 = Get(execute=@n3.else)
   > RemoveBlock(Block=@n6)
   > Set Block_SecondaryText()
  ]}
 SetText_BP {nodes [
   n1 = SetText_BP!() -> (then, Text:Text) {is=title K2Node_FunctionEntry}
   > n2 = Set HasText(HasText:bool=NOT Boolean(A=TextIsEmpty(InText=@n1))) -> (Output_Get:bool)
   > SetVisibility(
    self=Get Texts()
    InVisibility=Select(Option 0:ESlateVisibility Enum=Collapsed, Option 1:ESlateVisibility Enum=SelfHitTestInvisible, Index=@n2) -> (ReturnValue:ESlateVisibility Enum)
   )
   > Branch(Condition=@n2)
   > SetText(self=Get Block_Text(), Text=@n1)
  ]}
 ShouldPlayHoldSound {nodes [
   ShouldPlayHoldSound() {is=title K2Node_FunctionEntry}
   > n1 = Branch(Condition=Get bRequiresHold())
   > Return Node!(execute, Holdable:bool=true)
   n2 = Branch(
    execute=@n1.else
    Condition="Not Equal (Enum)"!(A:ECommonInputType Enum=Get CurrentInputType(), B:ECommonInputType Enum=Gamepad) -> (ReturnValue:bool)
   )
   n3 = IsHeldAction(self=Get InputActionWidget())
   > n4 = Branch(Condition=@n3)
   > Return Node(Holdable=NOT Boolean(A=IsHovered()))
   Return Node(execute=@n4.else, Holdable=false)
   Return Node(execute=@n2.else, Holdable=@n3)
  ]}
 ShowInputActionText {nodes [
   ShowInputActionText() {is=title K2Node_FunctionEntry}
   > SetText_BP(Text=GetDisplayText!(self:Common Action Widget Object Reference=Get InputActionWidget()) -> (ReturnValue:Text)) {is=title K2Node_CallFunction}
  ]}
 ShowPromptFallbackAction {nodes [
   n1 = ShowPromptFallbackAction!() -> (then, Show:bool=~) {is=title K2Node_FunctionEntry}
   n2 = Get CurrentInputType()
   > Branch(Condition=AND Boolean(A="Equal (Enum)"(A=@n2, B=Gamepad), B=Get bShouldUseFallbackDefaultInputAction() -> (bShouldUseFallbackDefaultInputAction:bool)))
   > SetVisibility(self=Get Prompt(), InVisibility=Select(Option 0=Collapsed, Option 1=SelfHitTestInvisible, Index=@n1))
  ]}
 StopSoundHoldReleased {nodes [
   StopSoundHoldReleased() {is=title K2Node_FunctionEntry}
   > n1 = Sequence()
   > n2 = Get!(execute) -> (then, else, HoldReleasedSound:Sound Base Object Reference)
   > SpawnSound2D(Sound=@n2, VolumeMultiplier=1, PitchMultiplier=1, StartTime=0, bPersistAcrossLevelTransition=false, bAutoDestroy=true)
   FadeOutSound(execute=@n1.then_1)
  ]}
 UpdateAltPromptVisibility {nodes [
   UpdateAltPromptVisibility() {is=title K2Node_FunctionEntry}
   > n1 = Branch(Condition=HasAnyChildren(self=Get Slot_AltPrompt() -> (Slot_AltPrompt:Named Slot Object Reference)))
   > SetVisibility(self=Get Slot_AltPrompt(), InVisibility=SelfHitTestInvisible)
   > SetVisibility(self=Get Prompt(), InVisibility=Collapsed)
   SetVisibility(execute=@n1.else, self=Get Slot_AltPrompt(), InVisibility=Collapsed)
  ]}
 UpdateBackground {nodes [
   UpdateBackground() {is=title K2Node_FunctionEntry}
   n1 = Get Block_Background()
   > SetBackgroundMaterial(self=@n1, Material Instance=Get BackgroundMaterial() -> (BackgroundMaterial:Material Instance Object Reference))
   > SetOutlineMaterial(self=@n1, Material Instance=Get BorderMaterial() -> (BorderMaterial:Material Instance Object Reference))
   > SetCornerRadiuses!(
    execute
    self:WBP UIKit Block Background With Outline Object Reference=@n1
    TopLeft:float=0.325
    TopRight:float=0.325
    BottomLeft:float=0.325
    BottomRight:float=0.325
    RadiusInPixels:bool=false
   ) -> (then)
  ]}
 UpdateButtonPadding {nodes [
   UpdateButtonPadding() {is=title K2Node_FunctionEntry}
   n1 = NOT Boolean(A=Get HasText())
   n2 = Get HasIcon() -> (HasIcon:bool)
   > SetSize!(
    execute
    self:Spacer Object Reference=[Get Spacer_BottomRight() -> (Spacer_BottomRight:Spacer Object Reference), Get Spacer_TopLeft() -> (Spacer_TopLeft:Spacer Object Reference)]
    InSize_X:double=SelectFloat(A=0, B=24, bPickA=AND Boolean(A=@n2, B=@n1))
    InSize_Y:double=Select!(
     NewEnumerator0:float=8
     NewEnumerator1:float=4
     NewEnumerator2:float=0
     Index:E_UI_ButtonSize Enum=Get ButtonSize() -> (ButtonSize=NewEnumerator0)
    ) -> (ReturnValue:float)
   ) -> (then)
  ]}
 UpdateButtonSIze {nodes [
   UpdateButtonSIze() {is=title K2Node_FunctionEntry}
   n1 = Get Block_Icon()
   > n2 = Branch(Condition=IsValid(Object=@n1))
   > n3 = SetSize!(
    execute
    self:WBP UIKit Block Image Object Reference=@n1
    Desired Size:Vector 2D Structure=Select(
     NewEnumerator0:Vector 2D Structure=Get LargeIconSize()
     NewEnumerator1:Vector 2D Structure=Get MediumIconSize()
     NewEnumerator2:Vector 2D Structure=Get SmallIconSize()
     Index=Get ButtonSize() -> (ButtonSize=NewEnumerator0)
    ) -> (ReturnValue:Vector 2D Structure)
   ) -> (then)
   n4 = Branch(execute=[@n3.then, @n2.else], Condition=IsUsingCompactUI())
   n5 = Select!(
    Option 0:int=Select!(
     NewEnumerator0:int=Get MobileLargeHeight() -> (MobileLargeHeight:int=0)
     NewEnumerator1:int=Get MobileMediumHeight() -> (MobileMediumHeight:int=0)
     NewEnumerator2:int=Get MobileSmallHeight() -> (MobileSmallHeight:int=0)
     Index:E_UI_ButtonSize Enum=Get ButtonSize() -> (ButtonSize=NewEnumerator0)
    ) -> (ReturnValue)
    Option 1:int=Get MobileMinHeight() -> (MobileMinHeight=0)
    Index:bool=Get OverrideDefaultHeight() -> (OverrideDefaultHeight:bool)
   ) -> (ReturnValue:int)
   > SetMinDimensions!(
    execute
    self:Common Button Base Object Reference
    InMinWidth:int=SelectInt(A=Get MobileMinWidth() -> (MobileMinWidth=0), B=@n5, bPickA=Get HasText())
    InMinHeight:int=@n5
   ) -> (then)
   n6 = Select(
    NewEnumerator0=Get LargeHeight() -> (LargeHeight:int=0)
    NewEnumerator1=Get MediumHeight() -> (MediumHeight:int=0)
    NewEnumerator2=Get SmallHeight() -> (SmallHeight:int=0)
    Index=Get ButtonSize() -> (ButtonSize=NewEnumerator0)
   )
   SetMinDimensions(execute=@n4.else, InMinWidth=SelectInt(A=Get MinWidthOverride() -> (MinWidthOverride=0), B=@n6, bPickA=Get HasText()), InMinHeight=@n6)
  ]}
 UpdateFonts {nodes [
   UpdateFonts() {is=title K2Node_FunctionEntry}
   > n1 = Sequence()
   Get!(execute=@n1.then_1) -> (then, else, Block_SecondaryText:WBP UIKit Block Text Object Reference)
   > n2 = Branch(Condition=Get OverrideSecondaryTextSize())
   n3 = Select(
    NewEnumerator0:"Slate Font Info Structure (by ref)"=Get LargeSecondaryFont()
    NewEnumerator1:"Slate Font Info Structure (by ref)"=Get MeduimSecondaryFont()
    NewEnumerator2:"Slate Font Info Structure (by ref)"=Get SmallSecondaryFont()
    Index=Get ButtonSize() -> (ButtonSize=NewEnumerator0)
   ) -> (ReturnValue:"Slate Font Info Structure (by ref)")
   n4 = Break Slate Font Info!(SlateFontInfo=@n3) -> (
    FontObject:Object Reference
    FontMaterial:Object Reference
    OutlineSettings:Font Outline Settings Structure
    TypefaceFontName:Name
    Size=~
    LetterSpacing:int
    SkewAmount:float=~
    bForceMonospaced:bool=~
    bMaterialIsStencil=~
    MonospacedWidth:float=~
   )
   n5 = Select(
    NewEnumerator0=Get LargePrimaryFont()
    NewEnumerator1=Get MediumPrimaryFont()
    NewEnumerator2=Get SmallPrimaryFont()
    Index=Get ButtonSize() -> (ButtonSize=NewEnumerator0)
   )
   n6 = Break Slate Font Info(SlateFontInfo=@n5) -> (Size=~, SkewAmount=~, bForceMonospaced=~, bMaterialIsStencil=~, MonospacedWidth=~)
   n7 = Get OverrideSecondaryTextSizeMultiplier()
   n8 = Get Block_SecondaryText()
   > SetFont(
    self=@n8
    Font=MakeSlateFontInfo(
     FontObject=@n4
     FontMaterial=@n4.FontMaterial
     OutlineSettings=@n4.OutlineSettings
     TypefaceFontName=@n4.TypefaceFontName
     Size=float * float(A=@n6.Size, B=@n7)
     LetterSpacing=@n4.LetterSpacing
     SkewAmount=@n4.SkewAmount
     bForceMonospaced=@n4.bForceMonospaced
     bMaterialIsStencil=false
     MonospacedWidth=@n4.MonospacedWidth
    )
   )
   n9 = Get MobileFontSizeMultiplierLarge()
   n10 = Get MobileFontSizeMultiplierMedium()
   n11 = Get MobileFontSizeMultiplierSmall()
   n12 = Get ButtonSize() -> (ButtonSize=NewEnumerator0)
   > SetMobileFontSizeMultiplier(
    self=@n8
    MobileFontSizeMultiplier=Select!(
     Option 0:double=Select!(NewEnumerator0:double=@n9, NewEnumerator1:double=@n10, NewEnumerator2:double=@n11, Index:E_UI_ButtonSize Enum=@n12) -> (ReturnValue)
     Option 1:double=Get MobileFontSizeMultiplierSecondaryOverride()
     Index:bool=Get OverrideDefaultMobileFontSizeMultiplier()
    ) -> (ReturnValue:double)
   )
   n13 = Get Block_SecondaryText()
   SetFont(execute=@n2.else, self=@n13, Font=@n3)
   n14 = Get ButtonSize() -> (ButtonSize=NewEnumerator0)
   n15 = Get MobileFontSizeMultiplierSecondaryLarge()
   n16 = Get MobileFontSizeMultiplierSecondarySmall()
   n17 = Get MobileFontSizeMultiplierSecondarySmall()
   > SetMobileFontSizeMultiplier(
    self=@n13
    MobileFontSizeMultiplier=Select!(
     Option 0:float=Select!(NewEnumerator0:double=@n15, NewEnumerator1:double=@n16, NewEnumerator2:double=@n17, Index:E_UI_ButtonSize Enum=@n14) -> (ReturnValue:double)
     Option 1:float=Get MobileFontSizeMultiplierSecondaryOverride()
     Index:bool=Get OverrideDefaultMobileFontSizeMultiplier()
    ) -> (ReturnValue:float)
   )
   n18 = Get Block_Text()
   SetFont(execute=@n1.then_0, self=@n18, Font=@n5)
   n19 = Get ButtonSize() -> (ButtonSize=NewEnumerator0)
   n20 = Get MobileFontSizeMultiplierLarge()
   n21 = Get MobileFontSizeMultiplierMedium()
   n22 = Get MobileFontSizeMultiplierSmall()
   > SetMobileFontSizeMultiplier(
    self=@n18
    MobileFontSizeMultiplier=Select!(
     Option 0:double=Select!(NewEnumerator0:double=@n20, NewEnumerator1:double=@n21, NewEnumerator2:double=@n22, Index:E_UI_ButtonSize Enum=@n19) -> (ReturnValue)
     Option 1:double=Get MobileFontSizeMultiplierOverride() -> (MobileFontSizeMultiplierOverride:double)
     Index:bool=Get OverrideDefaultMobileFontSizeMultiplier()
    ) -> (ReturnValue)
   )
  ]}
 UpdatePrompt {nodes [
   n1 = UpdatePrompt!() -> (then, InputType:ECommonInputType Enum) {is=title K2Node_FunctionEntry}
   > n2 = Sequence()
   > SetVisibility(self=[Get HoldProgressBar(), Get HoldIcon() -> (HoldIcon:Image Object Reference), Get Prompt()], InVisibility=Collapsed)
   > n3 = Set CurrentHoldProgress(CurrentHoldProgress=0)
   > SetScalarParameterValue(self=Get HoldProgressMID(), ParameterName=Percentage, Value=@n3)
   Set IsPromptHoldable(execute=@n2.then_1, IsPromptHoldable:bool=IsHoldable()) -> (Output_Get:bool)
   > Set IsPromptIcon(IsPromptIcon:bool=IsPromptIcon!() -> (Valid:bool) {is=title K2Node_CallFunction}) -> (Output_Get:bool)
   n4 = AND Boolean(
    A=OR Boolean(A=Get IsPromptIcon() -> (IsPromptIcon:bool), B=Get IsPromptHoldable() -> (IsPromptHoldable:bool))
    B=NOT Boolean(A=Get bHideInputActionWithKeyboard() -> (bHideInputActionWithKeyboard:bool))
   )
   n5 = Get IsPromptHoldable()
   n6 = Get IsPromptIcon()
   > SetVisibility(
    self=Get Prompt()
    InVisibility=Select!(
     MouseAndKeyboard:ESlateVisibility Enum=Select!(Option 0:ESlateVisibility Enum=Collapsed, Option 1:ESlateVisibility Enum=SelfHitTestInvisible, Index:bool=@n4) -> (ReturnValue)
     Gamepad:ESlateVisibility Enum=Select(Option 0=Collapsed, Option 1=SelfHitTestInvisible, Index=@n6)
     Touch:ESlateVisibility Enum=Select(Option 0=Collapsed, Option 1=SelfHitTestInvisible, Index=@n5)
     Count:ESlateVisibility Enum=Collapsed
     Index:ECommonInputType Enum=@n1
    ) -> (ReturnValue:ESlateVisibility Enum)
   )
   > n7 = Branch(Condition=Get IsPromptHoldable())
   n8 = Get IsPromptIcon()
   > SetScalarParameterValue(
    self=Get HoldProgressMID()
    ParameterName=ShapeSwitch
    Value=Select!(
     MouseAndKeyboard:float=Select!(Option 0:float=0, Option 1:float=1, Index:bool=@n8) -> (ReturnValue)
     Gamepad:float=0
     Touch:float=0
     Count:float=0
     Index:ECommonInputType Enum=@n1
    ) -> (ReturnValue:float)
   )
   > SetVisibility(self=Get HoldProgressBar(), InVisibility=SelfHitTestInvisible)
   n9 = Get IsPromptIcon()
   > SetVisibility(
    self=Get HoldIcon()
    InVisibility=Select!(
     MouseAndKeyboard:ESlateVisibility Enum=Select!(Option 0:ESlateVisibility Enum=SelfHitTestInvisible, Option 1:ESlateVisibility Enum=Collapsed, Index:bool=@n9) -> (ReturnValue)
     Gamepad:ESlateVisibility Enum=Collapsed
     Touch:ESlateVisibility Enum=SelfHitTestInvisible
     Count:ESlateVisibility Enum=Collapsed
     Index:ECommonInputType Enum=@n1
    ) -> (ReturnValue:ESlateVisibility Enum)
   )
   SetScalarParameterValue(execute=@n7.else, self=Get HoldProgressMID(), ParameterName=Percentage, Value=0)
   > SetVisibility(self=Get HoldProgressBar(), InVisibility=Collapsed)
   > SetVisibility(self=Get HoldIcon(), InVisibility=Collapsed)
  ]}
]
variables [
 BackgroundMaterial {type=MaterialInstance, default=null}
 Block_Icon {type=WBP_UIKit_Block_Image_C, default=null}
 Block_SecondaryText {type=WBP_UIKit_Block_Text_C, default=null}
 BorderMaterial {type=MaterialInstance, default=null}
 ButtonSize {type=E_UI_ButtonSize, default=Large}
 CurrentHoldProgress {type=number}
 CurrentInputType {type=ECommonInputType, default=MouseAndKeyboard}
 EnableOnWidgetStyled {type=boolean}
 FadeInTime {type=number, default=0.1}
 FadeOutTime {type=number, default=0.2}
 HasIcon {type=boolean}
 HasText {type=boolean}
 HoldCompletedSound {type=SoundBase, default=null}
 HoldPressedSound {type=SoundBase, default=null}
 HoldProgressMID {type=MaterialInstanceDynamic}
 HoldReleasedSound {type=SoundBase, default=null}
 HoldSound {type=SoundBase, default=null}
 HoldSoundAudioComponent {type=AudioComponent, default=null}
 HoveredSound {type=SoundBase, default=null}
 Icon {type=Texture2D}
 IconBeforeText {type=boolean, default=true}
 IconMaterial {type=MaterialInstance, default=null}
 IconPadding {type=Margin, default=<ltrb 8 0 8 0>}
 InputSwitchDelay {type=number, default=0.25}
 IsDarkMode {type=boolean, default=true}
 LargeHeight {type=integer, default=88}
 LargeIconSize {type=Vector2D, default=<48 48>}
 LargePrimaryFont {
  type=SlateFontInfo
  default {
   bForceMonospaced=false
   bMaterialIsStencil=false
   fontMaterial=null
   fontObject=&HeadingNow
   letterSpacing=10
   monospacedWidth=1
   size=22.5
   skewAmount=0
   typefaceFontName=76 Bold
   outlineSettings {
    bApplyOutlineToDropShadows=false
    bMiteredCorners=false
    bSeparateFillAlpha=false
    outlineMaterial=null
    outlineSize=0
    outlineColor=<rgba 0 0 0 1>
   }
  }
 }
 LargeSecondaryFont {
  type=SlateFontInfo
  default {
   bForceMonospaced=false
   bMaterialIsStencil=false
   fontMaterial=null
   fontObject=&HeadingNow
   letterSpacing=10
   monospacedWidth=1
   size=15.75
   skewAmount=0
   typefaceFontName=76 Bold
   outlineSettings {
    bApplyOutlineToDropShadows=false
    bMiteredCorners=false
    bSeparateFillAlpha=false
    outlineMaterial=null
    outlineSize=0
    outlineColor=<rgba 0 0 0 1>
   }
  }
 }
 LastEnabledEvent {type=boolean}
 MediumHeight {type=integer, default=56}
 MediumIconSize {type=Vector2D, default=<40 40>}
 MediumPrimaryFont {
  type=SlateFontInfo
  default {
   bForceMonospaced=false
   bMaterialIsStencil=false
   fontMaterial=null
   fontObject=&HeadingNow
   letterSpacing=10
   monospacedWidth=1
   size=18
   skewAmount=0
   typefaceFontName=76 Bold
   outlineSettings {
    bApplyOutlineToDropShadows=false
    bMiteredCorners=false
    bSeparateFillAlpha=false
    outlineMaterial=null
    outlineSize=0
    outlineColor=<rgba 0 0 0 1>
   }
  }
 }
 MeduimSecondaryFont {
  type=SlateFontInfo
  default {
   bForceMonospaced=false
   bMaterialIsStencil=false
   fontMaterial=null
   fontObject=&HeadingNow
   letterSpacing=10
   monospacedWidth=1
   size=15.75
   skewAmount=0
   typefaceFontName=76 Bold
   outlineSettings {
    bApplyOutlineToDropShadows=false
    bMiteredCorners=false
    bSeparateFillAlpha=false
    outlineMaterial=null
    outlineSize=0
    outlineColor=<rgba 0 0 0 1>
   }
  }
 }
 MinHeightOverride {type=integer}
 MinWidthOverride {type=integer}
 MobileFontSizeMultiplierLarge {type=number, default=1.2}
 MobileFontSizeMultiplierMedium {type=number, default=1.3}
 MobileFontSizeMultiplierOverride {type=number, default=1}
 MobileFontSizeMultiplierSecondaryLarge {type=number, default=1.4}
 MobileFontSizeMultiplierSecondaryMedium {type=number, default=1.6}
 MobileFontSizeMultiplierSecondaryOverride {type=number, default=1}
 MobileFontSizeMultiplierSecondarySmall {type=number, default=1.7}
 MobileFontSizeMultiplierSmall {type=number, default=1.2}
 MobileLargeHeight {type=integer, default=88}
 MobileMediumHeight {type=integer, default=54}
 MobileMinHeight {type=integer}
 MobileMinWidth {type=integer}
 MobileSmallHeight {type=integer, default=48}
 OnFocusChanged {}
 OnHoldActionProgress {}
 OnHoldCompleted {}
 OnReleased {}
 OverrideDefaultHeight {type=boolean}
 OverrideDefaultMobileFontSizeMultiplier {type=boolean}
 OverrideSecondaryTextSize {type=boolean}
 OverrideSecondaryTextSizeMultiplier {type=number, default=1}
 PressedSound {type=SoundBase, default=null}
 PressedSoundAudioComponent {type=AudioComponent, default=null}
 SmallHeight {type=integer, default=48}
 SmallIconSize {type=Vector2D, default=<32 32>}
 SmallPrimaryFont {
  type=SlateFontInfo
  default {
   bForceMonospaced=false
   bMaterialIsStencil=false
   fontMaterial=null
   fontObject=&HeadingNow
   letterSpacing=10
   monospacedWidth=1
   size=15.75
   skewAmount=0
   typefaceFontName=76 Bold
   outlineSettings {
    bApplyOutlineToDropShadows=false
    bMiteredCorners=false
    bSeparateFillAlpha=false
    outlineMaterial=null
    outlineSize=0
    outlineColor=<rgba 0 0 0 1>
   }
  }
 }
 SmallSecondaryFont {
  type=SlateFontInfo
  default {
   bForceMonospaced=false
   bMaterialIsStencil=false
   fontMaterial=null
   fontObject=&HeadingNow
   letterSpacing=10
   monospacedWidth=1
   size=15.75
   skewAmount=0
   typefaceFontName=76 Bold
   outlineSettings {
    bApplyOutlineToDropShadows=false
    bMiteredCorners=false
    bSeparateFillAlpha=false
    outlineMaterial=null
    outlineSize=0
    outlineColor=<rgba 0 0 0 1>
   }
  }
 }
 UseDisabledVisualsOnLocked {type=boolean}
]
widgets {
 info {inheritedWidgetCount=1, parentClass=/Script/FortniteUI.FortCTAButton, rootWidgetClass=&GridPanel, widgetCount=13}
 widgets [
  Root {widgetClassPath=&GridPanel, bIsVariable=true, widget=Root, properties {columnFill [0,0,0,1,0,0], rowFill [0,1,0]}}
  Block_Background {
   widgetClassPath=/FortUILibrary/UIKit/Buttons/Blocks/WBP_UIKit_Block_BackgroundWithOutline
   bIsVariable=true
   parent=Root
   widget=Block_Background
   properties {backgroundMaterial=null, outlineMaterial=null}
   slot {class=GridSlot, properties {columnSpan=6, rowSpan=3}}
  }
  Spacer_TopLeft {
   widgetClassPath=&Spacer
   bIsVariable=true
   parent=Root
   widget=Spacer_TopLeft
   properties {size=<0 0>}
   slot {class=GridSlot, properties {horizontalAlignment=HAlign_Left}}
  }
  Slot_AltPrompt {
   widgetClassPath=&NamedSlot
   bIsVariable=true
   parent=Root
   widget=Slot_AltPrompt
   slot {
    class=GridSlot
    properties {column=1, horizontalAlignment=HAlign_Center, layer=20, row=1, verticalAlignment=VAlign_Center, padding=<ltrb 0 0 4 0>}
   }
  }
  Prompt {
   widgetClassPath=&GridPanel
   bIsVariable=true
   parent=Root
   widget=Prompt
   properties {columnFill [1], rowFill [1]}
   slot {class=GridSlot, properties {column=1, layer=20, row=1, padding=<ltrb 0 0 4 0>}}
  }
  CommonVisualAttachment {
   widgetClassPath=&CommonVisualAttachment
   parent=Prompt
   widget=CommonVisualAttachment
   properties {contentAnchor=<0.5 0.5>}
   slot {class=GridSlot, properties {horizontalAlignment=HAlign_Center, layer=10, verticalAlignment=VAlign_Center}}
  }
  HoldProgressBar {
   widgetClassPath=&Image
   bIsVariable=true
   parent=CommonVisualAttachment
   widget=HoldProgressBar
   properties {
    visibility=Collapsed
    brush {
     bIsDynamicallyLoaded=false
     drawAs=Image
     imageType=NoImage
     mirroring=NoMirror
     resourceName=None
     resourceObject=/FortUILibrary/UIKit/Buttons/MaterialInstances/CTAButtons/MI_UIKit_InputActionProgress
     tiling=NoTile
     imageSize=<48 48>
     margin=<ltrb 0 0 0 0>
     outlineSettings {
      bUseBrushTransparency=false
      roundingType=HalfHeightRadius
      width=0
      color {colorUseRule=UseColor_Specified, specifiedColor=<rgba 0 0 0 0>}
      cornerRadii=<0 0 0 0>
     }
     tintColor {colorUseRule=UseColor_Specified, specifiedColor=<rgba 1 1 1 1>}
     uVRegion {bIsValid=false, max=<0 0>, min=<0 0>}
    }
   }
   slot {class=SizeBoxSlot, properties {horizontalAlignment=HAlign_Center, verticalAlignment=VAlign_Center}}
  }
  HoldIcon {
   widgetClassPath=&Image
   bIsVariable=true
   parent=Prompt
   widget=HoldIcon
   properties {
    visibility=Collapsed
    brush {
     bIsDynamicallyLoaded=false
     drawAs=Image
     imageType=NoImage
     mirroring=NoMirror
     resourceName=None
     resourceObject=/FortUILibrary/Systems/ModularButton/MaterialInstances/MI_MousePressAndHold
     tiling=NoTile
     imageSize=<28 28>
     margin=<ltrb 0 0 0 0>
     outlineSettings {
      bUseBrushTransparency=false
      roundingType=HalfHeightRadius
      width=0
      color {colorUseRule=UseColor_Specified, specifiedColor=<rgba 0 0 0 0>}
      cornerRadii=<0 0 0 0>
     }
     tintColor {colorUseRule=UseColor_Specified, specifiedColor=<rgba 1 1 1 1>}
     uVRegion {bIsValid=false, max=<0 0>, min=<0 0>}
    }
   }
   slot {class=GridSlot, properties {horizontalAlignment=HAlign_Center, layer=30, verticalAlignment=VAlign_Center}}
  }
  InputActionWidget {
   widgetClassPath=&CommonActionWidget
   bInherited=true
   bIsVariable=true
   parent=Prompt
   widget=InputActionWidget
   properties {
    progressMaterialParam=Percentage
    visibility=HitTestInvisible
    progressMaterialBrush {
     bIsDynamicallyLoaded=false
     drawAs=Image
     imageType=NoImage
     mirroring=NoMirror
     resourceName=None
     resourceObject=/FortUILibrary/UIKit/Buttons/MaterialInstances/CTAButtons/MI_UIKit_EmptyProgress
     tiling=NoTile
     imageSize=<32 32>
     margin=<ltrb 0 0 0 0>
     outlineSettings {
      bUseBrushTransparency=false
      roundingType=HalfHeightRadius
      width=0
      color {colorUseRule=UseColor_Specified, specifiedColor=<rgba 0 0 0 0>}
      cornerRadii=<0 0 0 0>
     }
     tintColor {colorUseRule=UseColor_Specified, specifiedColor=<rgba 1 1 1 1>}
     uVRegion {bIsValid=false, max=<0 0>, min=<0 0>}
    }
   }
   slot {class=GridSlot, properties {horizontalAlignment=HAlign_Center, layer=40, verticalAlignment=VAlign_Center}}
  }
  Texts {
   widgetClassPath=&VerticalBox
   bIsVariable=true
   parent=Root
   widget=Texts
   slot {class=GridSlot, properties {column=3, row=1, verticalAlignment=VAlign_Center, padding=<ltrb 8 0 8 0>}}
  }
  Block_Text {
   widgetClassPath=/FortUILibrary/Systems/ModularButton/Blocks/WBP_UIKit_Block_Text
   bIsVariable=true
   parent=Texts
   widget=Block_Text
   slot {class=VerticalBoxSlot, properties {horizontalAlignment=HAlign_Center, verticalAlignment=VAlign_Center}}
  }
  BelowTextSlot {
   widgetClassPath=&NamedSlot
   bIsVariable=true
   parent=Texts
   widget=BelowTextSlot
   slot {class=VerticalBoxSlot, properties {horizontalAlignment=HAlign_Center, verticalAlignment=VAlign_Top}}
  }
  Spacer_BottomRight {
   widgetClassPath=&Spacer
   bIsVariable=true
   parent=Root
   widget=Spacer_BottomRight
   properties {size=<0 0>}
   slot {class=GridSlot, properties {column=5, horizontalAlignment=HAlign_Right, row=2}}
  }
 ]
}
