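# M_UIKit_CollapsableArrow is a masked UI-domain material in the UIKit Buttons folder, named for a collapsable arrow graphic.
# Its graph holds a single texture sample of the CollapsableArrow texture, with no parameters or math nodes.
# No connection from that sample to a material output is shown.
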Material /FortUILibrary/UIKit/Buttons/Materials/M_UIKit_CollapsableArrow
graph {
 properties {bUsedWithStaticMesh=false, blendMode=BLEND_Masked, materialDomain=MD_UI, pixelDepthOffsetMode=PDOM_Legacy, usedShadingModels=Default Lit}
 nodes [TextureSample() {texture=/FortUILibrary/UIKit/Buttons/Textures/CollapsableArrow}]
}
