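# M_UIKit_RadialGlow is a translucent UI material in the UIKit Buttons folder that produces a tintable radial glow.
# MF_UI_RadialGradient builds a gradient from the From and To parameters, and its Smooth output is raised to RadialGlowFalloffRate to shape the falloff.
# That mask is multiplied by the Color parameter's alpha and AlphaPost, then saturated.
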Material /FortUILibrary/UIKit/Buttons/Materials/M_UIKit_RadialGlow
graph {
 properties {
  bUsedWithStaticMesh=false
  blendMode=BLEND_Translucent
  materialDomain=MD_UI
  pixelDepthOffsetMode=PDOM_Legacy
  usedShadingModels=Default Lit
 }
 nodes [
  n1 = VectorParameter() {parameterName=Color, defaultValue=<rgba 1 0 1 1>}
  n2 = Power(
   Base=MaterialFunctionCall(GradientStart=ScalarParameter() {parameterName=From, sliderMax=1}, GradientEnd=ScalarParameter() {defaultValue=1, parameterName=To}) {materialFunction=&MF_UI_RadialGradient}.Smooth
   Exp=ScalarParameter() {defaultValue=1, parameterName=RadialGlowFalloffRate}
  )
  Saturate(None=Multiply(A=Multiply(A=@n1.A, B=@n2), B=ScalarParameter() {defaultValue=1, parameterName=AlphaPost}))
 ]
}
