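# UIKit modular button widget that shows a toggle indicator with an outline, styled by a ToggleType that defaults to Radio.
# PreConstruct registers the Toggle and Outline blocks with the modular button base and sets their state-transition timings.
# SetToggleType passes the type to the toggle block and derives outline corner rounding, toggleability and interactable-when-selected from it.
# A spacer sets the size, which grows on mobile.
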WidgetBlueprint /FortUILibrary/UIKit/Buttons/WBP_UIKit_ButtonWithToggle
graphs [
 EventGraph {nodes [
   n1 = Event PreConstruct()
   > Parent: PreConstruct(IsDesignTime=@n1.IsDesignTime)
   > SetGridSlotAlignment!(
    execute
    Widget:Widget Object Reference=Get Slot_Main() -> (Slot_Main:Named Slot Object Reference)
    Horizontal:EHorizontalAlignment Enum=HAlign_Center
    Vertical:EVerticalAlignment Enum=VAlign_Center
   ) -> (then)
   > SetToggleType!(execute, self, Type:E_UI_ToggleType Enum=Get ToggleType() -> (ToggleType:E_UI_ToggleType Enum=NewEnumerator0)) -> (then)
   > RegisterModularBlock!(
    execute
    self:UIKit Modular Button Object Reference
    Block:UIKit Block Object Reference=Get Block_Toggle() -> (Block_Toggle:WBP UIKit Block Toggle Object Reference)
   ) -> (then)
   > RegisterModularBlock(Block=Get Block_Outline() -> (Block_Outline:WBP UIKit Block Outline Object Reference))
   > SetBlocksTiming!(
    execute
    self:UIKit Modular Button Object Reference
    Timing_Hovering:float=0
    Timing_Unhovering:float=0
    Timing_Focusing:float=0.3
    Timing_Unfocusing:float=0.15
    Timing_Pressing:float=0
    Timing_Releasing:float=0
    Timing_Disabling:float=0.1
    Timing_Enabling:float=0.1
    Timing_Selecting:float=0.1
    Timing_DeselectingIdle:float=0.15
    Timing_DeselectingFocused:float=0
   ) -> (then)
   > SetBlockTiming!(
    execute
    self:UIKit Block Object Reference=Get Block_Toggle()
    Timing_Hovering:float=0
    Timing_Unhovering:float=0
    Timing_Focusing:float=0.1
    Timing_Unfocusing:float=0
    Timing_Pressing:float=0
    Timing_Releasing:float=0
    Timing_Disabling:float=0.1
    Timing_Enabling:float=0.1
    Timing_Selecting:float=0.05
    Timing_DeselectingIdle:float=0
    Timing_DeselectingFocused:float=0
   ) -> (then)
   > SetCompactUIStyle() {is=title K2Node_CallFunction}
   Event Construct()
   > Parent: Construct()
   Event On Widget Styled() {is=title K2Node_Event}
   > Parent: On Widget Styled() {is=title K2Node_CallParentFunction}
  ]}
 SetCompactUIStyle {nodes [
   SetCompactUIStyle() {is=title K2Node_FunctionEntry}
   > Branch(Condition=IsMobileGame())
   > SetSize!(execute, self:Spacer Object Reference=Get Spacer_Size() -> (Spacer_Size:Spacer Object Reference), InSize:Vector 2D Structure=(X=64,Y=64)) -> (then)
  ]}
 SetToggleType {nodes [
   n1 = SetToggleType!() -> (then, Type:E_UI_ToggleType Enum=NewEnumerator0) {is=title K2Node_FunctionEntry}
   > n2 = Set ToggleType(ToggleType:E_UI_ToggleType Enum=@n1) -> (Output_Get:E_UI_ToggleType Enum=NewEnumerator0)
   > SetToggleType(self:WBP UIKit Block Toggle Object Reference=Get Block_Toggle(), Type=@n2) {is=title K2Node_CallFunction}
   n3 = Select!(NewEnumerator0:float=1, NewEnumerator1:float=0.3, NewEnumerator2:float=0.3, Index:E_UI_ToggleType Enum=@n2) -> (ReturnValue:float)
   > SetCornerRadiuses!(
    execute
    self:WBP UIKit Block Outline Object Reference=Get Block_Outline()
    TopLeft:float=@n3
    TopRight:float=@n3
    BottomLeft:float=@n3
    BottomRight:float=@n3
    RadiusInPixels:bool=false
   ) -> (then)
   n4 = Select(NewEnumerator0:bool=false, NewEnumerator1:bool=true, NewEnumerator2:bool=true, Index=@n2) -> (ReturnValue:bool)
   > SetIsToggleable!(execute, self:Common Button Base Object Reference, bInIsToggleable:bool=@n4) -> (then)
   > SetIsInteractableWhenSelected!(execute, self:Common Button Base Object Reference, bInInteractableWhenSelected:bool=@n4) -> (then)
  ]}
]
variables [ToggleType {type=E_UI_ToggleType, default=Radio}]
widgets {
 info {
  inheritedWidgetCount=0
  namedSlotCount=1
  parentClass=/FortUILibrary/Systems/ModularButton/WBP_UIKit_ButtonGeneric_Base.WBP_UIKit_ButtonGeneric_Base_C
  rootWidgetClass=null
  widgetCount=4
 }
 widgets [
  Main_Root {widgetClassPath=&GridPanel, widget=Main_Root, properties {columnFill [1], rowFill [1]}}
  Block_Toggle {
   widgetClassPath=/FortUILibrary/Systems/ModularButton/Blocks/WBP_UIKit_Block_Toggle
   bIsVariable=true
   parent=Main_Root
   widget=Block_Toggle
   slot {class=GridSlot}
  }
  Block_Outline {
   widgetClassPath=/FortUILibrary/UIKit/Buttons/Blocks/WBP_UIKit_Block_Outline
   bIsVariable=true
   parent=Main_Root
   widget=Block_Outline
   properties {bottomLeftCornerRadius=1, bottomRightCornerRadius=1, topLeftCornerRadius=1, topRightCornerRadius=1}
   slot {class=GridSlot, properties {layer=1}}
  }
  Spacer_Size {widgetClassPath=&Spacer, bIsVariable=true, parent=Main_Root, widget=Spacer_Size, properties {size=<32 32>}, slot {class=GridSlot}}
 ]
}
