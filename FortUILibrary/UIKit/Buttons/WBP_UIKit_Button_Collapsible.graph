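# Collapsible UIKit button, built on the generic button base, that animates its label away, with Regular or Quiet background and font materials.
# OnCollapse and OnExpand fade and slide the content stack and drive an animated value that eases width and padding, resuming mid-way when reversed.
# Prompt and content share an Overlay; on gamepad, padding moves the prompt below the icon when collapsed.
# Input method also sets text padding and Arabic status-indicator spacing.
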WidgetBlueprint /FortUILibrary/UIKit/Buttons/WBP_UIKit_Button_Collapsible
animations [
 OnCollapse {
  binding_count=2
  length_seconds=0.35
  object=&self:OnCollapse
  sequence {
   bindings [
    Main_Stack {
     id=#1
     possessed_class=StackBox
     tracks [
      {
       type=MovieSceneFloatTrack
       display_name=Render Opacity
       sections [{
         type=MovieSceneFloatSection
         end_frame=35
         start_frame=0
         channels [None {type=MovieSceneScriptingFloatChannel, keys [[0,1],[9,1],[14,0],[34.9983,0]]}]
        }]
      }
      {
       type=MovieScene2DTransformTrack
       sections [{
         type=MovieScene2DTransformSection
         end_frame=35
         start_frame=0
         channels [Translation.X {type=MovieSceneScriptingFloatChannel, keys [[0,0],[14,-10],[34.9983,-10]]}]
        }]
      }
     ]
    }
    Self {
     id=#2
     possessed_class=WBP_UIKit_Button_Collapsible_C
     tracks [
      {
       type=MovieSceneDoubleTrack
       display_name=On Collapse Anim Value
       sections [{
         type=MovieSceneDoubleSection
         end_frame=35
         start_frame=0
         channels [None {type=MovieSceneScriptingDoubleChannel, keys [[0,0],[12,0],[34.9983,1]]}]
        }]
      }
      {
       type=MovieSceneEventTrack
       sections [{
         type=MovieSceneEventTriggerSection
         channels [None {type=MovieSceneScriptingEventChannel, const ["<Struct 'MovieSceneEvent' (0x466b2f2b0e0) {}>", 1]}]
        }]
      }
     ]
    }
   ]
   display_rate {numerator=100}
  }
 }
 OnExpand {
  binding_count=2
  length_seconds=0.35
  object=&self:OnExpand
  sequence {
   bindings [
    Main_Stack {
     id=#1
     possessed_class=StackBox
     tracks [
      {
       type=MovieSceneFloatTrack
       display_name=Render Opacity
       sections [{
         type=MovieSceneFloatSection
         end_frame=35
         start_frame=0
         channels [None {type=MovieSceneScriptingFloatChannel, keys [[0,0],[10,0],[19,1],[35,1]]}]
        }]
      }
      {
       type=MovieScene2DTransformTrack
       sections [{
         type=MovieScene2DTransformSection
         end_frame=35
         start_frame=0
         channels [Translation.X {type=MovieSceneScriptingFloatChannel, keys [[0,-10],[10,-10],[23,0],[35,0]]}]
        }]
      }
     ]
    }
    Self {
     id=#2
     possessed_class=WBP_UIKit_Button_Collapsible_C
     tracks [
      {
       type=MovieSceneDoubleTrack
       display_name=On Collapse Anim Value
       sections [{
         type=MovieSceneDoubleSection
         end_frame=35
         start_frame=0
         channels [None {type=MovieSceneScriptingDoubleChannel, keys [[0,1],[23,0],[35,0]]}]
        }]
      }
      {
       type=MovieSceneEventTrack
       sections [{
         type=MovieSceneEventTriggerSection
         channels [None {type=MovieSceneScriptingEventChannel, const ["<Struct 'MovieSceneEvent' (0x466b2f2bae0) {}>", 1]}]
        }]
      }
     ]
    }
   ]
   display_rate {numerator=100}
  }
 }
]
graphs [
 CalculatePaddingCorrection {nodes [
   CalculatePaddingCorrection() {is=title K2Node_FunctionEntry}
   n1 = Get!() -> (IconSize_X:double, IconSize_Y:double)
   n2 = Get!() -> (InputActionIconSizeReference_X:double, InputActionIconSizeReference_Y:double)
   > Set!(
    execute
    InputActionPaddingCorrection_X:double="Max (Float)"(A=float - float(A=@n1, B=@n2), B=0)
    InputActionPaddingCorrection_Y:double="Max (Float)"(A=float - float(A=@n1.IconSize_Y, B=@n2.InputActionIconSizeReference_Y), B=0)
   ) -> (then, Output_Get:Vector 2D Structure)
   n3 = Get()
   n4 = Get!() -> (IconSize_X:double, IconSize_Y:double)
   > Set!(
    execute
    IconPaddingCorrection_X:double="Max (Float)"(A=float - float(A=@n3, B=@n4), B=0)
    IconPaddingCorrection_Y:double="Max (Float)"(A=float - float(A=@n3.InputActionIconSizeReference_Y, B=@n4.IconSize_Y), B=0)
   ) -> (then, Output_Get)
  ]}
 EventGraph {nodes [
   n1 = ChangeEnabledStyle!() -> (OutputDelegate, then, Enabled:bool=~) {is=title K2Node_CustomEvent}
   > n2 = Branch(Condition=@n1.Enabled)
   SetColorAndOpacity(execute=[@n8.then, @n2.then], self:User Widget Object Reference, InColorAndOpacity:Linear Color Structure=(R=1,G=1,B=1,A=1))
   SetColorAndOpacity(execute=[@n6.then, @n2.else], InColorAndOpacity=(R=1,G=1,B=1,A=0.4))
   n3 = Event PreConstruct()
   > Parent: PreConstruct(IsDesignTime=@n3.IsDesignTime)
   > SetSlotPadding()
   > n4 = Sequence()
   > Branch(Condition=@n3.IsDesignTime)
   > SetupCollapsibleLayout()
   > SetPromptRelatedPadding!(execute, self, Value:double=0) -> (then)
   > SetVisibility(self=Get Prompt() -> (Prompt:Grid Panel Object Reference), InVisibility=Collapsed)
   SetColorAndOpacity(
    execute=@n4.then_1
    self=Get Block_Icon() -> (Block_Icon:WBP UIKit Block Image Object Reference)
    InColorAndOpacity=Get IconColorAndOpacityOverride() -> (IconColorAndOpacityOverride:Linear Color Structure)
   )
   > SetWidthOverride(self=Get Main_StackSize() -> (Main_StackSize:Size Box Object Reference), InWidthOverride=GetDesiredMaxTextWidth!() -> (MaxTextWidth:double))
   > SetContextualTextPadding()
   > SetButtonType!(execute, self, ButtonType:E_UI_CTAButtonType Enum=Get ButtonType() -> (ButtonType:E_UI_CTAButtonType Enum=NewEnumerator0)) -> (then)
   > Branch(Condition="Equal Exactly (String)"(A=GetCurrentCulture!() -> (ReturnValue:String), B=ar))
   > SetGridSlotPadding!(
    execute
    Widget:Widget Object Reference=Get StatusIndicator() -> (StatusIndicator:WBP UIKit Status Indicator Object Reference)
    Padding_Left:float=Select!(
     MouseAndKeyboard:double=8
     Gamepad:double=48
     Touch:double=8
     Count:double=8
     Index:ECommonInputType Enum=Get CurrentInputType() -> (CurrentInputType:ECommonInputType Enum)
    ) -> (ReturnValue:double)
    Padding_Top:float=0
    Padding_Right:float=0
    Padding_Bottom:float=0
   ) -> (then)
   Event Construct()
   > Parent: Construct()
   > SetupCollapsibleLayout()
   > SetPromptRelatedPadding(Value=0)
   n5 = Event SetIsCollapsed!() -> (OutputDelegate, then, IsCollapsed:bool, SkipAnimation:bool) {is=title K2Node_Event}
   > PlayCollapseAnimation!(execute, self, ShouldCollapse:bool=@n5.IsCollapsed, SkipAnimation:bool=@n5.SkipAnimation) -> (then)
   n6 = Event On Disabled() {is=title K2Node_Event}
   n7 = Event On Input Method Changed!() -> (OutputDelegate, then, CurrentInputType:ECommonInputType Enum) {is=title K2Node_Event}
   > Parent: On Input Method Changed!(execute, CurrentInputType:ECommonInputType Enum=@n7.CurrentInputType) -> (then) {is=title K2Node_CallParentFunction}
   > SetContextualTextPadding()
   > SetPromptRelatedPadding(Value=SelectFloat(A=1, B=0, bPickA=Get IsCollapsed() -> (IsCollapsed:bool)))
   > Branch(Condition="Equal Exactly (String)"(A=GetCurrentCulture(), B=ar))
   > SetGridSlotPadding(
    Widget=Get StatusIndicator()
    Padding_Left=Select(MouseAndKeyboard=8, Gamepad=48, Touch=8, Count=8, Index=@n7.CurrentInputType)
    Padding_Top=0
    Padding_Right=0
    Padding_Bottom=0
   )
   n8 = Event On Enabled() {is=title K2Node_Event}
   Event Destruct()
   > Parent: Destruct()
  ]}
 GetDesiredMaxTextWidth {nodes [
   GetDesiredMaxTextWidth() {is=title K2Node_FunctionEntry}
   n1 = Get MaxTextWidth() -> (MaxTextWidth:double)
   n2 = Get!() -> (InputActionIconSizeReference_X:double, InputActionIconSizeReference_Y:double)
   > Return Node!(
    execute
    MaxTextWidth:double=SelectFloat(A=float - float(A=@n1, B=@n2), B=Get MaxTextWidth(), bPickA=Is Input Gamepad!() -> (Result:bool=~) {is=title K2Node_MacroInstance})
   )
  ]}
 GetRemainingAnimTime {nodes [
   GetRemainingAnimTime() {is=title K2Node_FunctionEntry}
   > n1 = Branch(Condition=IsAnimationPlaying(InAnimation=Get OnCollapse() -> (OnCollapse:Widget Animation Object Reference)))
   n2 = Get OnCollapse()
   > Return Node!(
    execute
    FunctionIsImplemented:bool=true
    RemainingTime:double=float - float(
     A=GetEndTime(self=@n2)
     B=GetAnimationCurrentTime!(self:User Widget Object Reference, InAnimation:Widget Animation Object Reference=@n2) -> (ReturnValue:float)
    )
    TotalTime:double=GetEndTime(self=Get OnCollapse())
   )
   n3 = Branch(execute=@n1.else, Condition=IsAnimationPlaying(InAnimation=Get OnExpand() -> (OnExpand:Widget Animation Object Reference)))
   n4 = Get OnExpand()
   > Return Node(
    FunctionIsImplemented=true
    RemainingTime=float - float(A=GetEndTime(self=@n4), B=GetAnimationCurrentTime(InAnimation=@n4))
    TotalTime=GetEndTime(self=Get OnExpand())
   )
   PrintString!(execute=@n3.else, InString=Neither Anim is playing, bPrintToScreen, bPrintToLog, TextColor, Duration, Key) -> (then)
   > Return Node(FunctionIsImplemented=true, RemainingTime=0, TotalTime=0)
   n5 = Get OnCollapse()
   n6 = GetStartTime!(self:Widget Animation Object Reference=@n5) -> (ReturnValue:float)
   n7 = GetEndTime(self=@n5)
   n8 = GetAnimationCurrentTime(InAnimation=@n5)
   PrintString(InString=Append!(
     A="Start: "
     B="To String (Float)"(InDouble=@n6)
     C:String=" || Current: "
     D:String="To String (Float)"(InDouble=@n8)
     E:String=" || End: "
     F:String="To String (Float)"(InDouble=@n7)
    ) -> (ReturnValue))
   n9 = Get OnExpand()
   n10 = GetStartTime(self=@n9)
   n11 = GetEndTime(self=@n9)
   n12 = GetAnimationCurrentTime(InAnimation=@n9)
   PrintString(InString=Append(
     A="Start: "
     B="To String (Float)"(InDouble=@n10)
     C=" || Current: "
     D="To String (Float)"(InDouble=@n12)
     E=" || End: "
     F="To String (Float)"(InDouble=@n11)
    ))
  ]}
 HandleOnAnimFinished {nodes [
   HandleOnAnimFinished() {is=title K2Node_FunctionEntry}
   > SetOnCollapseAnimValue!(execute, self, Value:double=SelectFloat(A=1, B=0, bPickA=Get IsCollapsed())) -> (then)
   > Set RenderOpacity!(
    execute
    RenderOpacity:float=SelectFloat(A=0, B=1, bPickA=Get IsCollapsed())
    self:Widget Object Reference=Get Main_Stack() -> (Main_Stack:Stack Box Object Reference)
   ) -> (then, Output_Get:float)
   > SetRenderTranslation!(execute, self=Get Main_Stack(), Translation_X:double=SelectFloat(A=-10, B=0, bPickA=Get IsCollapsed()), Translation_Y:double=0) -> (then)
  ]}
 IsInputGamepad {nodes [
   Inputs!()
   Outputs!(Result:bool="Equal (Enum)"(A=Get CurrentInputType(), B=Gamepad))
  ]}
 OverrideFontMaterials {nodes [
   n1 = OverrideFontMaterials!() -> (then, PrimaryFontMaterial:Material Instance Object Reference, SecondaryFontMaterial:Material Instance Object Reference) {is=title K2Node_FunctionEntry}
   > n2 = Sequence()
   n3 = Break Slate Font Info!(SlateFontInfo:"Slate Font Info Structure (by ref)"=Get Font() -> (Font:Slate Font Info Structure)) -> (
    FontObject:Object Reference
    FontMaterial:Object Reference
    OutlineSettings:Font Outline Settings Structure
    TypefaceFontName:Name
    Size:float=~
    LetterSpacing:int
    SkewAmount:float=~
    bForceMonospaced:bool=~
    bMaterialIsStencil:bool=~
    MonospacedWidth:float=~
   ) {is=title K2Node_BreakStruct}
   > Set Font(Font:Slate Font Info Structure=MakeSlateFontInfo!(
     FontObject:Object Reference=@n3
     FontMaterial:Object Reference=@n1
     OutlineSettings:Font Outline Settings Structure=@n3.OutlineSettings
     TypefaceFontName:Name=@n3.TypefaceFontName
     Size:float=@n3.Size
     LetterSpacing:int=@n3.LetterSpacing
     SkewAmount:float=@n3.SkewAmount
     bForceMonospaced:bool=false
     bMaterialIsStencil:bool=false
     MonospacedWidth:float=1
    ) -> (ReturnValue:Slate Font Info Structure)) -> (Output_Get:Slate Font Info Structure)
   n4 = Break Slate Font Info(SlateFontInfo=Get SecondFont() -> (SecondFont:Slate Font Info Structure)) -> (Size=~, SkewAmount=~, bForceMonospaced=~, bMaterialIsStencil=~, MonospacedWidth=~)
   Set SecondFont(
    execute=@n2.then_1
    SecondFont:Slate Font Info Structure=MakeSlateFontInfo(
     FontObject=@n4
     FontMaterial=@n1.SecondaryFontMaterial
     OutlineSettings=@n4.OutlineSettings
     TypefaceFontName=@n4.TypefaceFontName
     Size=@n4.Size
     LetterSpacing=@n4.LetterSpacing
     SkewAmount=@n4.SkewAmount
     bForceMonospaced=false
     bMaterialIsStencil=false
     MonospacedWidth=1
    )
   ) -> (Output_Get:Slate Font Info Structure)
  ]}
 PlayCollapseAnimation {nodes [
   n1 = PlayCollapseAnimation!() -> (then, ShouldCollapse:bool=~, SkipAnimation:bool=~) {is=title K2Node_FunctionEntry}
   > Set IsCollapsed(IsCollapsed:bool=@n1) -> (Output_Get:bool)
   > n2 = Sequence()
   n3 = Branch(execute=@n2.then_1, Condition=Get SkipAnimation() -> (SkipAnimation:bool))
   > HandleOnAnimFinished() {is=title K2Node_CallFunction}
   PlayAnimation(
    execute=@n3.else
    InAnimation=Select!(
     Option 0:Widget Animation Object Reference=Get OnExpand()
     Option 1:Widget Animation Object Reference=Get OnCollapse()
     Index:bool=Get ShouldCollapse() -> (ShouldCollapse:bool)
    ) -> (ReturnValue:Widget Animation Object Reference)
    StartAtTime=Get StoppedAnimationTime() -> (StoppedAnimationTime:double)
   )
   n4 = Set AnimationToStop(
    execute=@n2.then_0
    AnimationToStop:Widget Animation Object Reference=Select(Option 0=Get OnCollapse(), Option 1=Get OnExpand(), Index=Get ShouldCollapse())
   ) -> (Output_Get:Widget Animation Object Reference)
   n5 = GetAnimationCurrentTime(InAnimation=@n4)
   n6 = GetEndTime(self=@n4)
   > Set StoppedAnimationTime(StoppedAnimationTime:double=Select(Option 0:float=0, Option 1:float=float - float(A=@n6, B=@n5), Index=IsAnimationPlaying(InAnimation=@n4)) -> (ReturnValue:float)) -> (Output_Get:double)
   > StopAnimation(InAnimation=Get AnimationToStop() -> (AnimationToStop:Widget Animation Object Reference))
  ]}
 SetButtonType {nodes [
   n1 = SetButtonType!() -> (then, ButtonType:E_UI_CTAButtonType Enum=NewEnumerator0) {is=title K2Node_FunctionEntry}
   > n2 = Set ButtonType(ButtonType:E_UI_CTAButtonType Enum=@n1) -> (Output_Get:E_UI_CTAButtonType Enum=NewEnumerator0)
   > Set BackgroundMaterial(BackgroundMaterial:Material Instance Object Reference=Select!(
     NewEnumerator0:Material Instance Object Reference=Get RegularBackgroundMaterial() -> (RegularBackgroundMaterial:Material Instance Object Reference)
     NewEnumerator1:Material Instance Object Reference=Get QuietBackgroundMaterial() -> (QuietBackgroundMaterial:Material Instance Object Reference)
     Index:E_UI_CTAButtonType Enum=@n2
    ) -> (ReturnValue:Material Instance Object Reference)) -> (Output_Get:Material Instance Object Reference)
   > UpdateBackground(self:WBP UIKit Button Generic Base Object Reference)
   > OverrideFontMaterials!(
    execute
    self
    PrimaryFontMaterial:Material Instance Object Reference=Select(
     NewEnumerator0=Get RegularFontMaterial() -> (RegularFontMaterial:Material Instance Object Reference)
     NewEnumerator1=Get QuietFontMaterial() -> (QuietFontMaterial:Material Instance Object Reference)
     Index=@n2
    )
    SecondaryFontMaterial:Material Instance Object Reference=Select(
     NewEnumerator0=Get RegularSecondaryFontMaterial() -> (RegularSecondaryFontMaterial:Material Instance Object Reference)
     NewEnumerator1=Get QuietSecondaryFontMaterial() -> (QuietSecondaryFontMaterial:Material Instance Object Reference)
     Index=@n2
    )
   ) -> (then) {is=title K2Node_CallFunction}
  ]}
 SetContextualTextPadding {nodes [
   SetContextualTextPadding() {is=title K2Node_FunctionEntry}
   > SetTextPadding!(
    execute
    self:WBP UIKit Button Generic Base Object Reference
    TextPadding:Margin Structure=Select!(
     Option 0:Margin Structure=Get TextPaddingInputActionHidden() -> (TextPaddingInputActionHidden:Margin Structure)
     Option 1:Margin Structure=Get TextPaddingInputActionVisible() -> (TextPaddingInputActionVisible:Margin Structure)
     Index:bool=Is Input Gamepad() -> (Result=~)
    ) -> (ReturnValue:Margin Structure)
   ) -> (then)
  ]}
 SetOnCollapseAnimValue {nodes [
   SetOnCollapseAnimValue!() -> (then, Value:double=~) {is=title K2Node_FunctionEntry}
   > SetWidthOverride(
    self=Get Main_StackSize()
    InWidthOverride=Ease!(
     Function:EEasingFunc Enum=Linear
     Alpha:double=Get value() -> (value:double)
     A:double=GetDesiredMaxTextWidth() {is=title K2Node_CallFunction}
     B:double
    ) -> (Result:double=~) {is=title K2Node_EaseFunction}
   )
   > SetPromptRelatedPadding(Value=Get value())
  ]}
 SetPromptRelatedPadding {nodes [
   SetPromptRelatedPadding!() -> (then, Value:double=~) {is=title K2Node_FunctionEntry}
   > n1 = Branch(Condition=NOT Boolean(A=Is Input Gamepad() -> (Result=~)))
   > SetPadding!(
    execute
    self:Overlay Slot Object Reference=SlotAsOverlaySlot(Widget=Get Main() -> (Main:Grid Panel Object Reference))
    InPadding_Left:float=0
    InPadding_Top:float=0
    InPadding_Right:float=0
    InPadding_Bottom:float=0
   ) -> (then)
   > SetPadding(self=SlotAsOverlaySlot(Widget=Get Prompt()), InPadding_Left=0, InPadding_Top=0, InPadding_Right=0, InPadding_Bottom=0)
   n2 = Sequence(execute=@n1.else)
   n3 = Get Value() -> (Value:double)
   n4 = Ease!(Function=EaseInOut, Alpha=@n3, A, B=1, BlendExp:double=2) -> (Result=~)
   n5 = Get!() -> (IconPaddingCorrection_X:double, IconPaddingCorrection_Y:double)
   n6 = float * float(A=@n5, B=0.5) {is=title K2Node_PromotableOperator}
   n7 = Get!() -> (MainSlotPadding_Left:float=~, MainSlotPadding_Top:float=~, MainSlotPadding_Right:float=~, MainSlotPadding_Bottom:float=~)
   n8 = Get!() -> (InputActionIconSizeReference_X:double, InputActionIconSizeReference_Y:double)
   n9 = float * float(A=@n5.IconPaddingCorrection_Y, B=0.5)
   > SetPadding(
    self=SlotAsOverlaySlot(Widget=Get Main())
    InPadding_Left=Ease(Function=EaseInOut, Alpha=@n3, A=float + float(A=@n8, B=@n7), B=@n6, BlendExp=2) -> (Result=~)
    InPadding_Top=@n9
    InPadding_Right=float * float(A=float - float(A=1, B=@n4), B=@n6)
    InPadding_Bottom=@n9
   )
   n10 = Get!() -> (InputActionPaddingCorrection_X:double, InputActionPaddingCorrection_Y:double)
   n11 = float * float(A=@n10, B=0.5)
   n12 = Get!() -> (IconSize_X:double, IconSize_Y:double)
   n13 = Get!() -> (MainSlotPadding_Left:float=~, MainSlotPadding_Top:float=~, MainSlotPadding_Right:float=~, MainSlotPadding_Bottom:float=~)
   n14 = float * float(A=@n10.InputActionPaddingCorrection_Y, B=0.5)
   SetPadding(
    execute=@n2.then_1
    self=SlotAsOverlaySlot(Widget=Get Prompt())
    InPadding_Left=@n11
    InPadding_Top=Ease(Function=EaseInOut, Alpha=Get Value(), A=@n14, B=float + float(A=@n13.MainSlotPadding_Top, B=@n12.IconSize_Y), BlendExp=2) -> (Result=~)
    InPadding_Right=@n11
    InPadding_Bottom=@n14
   )
  ]}
 SetSlotPadding {nodes [
   SetSlotPadding() {is=title K2Node_FunctionEntry}
   n1 = Get!() -> (InputActionIconSizeReference_X:double, InputActionIconSizeReference_Y:double)
   n2 = Get!() -> (IconSize_X:double, IconSize_Y:double)
   n3 = float / float(A=float - float(A:int=Get MinHeight() -> (MinHeight:int=0), B="Max (Float)"(A=@n1.InputActionIconSizeReference_Y, B=@n2.IconSize_Y)), B=2)
   n4 = Get!() -> (ButtonPadding_Left:float=~, ButtonPadding_Top:float=~, ButtonPadding_Right:float=~, ButtonPadding_Bottom:float=~)
   n5 = float + float!(
    A="Max (Float)"(
     A=Get!() -> (InputActionIconSizeReference_X:double, InputActionIconSizeReference_Y:double).InputActionIconSizeReference_Y
     B=Get!() -> (IconSize_X:double, IconSize_Y:double).IconSize_Y
    )
    B=@n4.ButtonPadding_Top
    C:float=@n4.ButtonPadding_Bottom
   ) -> (ReturnValue)
   n6 = Get MinHeight() -> (MinHeight=0)
   > n7 = Set MainSlotPadding(MainSlotPadding:Margin Structure=Select(
     Option 0=Get ButtonPadding() -> (ButtonPadding:Margin Structure)
     Option 1=Make Margin(Left=@n3, Top=@n3, Right=@n3, Bottom=@n3)
     Index="float < float"(A=@n5, B:int=@n6)
    )) -> (Output_Get:Margin Structure)
   > SetGridSlotPadding!(execute, Widget=Get Slot_Main() -> (Slot_Main:Named Slot Object Reference), Padding:Margin Structure=@n7) -> (then)
  ]}
 SetupCollapsibleLayout {nodes [
   SetupCollapsibleLayout() {is=title K2Node_FunctionEntry}
   > n1 = AddChildToOverlay!(
    execute
    self:Overlay Object Reference=Get Overlay_Layout() -> (Overlay_Layout:Overlay Object Reference)
    Content:Widget Object Reference=Get Prompt()
   ) -> (then, ReturnValue:Overlay Slot Object Reference)
   > SetVerticalAlignment(self=@n1, InVerticalAlignment=VAlign_Top)
   > SetHorizontalAlignment(self=@n1, InHorizontalAlignment=HAlign_Left)
   > n2 = AddChildToOverlay(self=Get Overlay_Layout(), Content=Get Main())
   > SetVerticalAlignment(self=@n2, InVerticalAlignment=GetContentVerticalAlignment!() -> (VerticalAlignment:EVerticalAlignment Enum=VAlign_Fill))
   > SetHorizontalAlignment(self=@n2, InHorizontalAlignment=GetContentHorizontalAlignment!() -> (HorizontalAlignment:EHorizontalAlignment Enum=HAlign_Fill))
   > CalculatePaddingCorrection() {is=title K2Node_CallFunction}
  ]}
]
variables [
 ButtonType {type=E_UI_CTAButtonType, default=Regular}
 IconColorAndOpacityOverride {type=LinearColor, default=<rgba 0 0 0 1>}
 IconPaddingCorrection {type=Vector2D, default=<0 0>}
 InputActionIconSizeReference {type=Vector2D, default=<32 32>}
 InputActionPaddingCorrection {type=Vector2D, default=<0 0>}
 IsCollapsed {type=boolean}
 MainSlotPadding {type=Margin, default=<ltrb 8 8 8 8>}
 MaxTextWidth {type=number, default=224}
 OnCollapseAnimValue {type=number}
 QuietBackgroundMaterial {type=MaterialInstance, default=/FortUILibrary/UIKit/Buttons/MaterialInstances/CTAButtons/MI_UIKit_BgQuiet}
 QuietFontMaterial {type=MaterialInstance, default=/FortUILibrary/UIKit/Buttons/MaterialInstances/CTAButtons/MI_UIKit_TextQuiet}
 QuietSecondaryFontMaterial {type=MaterialInstance, default=&MI_UIKit_TextQuiet}
 RegularBackgroundMaterial {type=MaterialInstance, default=/FortUILibrary/UIKit/Buttons/MaterialInstances/CTAButtons/MI_UIKit_BgRegular}
 RegularFontMaterial {type=MaterialInstance, default=/FortUILibrary/UIKit/Buttons/MaterialInstances/CTAButtons/MI_UIKit_TextRegular}
 RegularSecondaryFontMaterial {type=MaterialInstance, default=&MI_UIKit_TextRegular}
 TextPaddingInputActionHidden {type=Margin, default=<ltrb 16 0 0 0>}
 TextPaddingInputActionVisible {type=Margin, default=<ltrb 6 0 0 0>}
]
widgets {
 info {
  inheritedWidgetCount=0
  namedSlotCount=1
  parentClass=/FortUILibrary/Systems/ModularButton/WBP_UIKit_ButtonGeneric_Base.WBP_UIKit_ButtonGeneric_Base_C
  rootWidgetClass=null
  widgetCount=1
 }
 widgets [Overlay_Layout {widgetClassPath=&Overlay, bIsVariable=true, widget=Overlay_Layout}]
}
