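# WBP_UIKit_Button_Generic is a generic UI Kit button widget derived from WBP_UIKit_ButtonGeneric_Base that adds hover and press animation feedback.
# Hover plays Anim_Bounce, a short vertical hop of the Main grid panel.
# Press plays Anim_Pressed, shrinking Root and Main slightly, then fires the OnPressed dispatcher; release plays that animation in reverse.
# It defines no widgets of its own, and its construct events only call the parent.
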WidgetBlueprint /FortUILibrary/UIKit/Buttons/WBP_UIKit_Button_Generic
animations [
 Anim_Bounce {
  binding_count=1
  length_seconds=0.2961
  object=&self:Anim_Bounce
  sequence {
   bindings [Main {
     id=#1
     possessed_class=GridPanel
     tracks [{
       type=MovieScene2DTransformTrack
       sections [{
         type=MovieScene2DTransformSection
         channels [
          Translation.X {type=MovieSceneScriptingFloatChannel}
          Translation.Y {type=MovieSceneScriptingFloatChannel, keys [[0,0],[4.423,-3],[13.268,0],[17.765,0]]}
         ]
        }]
      }]
    }]
   display_rate {numerator=60}
  }
 }
 Anim_Pressed {
  binding_count=2
  length_seconds=0.1
  object=&self:Anim_Pressed
  sequence {
   bindings [
    Root {
     id=#2
     possessed_class=GridPanel
     tracks [{
       type=MovieScene2DTransformTrack
       sections [{
         type=MovieScene2DTransformSection
         end_frame=6
         start_frame=0
         channels [
          Scale.X {type=MovieSceneScriptingFloatChannel, keys [[0,1],[6,0.975]]}
          Scale.Y {type=MovieSceneScriptingFloatChannel, keys [[0,1],[6,0.975]]}
         ]
        }]
      }]
    }
    Main {
     id=#3
     possessed_class=GridPanel
     tracks [{
       type=MovieScene2DTransformTrack
       sections [{
         type=MovieScene2DTransformSection
         channels [
          Scale.X {type=MovieSceneScriptingFloatChannel, keys [[0,1],[6,0.95]]}
          Scale.Y {type=MovieSceneScriptingFloatChannel, keys [[0,1],[6,0.975]]}
         ]
        }]
      }]
    }
   ]
   display_rate {numerator=60}
  }
 }
]
graphs [
 EventGraph {nodes [
   n1 = Event PreConstruct()
   > Parent: PreConstruct(IsDesignTime=@n1.IsDesignTime)
   Event Construct()
   > Parent: Construct()
   Event On Hovered()
   > Parent: On Hovered() {is=title K2Node_CallParentFunction}
   > PlayAnimationForward(InAnimation=Get Anim_Bounce() -> (Anim_Bounce:Widget Animation Object Reference))
   Event On Pressed() {is=title K2Node_Event}
   > Parent: On Pressed() {is=title K2Node_CallParentFunction}
   n2 = Get Anim_Pressed() -> (Anim_Pressed:Widget Animation Object Reference)
   > PlayAnimationForward(InAnimation=@n2)
   > Call On Pressed!(execute:Exec, self:WBP UIKit Button Generic Object Reference) -> (then:Exec) {is=title K2Node_CallDelegate}
   Event On Released() {is=title K2Node_Event}
   > Parent: On Released() {is=title K2Node_CallParentFunction}
   > PlayAnimationReverse(InAnimation=@n2)
  ]}
 OnPressed {nodes [OnPressed() {is=title K2Node_FunctionEntry}]}
]
variables [OnPressed {}]
widgets {
 info {
  inheritedWidgetCount=0
  parentClass=/FortUILibrary/Systems/ModularButton/WBP_UIKit_ButtonGeneric_Base.WBP_UIKit_ButtonGeneric_Base_C
  rootWidgetClass=null
  widgetCount=0
 }
 widgets []
}
