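# WBP_UIKit_Button_Loud is a UI Kit button widget, a child of WBP_UIKit_ButtonCTA_Base that adds no widgets or variables of its own.
# Its only logic handles the InputActionWidget's On Input Icon Updated event by calling Set Icon Dark Mode with IsDark set to true.
# This switches the input icon to dark mode whenever that icon updates.
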WidgetBlueprint /FortUILibrary/UIKit/Buttons/WBP_UIKit_Button_Loud
graphs [EventGraph {nodes [
   "On Input Icon Updated (InputActionWidget)"() {is=title K2Node_ComponentBoundEvent}
   > Set Icon Dark Mode!(execute, self:WBP UIKit Button CTA Base Object Reference, IsDark:bool=true) -> (then) {is=title K2Node_CallFunction}
  ]}]
variables []
widgets {
 info {
  inheritedWidgetCount=0
  parentClass=/FortUILibrary/Systems/ModularButton/WBP_UIKit_ButtonCTA_Base.WBP_UIKit_ButtonCTA_Base_C
  rootWidgetClass=null
  widgetCount=0
 }
 widgets []
}
