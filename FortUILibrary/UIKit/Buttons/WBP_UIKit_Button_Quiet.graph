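# WBP_UIKit_Button_Quiet is a UIKit button derived from the CTA button base that adapts its look to the current input type.
# HideBackground collapses the background and applies a prompt font on gamepad, or shows it with uppercase text for mouse and touch, then refreshes padding and size.
# Icon dark mode also follows input type.
# Non-gamepad presses play an animation and, gated by AudioGate after a delay, choose hold or press sounds.
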WidgetBlueprint /FortUILibrary/UIKit/Buttons/WBP_UIKit_Button_Quiet
graphs [
 EventGraph {nodes [
   n1 = "On Input Icon Updated (InputActionWidget)"() {is=title K2Node_ComponentBoundEvent}
   Set Icon Dark Mode!(
    execute=[@n11.then, @n1.then]
    self:WBP UIKit Button CTA Base Object Reference
    IsDark:bool=Select!(
     MouseAndKeyboard:bool=true
     Gamepad:bool=false
     Touch:bool=true
     Count:bool=false
     Index:ECommonInputType Enum=Get CurrentInputType() -> (CurrentInputType:ECommonInputType Enum)
    ) -> (ReturnValue:bool)
   ) -> (then) {is=title K2Node_CallFunction}
   n2 = Event PreConstruct()
   > Set ButtonSize(ButtonSize:E_UI_ButtonSize Enum=NewEnumerator2) -> (Output_Get:E_UI_ButtonSize Enum=NewEnumerator0)
   > n3 = Parent: PreConstruct(IsDesignTime=@n2.IsDesignTime)
   n4 = Get CommonInputSubsystem!() -> (ReturnValue:Common Input Object Reference) {is=title K2Node_GetBlueprintContextEx}
   Branch(execute=[@n3.then, @n12.then], Condition=IsValid(Object=@n4))
   n5 = GetCurrentInputType(self=@n4)
   > HideBackground!(execute, self, InputType:ECommonInputType Enum=@n5) -> (then)
   > Set IsDarkMode(IsDarkMode:bool=Select(MouseAndKeyboard=true, Gamepad=false, Touch=true, Count=false, Index=@n5)) -> (Output_Get:bool)
   Event On Pressed() {is=title K2Node_Event}
   > n6 = Branch(Condition="Equal (Enum)"(A=Get CurrentInputType(), B=Gamepad))
   > Set AudioGate(AudioGate:bool=true) -> (Output_Get:bool)
   > PlayAnimationForward(InAnimation=Get Anim_Pressed_Base() -> (Anim_Pressed_Base:Widget Animation Object Reference))
   > n7 = Branch(Condition=ShouldPlayHoldSound!() -> (Holdable:bool))
   > Delay(Duration=0.2)
   > n8 = Branch(Condition=Get AudioGate() -> (AudioGate:bool))
   > PlaySoundHoldPressed(self:WBP UIKit Button CTA Base Object Reference)
   PlaySoundPressed(execute=[@n8.else, @n7.else], self:WBP UIKit Button CTA Base Object Reference)
   Parent: On Pressed(execute=@n6.else) {is=title K2Node_CallParentFunction}
   Event On Released() {is=title K2Node_Event}
   > n9 = Branch(Condition="Equal (Enum)"(A=Get CurrentInputType(), B=Gamepad))
   > Set AudioGate(AudioGate=false)
   > PlayAnimationReverse(InAnimation=Get Anim_Pressed_Base(), PlaybackSpeed=2)
   n10 = Get HoldSoundAudioComponent() -> (HoldSoundAudioComponent:Audio Component Object Reference)
   > Is Valid(InputObject=@n10)
   > Branch(Condition=IsPlaying!(self:Audio Component Object Reference=@n10) -> (ReturnValue:bool))
   > StopSoundHoldReleased(self:WBP UIKit Button CTA Base Object Reference)
   Parent: On Released(execute=@n9.else) {is=title K2Node_CallParentFunction}
   Event Switch Input() {is=title K2Node_Event}
   > Parent: Switch Input() {is=title K2Node_CallParentFunction}
   > n11 = HideBackground(InputType=Get CurrentInputType())
   Event Construct()
   > n12 = Parent: Construct()
   IsHeldAction!(self:Common Action Widget Object Reference=Get InputActionWidget() -> (InputActionWidget:Common Action Widget Object Reference)) -> (ReturnValue:bool)
  ]}
 HideBackground {nodes [
   HideBackground!() -> (then, InputType:ECommonInputType Enum) {is=title K2Node_FunctionEntry}
   > Branch(Condition=Get UseHideBackground() -> (UseHideBackground:bool))
   > n1 = Branch(Condition=NOT Boolean(A=Get HideBackgroundAllInputTypes() -> (HideBackgroundAllInputTypes:bool)))
   > n2 = Switch on ECommonInputType!(execute, Selection:ECommonInputType Enum=Get InputType() -> (InputType:ECommonInputType Enum)) -> (MouseAndKeyboard:Exec, Gamepad:Exec, Touch:Exec, Count:Exec) {is=title K2Node_SwitchEnum}
   SetVisibility(
    execute=[@n2.Gamepad, @n1.else]
    self=Get Block_Background() -> (Block_Background:WBP UIKit Block Background With Outline Object Reference)
    InVisibility=Collapsed
   )
   n3 = Get Block_Text() -> (Block_Text:WBP UIKit Block Text Object Reference)
   > SetFont!(
    execute
    self:WBP UIKit Block Text Object Reference=@n3
    Font:"Slate Font Info Structure (by ref)"=Get GamepadPromptFont() -> (GamepadPromptFont:Slate Font Info Structure)
   ) -> (then)
   > n4 = SetTransformPolicy!(execute, self:WBP UIKit Block Text Object Reference=@n3, Transform Policy:ETextTransformPolicy Enum=None) -> (then)
   UpdateButtonPadding(execute=[@n4.then, @n5.then], self:WBP UIKit Button CTA Base Object Reference)
   > UpdateButtonSIze(self:WBP UIKit Button CTA Base Object Reference)
   > ResetButtonState(self:WBP UIKit Button CTA Base Object Reference)
   SetVisibility(execute=[@n2.MouseAndKeyboard, @n2.Touch], self=Get Block_Background(), InVisibility=Visible)
   > SetFont(self=@n3, Font=Get SmallPrimaryFont() -> (SmallPrimaryFont:Slate Font Info Structure))
   > n5 = SetTransformPolicy(self=@n3, Transform Policy=ToUpper)
  ]}
 UpdateButtonPadding {nodes [
   UpdateButtonPadding() {is=title K2Node_FunctionEntry}
   n1 = Get CurrentInputType()
   n2 = "Equal (Enum)"(A=@n1, B=Touch)
   n3 = NOT Boolean(A=Get HasText() -> (HasText:bool))
   n4 = Get HasIcon() -> (HasIcon:bool)
   > SetSize!(
    execute
    self:Spacer Object Reference=[Get Spacer_BottomRight() -> (Spacer_BottomRight:Spacer Object Reference), Get Spacer_TopLeft() -> (Spacer_TopLeft:Spacer Object Reference)]
    InSize_X:double=SelectFloat(
     A=SelectFloat(A=8, B=0, bPickA=@n2)
     B=Select(MouseAndKeyboard:double=24, Gamepad:double=8, Touch:double=24, Count:double=0, Index=@n1) -> (ReturnValue:double)
     bPickA=AND Boolean(A=@n4, B=@n3)
    )
    InSize_Y:double=0
   ) -> (then)
  ]}
 UpdateButtonSIze {nodes [
   UpdateButtonSIze() {is=title K2Node_FunctionEntry}
   > n1 = Branch(Condition=IsMobileGame())
   n2 = Select!(
    Option 0:int=Get MobileSmallHeight() -> (MobileSmallHeight:int=0)
    Option 1:int=Get MobileMinHeight() -> (MobileMinHeight:int=0)
    Index:bool=Get OverrideDefaultHeight() -> (OverrideDefaultHeight:bool)
   ) -> (ReturnValue:int)
   n3 = Get CurrentInputType()
   n4 = Get MobileMinWidth() -> (MobileMinWidth:int=0)
   > SetMinDimensions!(
    execute
    self:Common Button Base Object Reference
    InMinWidth:int=SelectInt(
     A=Select!(MouseAndKeyboard:int=@n4, Gamepad:int=@n2, Touch:int=@n4, Count:int=0, Index:ECommonInputType Enum=@n3) -> (ReturnValue)
     B=@n2
     bPickA=Get HasText()
    )
    InMinHeight:int=@n2
   ) -> (then)
   n5 = Get MinWidth() -> (MinWidth:int=0)
   n6 = Get CurrentInputType()
   n7 = Get SmallHeight() -> (SmallHeight:int=0)
   SetMinDimensions(
    execute=@n1.else
    InMinWidth=SelectInt(A=Select(MouseAndKeyboard=@n5, Gamepad=@n7, Touch=@n5, Count=0, Index=@n6), B=@n7, bPickA=Get HasText())
    InMinHeight=@n7
   )
  ]}
 UpdateFonts {nodes [
   UpdateFonts() {is=title K2Node_FunctionEntry}
   n1 = Get CurrentInputType()
   n2 = Get Block_Text()
   > SetFont(
    self=@n2
    Font=Select!(
     Option 0:"Slate Font Info Structure (by ref)"=Get SmallPrimaryFont()
     Option 1:"Slate Font Info Structure (by ref)"=Get GamepadPromptFont()
     Index:bool="Equal (Enum)"(A=@n1, B=Gamepad)
    ) -> (ReturnValue:"Slate Font Info Structure (by ref)")
   )
   > SetMobileFontSizeMultiplier!(
    execute
    self:WBP UIKit Block Text Object Reference=@n2
    MobileFontSizeMultiplier:double=Select(
     Option 0:double=Get MobileFontSizeMultiplierSmall() -> (MobileFontSizeMultiplierSmall:double)
     Option 1:double=Get MobileFontSizeMultiplierOverride() -> (MobileFontSizeMultiplierOverride:double)
     Index=Get OverrideDefaultMobileFontSizeMultiplier() -> (OverrideDefaultMobileFontSizeMultiplier:bool)
    ) -> (ReturnValue:double)
   ) -> (then)
  ]}
]
variables [
 AudioGate {type=boolean}
 GamepadPromptFont {
  type=SlateFontInfo
  default {
   bForceMonospaced=false
   bMaterialIsStencil=false
   fontMaterial=/FortUILibrary/UIKit/Buttons/MaterialInstances/CTAButtons/MI_UIKit_TextQuiet_Prompt
   fontObject=&HeadingNow
   letterSpacing=10
   monospacedWidth=1
   size=15.75
   skewAmount=0
   typefaceFontName=76 Bold
   outlineSettings {
    bApplyOutlineToDropShadows=false
    bMiteredCorners=false
    bSeparateFillAlpha=false
    outlineMaterial=null
    outlineSize=0
    outlineColor=<rgba 0 0 0 1>
   }
  }
 }
 HideBackgroundAllInputTypes {type=boolean}
 UseHideBackground {type=boolean, default=true}
]
widgets {
 info {
  inheritedWidgetCount=0
  parentClass=/FortUILibrary/Systems/ModularButton/WBP_UIKit_ButtonCTA_Base.WBP_UIKit_ButtonCTA_Base_C
  rootWidgetClass=null
  widgetCount=0
 }
 widgets []
}
