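# UIKit CTA button widget derived from WBP_UIKit_ButtonCTA_Base, adding Regular and Quiet style variants without declaring widgets of its own.
# SetButtonType selects background, border, primary and secondary font materials by ButtonType.
# OverrideFontMaterials rebuilds the six large, medium and small font structs with those materials.
# Styling runs before the parent's PreConstruct and on widget styling when EnableOnWidgetStyled is set; input icon updates force dark mode.
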WidgetBlueprint /FortUILibrary/UIKit/Buttons/WBP_UIKit_Button_Regular
graphs [
 EventGraph {nodes [
   "On Input Icon Updated (InputActionWidget)"() {is=title K2Node_ComponentBoundEvent}
   > Set Icon Dark Mode!(execute, self:WBP UIKit Button CTA Base Object Reference, IsDark:bool=true) -> (then) {is=title K2Node_CallFunction}
   n1 = Event PreConstruct()
   > SetButtonType!(execute, self, ButtonType:E_UI_CTAButtonType Enum=Get ButtonType() -> (ButtonType:E_UI_CTAButtonType Enum=NewEnumerator0)) -> (then)
   > Parent: PreConstruct(IsDesignTime=@n1.IsDesignTime)
   Event On Widget Styled() {is=title K2Node_Event}
   > Branch(Condition=Get EnableOnWidgetStyled() -> (EnableOnWidgetStyled:bool))
   > SetButtonType(ButtonType=Get ButtonType() -> (ButtonType=NewEnumerator0))
   > Parent: On Widget Styled() {is=title K2Node_CallParentFunction}
  ]}
 OverrideFontMaterials {nodes [
   n1 = OverrideFontMaterials!() -> (then, PrimaryFontMaterial:Material Instance Object Reference, SecondaryFontMaterial:Material Instance Object Reference) {is=title K2Node_FunctionEntry}
   > n2 = Sequence!(execute) -> (then_0, then_1, then_2:Exec, then_3:Exec, then_4:Exec, then_5:Exec)
   n3 = Break Slate Font Info!(SlateFontInfo:"Slate Font Info Structure (by ref)"=Get LargePrimaryFont() -> (LargePrimaryFont:Slate Font Info Structure)) -> (
    FontObject:Object Reference
    FontMaterial:Object Reference
    OutlineSettings:Font Outline Settings Structure
    TypefaceFontName:Name
    Size:float=~
    LetterSpacing:int
    SkewAmount:float=~
    bForceMonospaced:bool=~
    bMaterialIsStencil:bool=~
    MonospacedWidth:float=~
   ) {is=title K2Node_BreakStruct}
   > Set LargePrimaryFont(LargePrimaryFont:Slate Font Info Structure=MakeSlateFontInfo!(
     FontObject:Object Reference=@n3
     FontMaterial:Object Reference=@n1
     OutlineSettings:Font Outline Settings Structure=@n3.OutlineSettings
     TypefaceFontName:Name=@n3.TypefaceFontName
     Size:float=@n3.Size
     LetterSpacing:int=@n3.LetterSpacing
     SkewAmount:float=@n3.SkewAmount
     bForceMonospaced:bool=false
     bMaterialIsStencil:bool=false
     MonospacedWidth:float=1
    ) -> (ReturnValue:Slate Font Info Structure)) -> (Output_Get:Slate Font Info Structure)
   n4 = Break Slate Font Info(SlateFontInfo=Get SmallSecondaryFont() -> (SmallSecondaryFont:Slate Font Info Structure)) -> (Size=~, SkewAmount=~, bForceMonospaced=~, bMaterialIsStencil=~, MonospacedWidth=~)
   Set SmallSecondaryFont(
    execute=@n2.then_5
    SmallSecondaryFont:Slate Font Info Structure=MakeSlateFontInfo(
     FontObject=@n4
     FontMaterial=@n1.SecondaryFontMaterial
     OutlineSettings=@n4.OutlineSettings
     TypefaceFontName=@n4.TypefaceFontName
     Size=@n4.Size
     LetterSpacing=@n4.LetterSpacing
     SkewAmount=@n4.SkewAmount
     bForceMonospaced=false
     bMaterialIsStencil=false
     MonospacedWidth=1
    )
   ) -> (Output_Get:Slate Font Info Structure)
   n5 = Break Slate Font Info(SlateFontInfo=Get MeduimSecondaryFont() -> (MeduimSecondaryFont:Slate Font Info Structure)) -> (Size=~, SkewAmount=~, bForceMonospaced=~, bMaterialIsStencil=~, MonospacedWidth=~)
   Set MeduimSecondaryFont(
    execute=@n2.then_4
    MeduimSecondaryFont:Slate Font Info Structure=MakeSlateFontInfo(
     FontObject=@n5
     FontMaterial=@n1.SecondaryFontMaterial
     OutlineSettings=@n5.OutlineSettings
     TypefaceFontName=@n5.TypefaceFontName
     Size=@n5.Size
     LetterSpacing=@n5.LetterSpacing
     SkewAmount=@n5.SkewAmount
     bForceMonospaced=false
     bMaterialIsStencil=false
     MonospacedWidth=1
    )
   ) -> (Output_Get:Slate Font Info Structure)
   n6 = Break Slate Font Info(SlateFontInfo=Get LargeSecondaryFont() -> (LargeSecondaryFont:Slate Font Info Structure)) -> (Size=~, SkewAmount=~, bForceMonospaced=~, bMaterialIsStencil=~, MonospacedWidth=~)
   Set LargeSecondaryFont(
    execute=@n2.then_3
    LargeSecondaryFont:Slate Font Info Structure=MakeSlateFontInfo(
     FontObject=@n6
     FontMaterial=@n1.SecondaryFontMaterial
     OutlineSettings=@n6.OutlineSettings
     TypefaceFontName=@n6.TypefaceFontName
     Size=@n6.Size
     LetterSpacing=@n6.LetterSpacing
     SkewAmount=@n6.SkewAmount
     bForceMonospaced=false
     bMaterialIsStencil=false
     MonospacedWidth=1
    )
   ) -> (Output_Get:Slate Font Info Structure)
   n7 = Break Slate Font Info(SlateFontInfo=Get MediumPrimaryFont() -> (MediumPrimaryFont:Slate Font Info Structure)) -> (Size=~, SkewAmount=~, bForceMonospaced=~, bMaterialIsStencil=~, MonospacedWidth=~)
   Set MediumPrimaryFont(
    execute=@n2.then_1
    MediumPrimaryFont:Slate Font Info Structure=MakeSlateFontInfo(
     FontObject=@n7
     FontMaterial=@n1
     OutlineSettings=@n7.OutlineSettings
     TypefaceFontName=@n7.TypefaceFontName
     Size=@n7.Size
     LetterSpacing=@n7.LetterSpacing
     SkewAmount=@n7.SkewAmount
     bForceMonospaced=false
     bMaterialIsStencil=false
     MonospacedWidth=1
    )
   ) -> (Output_Get:Slate Font Info Structure)
   n8 = Break Slate Font Info(SlateFontInfo=Get SmallPrimaryFont() -> (SmallPrimaryFont:Slate Font Info Structure)) -> (Size=~, SkewAmount=~, bForceMonospaced=~, bMaterialIsStencil=~, MonospacedWidth=~)
   Set SmallPrimaryFont(
    execute=@n2.then_2
    SmallPrimaryFont:Slate Font Info Structure=MakeSlateFontInfo(
     FontObject=@n8
     FontMaterial=@n1
     OutlineSettings=@n8.OutlineSettings
     TypefaceFontName=@n8.TypefaceFontName
     Size=@n8.Size
     LetterSpacing=@n8.LetterSpacing
     SkewAmount=@n8.SkewAmount
     bForceMonospaced=false
     bMaterialIsStencil=false
     MonospacedWidth=1
    )
   ) -> (Output_Get:Slate Font Info Structure)
  ]}
 SetButtonType {nodes [
   n1 = SetButtonType!() -> (then, ButtonType:E_UI_CTAButtonType Enum=NewEnumerator0) {is=title K2Node_FunctionEntry}
   > n2 = Set ButtonType(ButtonType:E_UI_CTAButtonType Enum=@n1) -> (Output_Get:E_UI_CTAButtonType Enum=NewEnumerator0)
   > Set BackgroundMaterial(BackgroundMaterial:Material Instance Object Reference=Select!(
     NewEnumerator0:Material Instance Object Reference=Get RegularBackgroundMaterial() -> (RegularBackgroundMaterial:Material Instance Object Reference)
     NewEnumerator1:Material Instance Object Reference=Get QuietBackgroundMaterial() -> (QuietBackgroundMaterial:Material Instance Object Reference)
     Index:E_UI_CTAButtonType Enum=@n2
    ) -> (ReturnValue:Material Instance Object Reference)) -> (Output_Get:Material Instance Object Reference)
   > Set BorderMaterial(BorderMaterial:Material Instance Object Reference=Select(
     NewEnumerator0=Get RegularOutlineMaterial() -> (RegularOutlineMaterial:Material Instance Object Reference)
     NewEnumerator1=Get QuietOutlineMaterial() -> (QuietOutlineMaterial:Material Instance Object Reference)
     Index=@n2
    )) -> (Output_Get:Material Instance Object Reference)
   > OverrideFontMaterials!(
    execute
    self
    PrimaryFontMaterial:Material Instance Object Reference=Select(
     NewEnumerator0=Get RegularFontMaterial() -> (RegularFontMaterial:Material Instance Object Reference)
     NewEnumerator1=Get QuietFontMaterial() -> (QuietFontMaterial:Material Instance Object Reference)
     Index=@n2
    )
    SecondaryFontMaterial:Material Instance Object Reference=Select(
     NewEnumerator0=Get RegularSecondaryFontMaterial() -> (RegularSecondaryFontMaterial:Material Instance Object Reference)
     NewEnumerator1=Get QuietSecondaryFontMaterial() -> (QuietSecondaryFontMaterial:Material Instance Object Reference)
     Index=@n2
    )
   ) -> (then) {is=title K2Node_CallFunction}
  ]}
]
variables [
 ButtonType {type=E_UI_CTAButtonType, default=Regular}
 QuietBackgroundMaterial {type=MaterialInstance, default=/FortUILibrary/UIKit/Buttons/MaterialInstances/CTAButtons/MI_UIKit_BgQuiet}
 QuietFontMaterial {type=MaterialInstance, default=/FortUILibrary/UIKit/Buttons/MaterialInstances/CTAButtons/MI_UIKit_TextQuiet}
 QuietOutlineMaterial {type=MaterialInstance, default=/FortUILibrary/UIKit/Buttons/MaterialInstances/TileOutline/MI_UIKit_Outline}
 QuietSecondaryFontMaterial {type=MaterialInstance, default=&MI_UIKit_TextQuiet}
 RegularBackgroundMaterial {type=MaterialInstance, default=/FortUILibrary/UIKit/Buttons/MaterialInstances/CTAButtons/MI_UIKit_BgRegular}
 RegularFontMaterial {type=MaterialInstance, default=/FortUILibrary/UIKit/Buttons/MaterialInstances/CTAButtons/MI_UIKit_TextRegular}
 RegularOutlineMaterial {type=MaterialInstance, default=&MI_UIKit_Outline}
 RegularSecondaryFontMaterial {type=MaterialInstance, default=&MI_UIKit_TextRegular}
]
widgets {
 info {
  inheritedWidgetCount=0
  parentClass=/FortUILibrary/Systems/ModularButton/WBP_UIKit_ButtonCTA_Base.WBP_UIKit_ButtonCTA_Base_C
  rootWidgetClass=null
  widgetCount=0
 }
 widgets []
}
