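# UIKit tab button widget, a child of WBP_UIKit_Button_Selectable with no widgets of its own.
# Its variables hold settings for bang type and layout, a tutorial name, and a status indicator class defaulting to WBP_UIKit_StatusIndicator in Dot, Attention style.
# The event graph only forwards PreConstruct and Construct to the parent, so nothing in this graph reads those settings.
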WidgetBlueprint /FortUILibrary/UIKit/Buttons/WBP_UIKit_Button_Tab_Enhanced
graphs [EventGraph {nodes [
   n1 = Event PreConstruct()
   > Parent: PreConstruct(IsDesignTime=@n1.IsDesignTime)
   Event Construct()
   > Parent: Construct()
  ]}]
variables [
 BangHorizontalAlignment {type=EHorizontalAlignment, default=HAlign_Left}
 BangNudge {type=Vector2D, default=<0 0>}
 BangPadding {type=Margin, default=<ltrb 4 4 4 4>}
 BangType {type=EFortBangType, default=Invalid}
 BangVerticalAlignment {type=EVerticalAlignment, default=VAlign_Center}
 Configuration_0 {type=E_UI_StatusIndicator_Configuration, default=Dot}
 EnableCustomLayout {type=boolean, default=true}
 StatusIndicatorClass_0 {type=Class, default=/FortUILibrary/UIKit/StatusIndicator/WBP_UIKit_StatusIndicator.WBP_UIKit_StatusIndicator_C}
 TutorialName {type=string, default=None}
 Type_0 {type=E_UI_StatusIndicator_Type, default=Attention}
]
widgets {
 info {
  inheritedWidgetCount=0
  parentClass=&/WBP_UIKit_Button_Selectable.WBP_UIKit_Button_Selectable_C
  rootWidgetClass=null
  widgetCount=0
 }
 widgets []
}
