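# UIKit toggle switch button built on the generic modular button base, selectable and clickable again to deselect.
# Pre-construct registers switcher and outline blocks, applies shared block timing, and re-parents the prompt and icon into the switcher's grid.
# On gamepad, the icon hides while an input prompt is visible, updating on input-method and enabled-state changes.
# Press, or click with a visible gamepad prompt, plays an on/off sound based on selection.
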WidgetBlueprint /FortUILibrary/UIKit/Buttons/WBP_UIKit_SwitcherButton
graphs [
 EventGraph {nodes [
   n1 = Event PreConstruct()
   > Parent: PreConstruct(IsDesignTime=@n1.IsDesignTime)
   > Assign On Input Method Changed!(execute:Exec, self:Common Input Object Reference=Get CommonInputSubsystem(), Delegate:"Delegate (by ref)"=Create Event()) -> (then:Exec) {is=title K2Node_AssignDelegate}
   > SetComponents()
   > SetBlockTiming()
   > UpdateSwitcherIcon()
   > SetIsSelected!(execute, self:Common Button Base Object Reference, InSelected:bool=Get Selected() -> (Selected:bool), bGiveClickFeedback:bool=false) -> (then)
   Event Construct()
   > Parent: Construct()
   > UpdateIcon(self:WBP UIKit Button Generic Base Object Reference)
   > OnInputTypeChanged!(execute, self, bNewInputType:ECommonInputType Enum=GetCurrentInputType(self=Get CommonInputSubsystem())) -> (then)
   > Set IsInstantSelection(IsInstantSelection:bool=false) -> (Output_Get:bool)
   > SetComponents()
   Event On Enabled() {is=title K2Node_Event}
   > Parent: On Enabled() {is=title K2Node_CallParentFunction}
   n2 = "Equal (Enum)"(A=GetCurrentInputType(self=Get CommonInputSubsystem()), B=Gamepad)
   > Branch(Condition=@n2)
   > SetVisibility(
    self=Get Block_Icon() -> (Block_Icon:WBP UIKit Block Image Object Reference)
    InVisibility=Select(Option 0=SelfHitTestInvisible, Option 1=Hidden, Index=IsPromptVisible!() -> (IsPromptVisible:bool))
   )
   Event On Selected()
   > Parent: On Selected() {is=title K2Node_CallParentFunction}
   > SetInputIconDarkMode!(execute, self:WBP UIKit Button Generic Base Object Reference, UseDark:bool=true) -> (then)
   Event On Deselected()
   > n3 = Branch(Condition=IsHoveredAndGamepadInput!() -> (ReturnValue:bool))
   > SetBlocksIsSelected!(execute, self:WBP UIKit Button Generic Base Object Reference, IsSelected:bool=false) -> (then)
   > n4 = RequestTransitionOnBlocks!(execute, self:UIKit Modular Button Object Reference, State:EUIKitBlockVisualState Enum=DeselectedFocused) -> (then)
   SetInputIconDarkMode(execute=[@n4.then, @n5.then], UseDark=false)
   n5 = Parent: On Deselected(execute=@n3.else) {is=title K2Node_CallParentFunction}
   Event On Disabled() {is=title K2Node_Event}
   > Parent: On Disabled() {is=title K2Node_CallParentFunction}
   > Branch(Condition=@n2)
   > SetVisibility(self=Get Block_Icon(), InVisibility=SelfHitTestInvisible)
   Event On Clicked()
   n6 = GetCurrentInputType(self=Get CommonInputSubsystem())
   > n7 = Branch(Condition=AND Boolean(A=IsPromptVisible(), B="Equal (Enum)"(A=@n6, B=Gamepad)))
   n8 = Branch(execute=[@n7.then, @n9.then], Condition=GetSelected!(self:Common Button Base Object Reference) -> (ReturnValue:bool))
   > PlaySound2D(Sound=/Game/Sounds/Fort_UI_Sounds/FrontEnd/Equip_Claim_Buttons/MSS_UI_Toggle_Button_Off)
   PlaySound2D(execute=@n8.else, Sound=/Game/Sounds/Fort_UI_Sounds/FrontEnd/Equip_Claim_Buttons/MSS_UI_Toggle_Button)
   n9 = Event On Pressed() {is=title K2Node_Event}
  ]}
 IsHoveredAndGamepadInput {nodes [
   IsHoveredAndGamepadInput() {is=title K2Node_FunctionEntry}
   n1 = GetCurrentInputType(self=Get CommonInputSubsystem())
   > Return Node!(execute, ReturnValue:bool=AND Boolean(A=IsHovered(), B="Equal (Enum)"(A=@n1, B=Gamepad)))
  ]}
 IsPromptVisible {nodes [
   IsPromptVisible() {is=title K2Node_FunctionEntry}
   n1 = OR Boolean(
    A=Get bRequiresHold() -> (bRequiresHold:bool)
    B=IsHeldAction!(self:Common Action Widget Object Reference=Get InputActionWidget() -> (InputActionWidget:Common Action Widget Object Reference)) -> (ReturnValue:bool)
   )
   n2 = "Not Equal (Boolean)"(A=GetInputAction!(self:Common Button Base Object Reference) -> (InputActionRow:Data Table Row Handle Structure, ReturnValue:bool).ReturnValue) -> (ReturnValue=~)
   > Return Node!(execute, IsPromptVisible:bool=AND Boolean(A=OR Boolean(A=@n2, B=@n1), B=GetIsEnabled!(self:Widget Object Reference) -> (ReturnValue:bool)))
  ]}
 OnInputTypeChanged {nodes [
   n1 = OnInputTypeChanged!() -> (then, bNewInputType:ECommonInputType Enum) {is=title K2Node_FunctionEntry}
   n2 = IsPromptVisible() {is=title K2Node_CallFunction}
   n3 = "Equal (Enum)"(A=@n1, B=Gamepad)
   > SetVisibility(self=Get Block_Icon(), InVisibility=Select(Option 0=SelfHitTestInvisible, Option 1=Hidden, Index=AND Boolean(A=@n3, B=@n2)))
  ]}
 Set Language Direction {nodes [Set Language Direction() {is=title K2Node_FunctionEntry}]}
 SetBlockTiming {nodes [
   SetBlockTiming() {is=title K2Node_FunctionEntry}
   > SetBlocksTiming!(
    execute
    self:UIKit Modular Button Object Reference
    Timing:"UIKit Block Timing Structure (by ref)"=Get BlocksTiming() -> (BlocksTiming:UIKit Block Timing Structure)
   ) -> (then)
   > SetBlockTiming!(
    execute
    self:UIKit Block Object Reference=Get Block_Switcher() -> (Block_Switcher:WBP UIKit Block Switcher Object Reference)
    Timing:"UIKit Block Timing Structure (by ref)"=Get BlocksTiming()
   ) -> (then) {is=title K2Node_CallFunction}
   > SetBlockTiming(self=Get Block_Outline() -> (Block_Outline:WBP UIKit Block Outline Object Reference), Timing=Get BlocksTiming())
  ]}
 SetComponents {nodes [
   SetComponents() {is=title K2Node_FunctionEntry}
   > n1 = AddChildToGrid!(
    execute
    self:Grid Panel Object Reference=Get Switcher() -> (Switcher:Grid Panel Object Reference)
    Content:Widget Object Reference=Get Prompt() -> (Prompt:Grid Panel Object Reference)
    InRow:int=0
    InColumn:int=0
   ) -> (then, ReturnValue:Grid Slot Object Reference)
   > SetPadding!(
    execute
    self:Grid Slot Object Reference=@n1
    InPadding_Left:float=-8
    InPadding_Top:float=-8
    InPadding_Right:float=-8
    InPadding_Bottom:float=-8
   ) -> (then)
   > SetLayer!(execute, self:Grid Slot Object Reference=@n1, InLayer:int=3) -> (then)
   > RegisterModularBlock!(execute, self:UIKit Modular Button Object Reference, Block:UIKit Block Object Reference=Get Block_Switcher()) -> (then)
   > SetToggleType()
   > SetVerticalAlignment(
    self:Grid Slot Object Reference=SlotAsGridSlot(Widget=Get Block_Text() -> (Block_Text:WBP UIKit Block Text Object Reference))
    InVerticalAlignment=VAlign_Center
   )
   > RegisterModularBlock(Block=Get Block_Outline())
   > SetVerticalAlignment(self=SlotAsGridSlot(Widget=Get Slot_Main() -> (Slot_Main:Named Slot Object Reference)), InVerticalAlignment=VAlign_Center)
  ]}
 SetToggleType {nodes [
   SetToggleType() {is=title K2Node_FunctionEntry}
   > SetIsToggleable!(execute, self:Common Button Base Object Reference, bInIsToggleable:bool=true) -> (then)
   > SetIsInteractableWhenSelected!(execute, self:Common Button Base Object Reference, bInInteractableWhenSelected:bool=true) -> (then)
  ]}
 UpdateSwitcherIcon {nodes [
   UpdateSwitcherIcon() {is=title K2Node_FunctionEntry}
   > SetMaterial!(
    execute
    self:WBP UIKit Block Switcher Object Reference=Get Block_Switcher()
    InMaterial:"Material Instance Object Reference (by ref)"=Get SwitcherBockgroundmaterial() -> (SwitcherBockgroundmaterial:Material Instance Object Reference)
   ) -> (then)
   > SetIconMaterial!(
    execute
    self:WBP UIKit Button Generic Base Object Reference
    Material Instance:Material Instance Object Reference=Get IconMaterial() -> (IconMaterial:Material Instance Object Reference)
   ) -> (then)
   > n1 = Get!(execute:Exec) -> (then:Exec, else:Exec, Block_Icon:WBP UIKit Block Image Object Reference)
   > n2 = AddChildToGrid(self=Get Switcher(), Content=@n1, InRow=0, InColumn=0)
   > SetPadding!(execute, self=@n2, InPadding:Margin Structure) -> (then)
   > SetLayer(self=@n2, InLayer=0)
   > SetIconSize!(execute, self:WBP UIKit Button Generic Base Object Reference, Icon:Vector 2D Structure=(X=28,Y=28)) -> (then)
   > SetGridSlotAlignment!(
    execute
    Widget:Widget Object Reference=Get Block_Icon()
    Horizontal:EHorizontalAlignment Enum=HAlign_Center
    Vertical:EVerticalAlignment Enum=VAlign_Center
   ) -> (then)
  ]}
]
variables [
 BlocksTiming {
  type=UIKitBlockTiming
  default {
   deselectingFocused=0
   deselectingIdle=0.2
   disabling=0
   enabling=0
   focusing=0
   hovering=0
   pressing=0
   releasing=0
   selecting=0.2
   unfocusing=0
   unhovering=0
  }
 }
 Selected {type=boolean}
 SwitcherBockgroundmaterial {type=MaterialInstance, default=/FortUILibrary/UIKit/Buttons/MaterialInstances/SelectableButtons/MI_UIKit_ToggleSwitchedBackground}
]
widgets {
 info {
  inheritedWidgetCount=0
  namedSlotCount=2
  parentClass=/FortUILibrary/Systems/ModularButton/WBP_UIKit_ButtonGeneric_Base.WBP_UIKit_ButtonGeneric_Base_C
  rootWidgetClass=null
  widgetCount=4
 }
 widgets [
  Main_Root {widgetClassPath=&GridPanel, bIsVariable=true, widget=Main_Root}
  Block_Switcher {
   widgetClassPath=/FortUILibrary/Systems/ModularButton/Blocks/WBP_UIKit_Block_Switcher
   bIsVariable=true
   parent=Main_Root
   widget=Block_Switcher
   slot {class=GridSlot, properties {verticalAlignment=VAlign_Center}}
  }
  Switcher {
   widgetClassPath=&GridPanel
   bIsVariable=true
   namedSlotHost=Block_Switcher
   widget=Switcher
   properties {columnFill [1], rowFill [1]}
  }
  Block_Outline {
   widgetClassPath=/FortUILibrary/UIKit/Buttons/Blocks/WBP_UIKit_Block_Outline
   bIsVariable=true
   parent=Main_Root
   widget=Block_Outline
   properties {bottomLeftCornerRadius=1, bottomRightCornerRadius=1, topLeftCornerRadius=1, topRightCornerRadius=1}
   slot {class=GridSlot, properties {layer=1}}
  }
 ]
}
